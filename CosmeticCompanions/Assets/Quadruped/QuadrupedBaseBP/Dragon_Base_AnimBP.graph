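# Base Animation Blueprint for quadruped dragon cosmetic companions on the TalonScone skeleton, resolving animations through Chooser proxy assets and an AnimTable.
# State machines build high- and low-fidelity poses, inertially switched, with additive banking, breathing, tail and antic layers.
# Emotes, look-at aim offset, voice-driven jaw curves and montage slots follow.
# Dead blending smooths this result before quadruped slope warping or root slope rotation, then four-leg IK.
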AnimBlueprint /CosmeticCompanions/Assets/Quadruped/QuadrupedBaseBP/Dragon_Base_AnimBP
target_skeleton=/CosmeticCompanions/Assets/Quadruped/TalonScone/Meshes/TalonScone_Quadruped_Skeleton
graphs [
 AllowedToSit_Exit {nodes [
   AllowedToSit_Exit!() -> (then, Context:"Anim Update Context Structure (by ref)", Node:"Anim Node Reference Structure (by ref)") {is=title K2Node_FunctionEntry}
   > Set bIsAllowedToSit(bIsAllowedToSit:bool=true) -> (Output_Get:bool)
   > Return Node()
  ]}
 AllowsCinematicOverride {nodes [
   AllowsCinematicOverride() {is=title K2Node_FunctionEntry}
   > Return Node!(execute, ReturnValue:bool=false)
  ]}
 AnimGraph {nodes [
   n1 = Quadruped Slope Warping!(
    ComponentPose:Component Space Pose Link Structure=Local To Component(LocalPose=Use cached pose 'PreSlopeWarpingPose'())
    Alpha:float=Get SlopeWarpingAlpha() -> (SlopeWarpingAlpha:float)
    MaxStepHeight:float=50
    bOnlyAffectSkeletonRootBone:bool=true
    MaxSkeletonRootBoneRotationAngle:float=0
   ) -> (Pose:Component Space Pose Link Structure) {
    is=title FortAnimGraphNode_QuadrupedSlopeWarping
    bUseMovementComponentFloorInfoInsteadOfTraces=true
    alphaBoolBlend {blendOption=Linear}
    alphaScaleBiasClamp {bias=1, scale=-1}
    armsFeetDefinitions [
     {numBonesInLimb=3, fKFootBone {boneName=ball_fr_l}, iKFootBone {boneName=IK_Foot_Front_L}}
     {numBonesInLimb=3, fKFootBone {boneName=ball_fr_r}, iKFootBone {boneName=IK_Foot_Front_R}}
    ]
    armsIKFootRootBone {boneName=IK_Foot_Front_Root}
    armsRootBone {boneName=spine_03}
    feetDefinitions [
     {numBonesInLimb=3, fKFootBone {boneName=ball_bk_l}, iKFootBone {boneName=IK_Foot_Rear_L}}
     {numBonesInLimb=3, fKFootBone {boneName=ball_bk_r}, iKFootBone {boneName=IK_Foot_Rear_R}}
    ]
    iKFootRootBone {boneName=IK_Foot_Rear_Root}
    pelvisBone {boneName=hips}
    skeletonRootBone {boneName=root}
   }
   n2 = Component To Local(ComponentPose=@n1)
   n3 = "Transform (Modify) Bone - Bone: root"!(
    ComponentPose=Local To Component(LocalPose=Use cached pose 'PreSlopeWarpingPose'())
    Alpha=Get SlopeMatchingWeight() -> (SlopeMatchingWeight:float)
    Rotation=Get SlopeMatchingRotationOffset() -> (SlopeMatchingRotationOffset:Rotator)
   ) -> (Pose) {
    rotationMode=BMM_Additive
    rotationSpace=BCS_WorldSpace
    alphaScaleBiasClamp {bInterpResult=true, interpSpeedDecreasing=8, interpSpeedIncreasing=8}
    boneToModify {boneName=root}
   }
   n4 = Component To Local(ComponentPose=@n3)
   n5 = Get bUseRootSlopeOrientation() -> (bUseRootSlopeOrientation:bool)
   n6 = Leg IK!(
    ComponentPose:Component Space Pose Link Structure=Local To Component(LocalPose=Blend Poses by bool(BlendPose_0=@n2, BlendPose_1=@n4, BlendTime_0=0.3, BlendTime_1=0.3, bActiveValue=@n5) {blendType=Linear})
    Alpha:float=Get IKAlpha() -> (IKAlpha:float)
   ) -> (Pose:Component Space Pose Link Structure) {
    is=title AnimGraphNode_LegIK
    alphaBoolBlend {blendOption=Linear}
    alphaScaleBiasClamp {bias=1, scale=-1}
    legsDefinition [
     {
      bEnableKneeTwistCorrection=true
      bEnableRotationLimit=false
      footBoneForwardAxis=Y
      hingeRotationAxis=None
      minRotationAngle=15
      numBonesInLimb=3
      twistOffsetCurveName=None
      fKFootBone {boneName=ball_bk_l}
      iKFootBone {boneName=IK_Foot_Rear_L}
     }
     {
      bEnableKneeTwistCorrection=true
      bEnableRotationLimit=false
      footBoneForwardAxis=Y
      hingeRotationAxis=None
      minRotationAngle=15
      numBonesInLimb=3
      twistOffsetCurveName=None
      fKFootBone {boneName=ball_bk_r}
      iKFootBone {boneName=IK_Foot_Rear_R}
     }
     {
      bEnableKneeTwistCorrection=true
      bEnableRotationLimit=false
      footBoneForwardAxis=Y
      hingeRotationAxis=None
      minRotationAngle=15
      numBonesInLimb=3
      twistOffsetCurveName=None
      fKFootBone {boneName=ball_fr_r}
      iKFootBone {boneName=IK_Foot_Front_R}
     }
     {
      bEnableKneeTwistCorrection=true
      bEnableRotationLimit=false
      footBoneForwardAxis=Y
      hingeRotationAxis=None
      minRotationAngle=15
      numBonesInLimb=3
      twistOffsetCurveName=None
      fKFootBone {boneName=ball_fr_l}
      iKFootBone {boneName=IK_Foot_Front_L}
     }
    ]
   }
   Output Pose(Result=Component To Local(ComponentPose=@n6))
   n7 = Blend Poses by bool(
    BlendPose_0=Use cached pose 'LowFidelityPose'()
    BlendPose_1=Use cached pose 'HighFidelityPose'()
    BlendTime_0=0.5
    BlendTime_1=0.5
    bActiveValue=Get bUseLowFidelity() -> (bUseLowFidelity:bool)
   ) {blendType=Linear, transitionType=Inertialization}
   n8 = Get CompanionDirection() -> (CompanionDirection:float)
   AnimGraphNode_ChooserPlayer_1 = Chooser Player: PA_COR_Loco_Banking_BS!(BlendSpaceX:float=@n8) -> (Pose:Pose Link Structure) {
    is=title AnimGraphNode_ChooserPlayer
    evaluationFrequency=OnInitialUpdate
    chooser {
     _structType=/Script/ProxyTable.LookupProxy
     proxy=/CosmeticCompanions/Chooser/AnimationProxyAssets/Core/Locomotion/PA_COR_Loco_Banking_BS
     proxyTable {
      _structType=/Script/ProxyTable.ProxyTableContextProperty
      binding {contextIndex=0, displayName=Anim Table, isBoundToRoot=false, propertyBindingChain [AnimTable]}
     }
    }
    chooserContextDefinition [
     {
      _structType=/Script/Chooser.ContextObjectTypeClass
      class=&/Dragon_Base_AnimBP.Dragon_Base_AnimBP_C
      direction=ReadWrite
     }
     {struct=/Script/Chooser.ChooserPlayerSettings, _structType=/Script/Chooser.ContextObjectTypeStruct, direction=Read}
    ]
   }
   AnimGraphNode_ChooserPlayer_4 = Chooser Player: PA_COR_Loco_Turn_BS!(BlendSpaceX:float=@n8) -> (Pose:Pose Link Structure) {
    is=title AnimGraphNode_ChooserPlayer
    evaluationFrequency=OnInitialUpdate
    chooser {
     _structType=&LookupProxy
     proxy=/CosmeticCompanions/Chooser/AnimationProxyAssets/Core/Locomotion/PA_COR_Loco_Turn_BS
     proxyTable {
      _structType=&ProxyTableContextProperty
      binding {contextIndex=0, displayName=Anim Table, isBoundToRoot=false, propertyBindingChain [AnimTable]}
     }
    }
    chooserContextDefinition [
     {
      _structType=&ContextObjectTypeClass
      class=&Dragon_Base_AnimBP_C
      direction=ReadWrite
     }
     {struct=&ChooserPlayerSettings, _structType=&ContextObjectTypeStruct, direction=Read}
    ]
   }
   n9 = Two Way Blend!(A=@AnimGraphNode_ChooserPlayer_4, B=@AnimGraphNode_ChooserPlayer_1, AlphaCurveName:Name=Flying) -> (Pose) {blendNode {alphaInputType=Curve}}
   AnimGraphNode_ChooserPlayer_3 = Chooser Player: PA_COR_Breathing_ADD!() -> (Pose:Pose Link Structure) {
    is=title AnimGraphNode_ChooserPlayer
    evaluationFrequency=OnInitialUpdate
    chooser {
     _structType=&LookupProxy
     proxy=/CosmeticCompanions/Chooser/AnimationProxyAssets/Core/PA_COR_Breathing_ADD
     proxyTable {
      _structType=&ProxyTableContextProperty
      binding {contextIndex=0, displayName=Anim Table, isBoundToRoot=false, propertyBindingChain [AnimTable]}
     }
    }
    chooserContextDefinition [
     {
      _structType=&ContextObjectTypeClass
      class=&Dragon_Base_AnimBP_C
      direction=ReadWrite
     }
     {struct=&ChooserPlayerSettings, _structType=&ContextObjectTypeStruct, direction=Read}
    ]
    defaultSettings {bForceLooping=true}
   }
   Save cached pose 'PreEmotePose'(Pose=Apply Additive(
     Base=Apply Additive(Base=@n7, Additive=@n9) {alphaBoolBlend {blendOption=Linear}}
     Additive=@AnimGraphNode_ChooserPlayer_3
     Alpha=Get BreathingAlpha() -> (BreathingAlpha:double)
    ) {alphaBoolBlend {blendOption=Linear}}) {cacheName=PreEmotePose}
   AnimGraphNode_StateMachine_4 = Emote SM() {is=title AnimGraphNode_StateMachine}
   n10 = "AimOffset Player '(None)'"!(
    X:float=Get LookAtYawAdjusted() -> (LookAtYawAdjusted:double)
    Y:float=Get LookAtPitchActual() -> (LookAtPitchActual:float)
    BasePose:Pose Link Structure=Slot 'DefaultSlot'(Source=@AnimGraphNode_StateMachine_4)
    Alpha:float=Get AimOffsetAlpha() -> (AimOffsetAlpha:float)
   ) -> (Pose:Pose Link Structure) {
    initialUpdateFunction {bSelfContext=true, memberGuid=#1, memberName=OnInitialUpdate}
    node {
     alphaBoolBlend {blendInTime=0.2, blendOutTime=0.2}
     alphaScaleBiasClamp {bInterpResult=true, bias=1, interpSpeedDecreasing=3, interpSpeedIncreasing=3, scale=-1}
     initialUpdateFunction {functionName=OnInitialUpdate}
    }
   }
   Save cached pose 'PreVO'(Pose=@n10) {cacheName=PreVO}
   AnimGraphNode_StateMachine_3 = LowFidelityStateMachine() {is=title AnimGraphNode_StateMachine, bAllowConduitEntryStates=true}
   n11 = Apply Additive(Base=@AnimGraphNode_StateMachine_3, Additive=Use cached pose 'BodyAdditivePose'()) {alphaBoolBlend {blendOption=Linear}}
   AnimGraphNode_ChooserPlayer_0 = Chooser Player: PA_ACT_Break_ADD!() -> (Pose:Pose Link Structure) {
    is=title AnimGraphNode_ChooserPlayer
    evaluationFrequency=OnInitialUpdate
    chooser {
     _structType=&LookupProxy
     proxy=/CosmeticCompanions/Chooser/AnimationProxyAssets/Actions/PA_ACT_Break_ADD
     proxyTable {
      _structType=&ProxyTableContextProperty
      binding {contextIndex=0, displayName=Anim Table, isBoundToRoot=false, propertyBindingChain [AnimTable]}
     }
    }
    chooserContextDefinition [
     {
      _structType=&ContextObjectTypeClass
      class=&Dragon_Base_AnimBP_C
      direction=ReadWrite
     }
     {struct=&ChooserPlayerSettings, _structType=&ContextObjectTypeStruct, direction=Read}
    ]
   }
   n12 = Get IdleAnticAlpha() -> (IdleAnticAlpha:double)
   Save cached pose 'LowFidelityPose'(Pose=Apply Additive(
     Base=Apply Additive(Base=@n11, Additive=@AnimGraphNode_ChooserPlayer_0, Alpha=@n12) {alphaBoolBlend {blendOption=Linear}}
     Additive=Use cached pose 'DefaultTailWagPose'()
     Alpha=Get TailWagAlpha() -> (TailWagAlpha:float)
    ) {alphaCurveName=Appendage01, alphaBoolBlend {blendOption=Linear}}) {cacheName=LowFidelityPose}
   AnimGraphNode_StateMachine_0 = FlyHeightSM() {is=title AnimGraphNode_StateMachine, bSkipFirstUpdateTransition=false}
   AnimGraphNode_StateMachine_5 = Locomotion() {is=title AnimGraphNode_StateMachine, bAllowConduitEntryStates=true}
   n13 = Apply Additive!(Base=@AnimGraphNode_StateMachine_5, Additive=@AnimGraphNode_StateMachine_0, AlphaCurveName:Name=Flying) -> (Pose) {alphaInputType=Curve, alphaScaleBiasClamp {interpSpeedDecreasing=3, interpSpeedIncreasing=3}}
   n14 = Use cached pose 'BodyAdditivePose'()
   AnimGraphNode_StateMachine_8 = IdleAntics() {is=title AnimGraphNode_StateMachine, bAllowConduitEntryStates=true}
   AnimGraphNode_StateMachine_2 = Tail() {is=title AnimGraphNode_StateMachine, bAllowConduitEntryStates=true}
   n15 = Apply Additive!(
    Base=Apply Additive(
     Base=Apply Additive!(Base=@n13, Additive=@n14) -> (Pose) {alphaBoolBlend {blendInTime=0.25, blendOption=Sinusoidal}}
     Additive=@AnimGraphNode_StateMachine_8
     Alpha=Get IdleAnticAlpha()
    ) {alphaBoolBlend {blendOption=Linear}}
    Additive=@AnimGraphNode_StateMachine_2
    Alpha:float=Get TailWagAlpha()
   ) -> (Pose) {alphaCurveName=Appendage01, alphaBoolBlend {blendOption=Linear}, alphaScaleBiasClamp {interpSpeedDecreasing=8, interpSpeedIncreasing=8}}
   Save cached pose 'HighFidelityPose'(Pose=@n15) {cacheName=HighFidelityPose}
   n16 = Modify Curve(SourcePose=Use cached pose 'PreVO'(), CurveValues_0=1, Alpha=Get AudioEnvelopeValue() -> (AudioEnvelopeValue:float)) {curveNames [VOActivity], curveValues [0]}
   n17 = Get JawOpenValue() -> (JawOpenValue:float)
   n18 = Slot 'PostVO'(Source=Blend Poses by bool(
     BlendPose_0=Modify Curve(SourcePose=@n16, CurveValues_0=0.6, Alpha=@n17) {curveNames [jaw_open_pose], curveValues [0]}
     BlendPose_1=Use cached pose 'PreVO'()
     bActiveValue=Get bIsVOActive() -> (bIsVOActive:bool)
    )) {slotName=PostVO}
   Save cached pose 'PreSlopeWarpingPose'(Pose=Dead Blending!(Source:Pose Link Structure=Slot 'SequencerDefault'(Source=@n18) {slotName=SequencerDefault}) -> (Pose:Pose Link Structure) {is=title AnimGraphNode_DeadBlending, defaultBlendMode=Linear, filteredCurves [Reaction_Happy, Reaction_Happy_01, EmoteCurveA, EmoteCurveB]}) {cacheName=PreSlopeWarpingPose}
   AnimGraphNode_StateMachine_1 = Idle() {is=title AnimGraphNode_StateMachine, bAllowConduitEntryStates=true}
   Save cached pose 'IdlePose'(Pose=@AnimGraphNode_StateMachine_1) {cacheName=IdlePose}
   AnimGraphNode_StateMachine_7 = MovementSM() {is=title AnimGraphNode_StateMachine}
   Save cached pose 'MovementPose'(Pose=@AnimGraphNode_StateMachine_7) {cacheName=MovementPose}
   AnimGraphNode_StateMachine_6 = BodyAdditives() {is=title AnimGraphNode_StateMachine}
   Save cached pose 'BodyAdditivePose'(Pose=@AnimGraphNode_StateMachine_6) {cacheName=BodyAdditivePose}
   Save cached pose 'DefaultTailWagPose'(Pose=DefaultTailWag!() -> (Pose:Pose Link Structure) {is=title AnimGraphNode_LinkedAnimLayer, layer=DefaultTailWag}) {cacheName=DefaultTailWagPose}
   Quadruped Slope Warping(Alpha=Get SlopeWarpingAlpha(), MaxStepHeight=50, bOnlyAffectSkeletonRootBone=False, MaxSkeletonRootBoneRotationAngle=0) {
    bUseMovementComponentFloorInfoInsteadOfTraces=true
    alphaBoolBlend {blendOption=Linear}
    alphaScaleBiasClamp {bias=1, scale=-1}
    armsFeetDefinitions [
     {numBonesInLimb=3, fKFootBone {boneName=ball_fr_l}, iKFootBone {boneName=IK_Foot_Front_L}}
     {numBonesInLimb=3, fKFootBone {boneName=ball_fr_r}, iKFootBone {boneName=IK_Foot_Front_R}}
    ]
    armsIKFootRootBone {boneName=IK_Foot_Front_Root}
    armsRootBone {boneName=spine_03}
    feetDefinitions [
     {numBonesInLimb=3, fKFootBone {boneName=ball_bk_l}, iKFootBone {boneName=IK_Foot_Rear_L}}
     {numBonesInLimb=3, fKFootBone {boneName=ball_bk_r}, iKFootBone {boneName=IK_Foot_Rear_R}}
    ]
    iKFootRootBone {boneName=IK_Foot_Rear_Root}
    pelvisBone {boneName=hips}
    skeletonRootBone {boneName=root}
   }
  ]}
 AnimationBlendStackGraph_0 {
  path=AnimGraph.AnimGraphNode_ChooserPlayer_0.AnimationBlendStackGraph_0
  nodes [Output Pose!(Result:Pose Link Structure=Blend Stack Input!() -> (Pose:Pose Link Structure) {is=title AnimGraphNode_BlendStackInput}) {is=title AnimGraphNode_BlendStackResult}]
 }
 AnimationBlendStackGraph_0 {path=AnimGraph.AnimGraphNode_ChooserPlayer_1.AnimationBlendStackGraph_0, nodes [Output Pose(Result=Blend Stack Input())]}
 AnimationBlendStackGraph_0 {
  path=AnimGraph.AnimGraphNode_ChooserPlayer_3.AnimationBlendStackGraph_0
  nodes [Output Pose(Result=Blend Stack Input!(PlayRate:float=Get BreathingPlayRate() -> (BreathingPlayRate:float)) -> (Pose) {bOverridePlayRate=true})]
 }
 AnimationBlendStackGraph_0 {path=AnimGraph.AnimGraphNode_ChooserPlayer_4.AnimationBlendStackGraph_0, nodes [Output Pose(Result=Blend Stack Input!() -> (Pose))]}
 FlyHeightSM {
  entry=NoHeight
  path=AnimGraph.AnimGraphNode_StateMachine_0.FlyHeightSM
  ids=id
  nodes [
   AnimStateNode_0 = NoHeight() {is=title AnimStateNode}
   AnimStateNode_1 = MaxHeight() {is=title AnimStateNode}
   AnimStateTransitionNode_2 = AnimStateTransitionNode() {bAllowInertializationForSelfTransitions=true, bDisabled=true, blendMode=CircularIn, crossfadeDuration=0.5}
   AnimStateTransitionNode_4 = AnimStateTransitionNode() {crossfadeDuration=5}
  ]
  links [
   {
    dst=NoHeight
    rule=AnimGraph.AnimGraphNode_StateMachine_0.FlyHeightSM.AnimStateTransitionNode_2.Transition
    src=MaxHeight
    via=AnimStateTransitionNode_2
   }
   {
    dst=MaxHeight
    rule=AnimGraph.AnimGraphNode_StateMachine_0.FlyHeightSM.AnimStateTransitionNode_4.Transition
    src=NoHeight
    via=AnimStateTransitionNode_4
   }
  ]
 }
 NoHeight {
  path=AnimGraph.AnimGraphNode_StateMachine_0.FlyHeightSM.AnimStateNode_0.NoHeight
  nodes [Output Animation Pose(Result=Additive Identity Pose())]
 }
 MaxHeight {
  path=AnimGraph.AnimGraphNode_StateMachine_0.FlyHeightSM.AnimStateNode_1.MaxHeight
  nodes [
   AnimGraphNode_ChooserPlayer_0 = Chooser Player: PA_COR_Loco_FlyHeight_ADD!() -> (Pose:Pose Link Structure) {
    is=title AnimGraphNode_ChooserPlayer
    evaluationFrequency=OnInitialUpdate
    chooser {
     _structType=&LookupProxy
     proxy=/CosmeticCompanions/Chooser/AnimationProxyAssets/Core/Locomotion/PA_COR_Loco_FlyHeight_ADD
     proxyTable {
      _structType=&ProxyTableContextProperty
      binding {contextIndex=0, displayName=Anim Table, isBoundToRoot=false, propertyBindingChain [AnimTable]}
     }
    }
    chooserContextDefinition [
     {
      _structType=&ContextObjectTypeClass
      class=&Dragon_Base_AnimBP_C
      direction=ReadWrite
     }
     {struct=&ChooserPlayerSettings, _structType=&ContextObjectTypeStruct, direction=Read}
    ]
   }
   Output Animation Pose(Result=@AnimGraphNode_ChooserPlayer_0)
  ]
 }
 AnimationBlendStackGraph_0 {
  path=AnimGraph.AnimGraphNode_StateMachine_0.FlyHeightSM.AnimStateNode_1.MaxHeight.AnimGraphNode_ChooserPlayer_0.AnimationBlendStackGraph_0
  nodes [Output Pose(Result=Blend Stack Input())]
 }
 Transition {
  path=AnimGraph.AnimGraphNode_StateMachine_0.FlyHeightSM.AnimStateTransitionNode_2.Transition
  nodes [
   Result(bCanEnterTransition=NOT Boolean(A=Get bIsFlying() -> (bIsFlying:bool)))
   Get bFlyStateActive() -> (bFlyStateActive:bool)
  ]
 }
 Transition {
  path=AnimGraph.AnimGraphNode_StateMachine_0.FlyHeightSM.AnimStateTransitionNode_4.Transition
  nodes [
   Result(bCanEnterTransition=true)
   Get bIsFlying()
   Get bFlyStateActive()
  ]
 }
 Idle {
  entry=Conduit
  path=AnimGraph.AnimGraphNode_StateMachine_1.Idle
  ids=id
  nodes [
   Alias() {is=title AnimStateAliasNode, also {aliases [SitPose,StandToSit]}}
   AnimStateConduitNode_0 = Conduit()
   AnimStateNode_0 = StandToSit() {is=title AnimStateNode}
   AnimStateNode_1 = SitPose() {is=title AnimStateNode}
   AnimStateNode_2 = StandingPose() {is=title AnimStateNode}
   AnimStateNode_3 = SitToStand() {is=title AnimStateNode}
   AnimStateTransitionNode_0 = AnimStateTransitionNode() {automaticRuleTriggerTime=0, bAutomaticRuleBasedOnSequencePlayerInState=true, logicType=TLT_Inertialization}
   AnimStateTransitionNode_1 = AnimStateTransitionNode() {logicType=TLT_Inertialization, priorityOrder=2}
   AnimStateTransitionNode_2 = AnimStateTransitionNode() {logicType=TLT_Inertialization, priorityOrder=100}
   AnimStateTransitionNode_3 = AnimStateTransitionNode() {logicType=TLT_Inertialization}
   AnimStateTransitionNode_4 = AnimStateTransitionNode()
   AnimStateTransitionNode_6 = AnimStateTransitionNode() {bAutomaticRuleBasedOnSequencePlayerInState=true}
   AnimStateTransitionNode_8 = AnimStateTransitionNode() {logicType=TLT_Inertialization}
  ]
  links [
   {
    dst=SitToStand
    rule=AnimGraph.AnimGraphNode_StateMachine_1.Idle.AnimStateTransitionNode_8.AnimationTransitionGraph_1
    src=Alias
    via=AnimStateTransitionNode_8
   }
   {
    dst=SitPose
    rule=AnimGraph.AnimGraphNode_StateMachine_1.Idle.AnimStateTransitionNode_3.AnimationTransitionGraph_1
    src=Conduit
    via=AnimStateTransitionNode_3
   }
   {
    dst=StandToSit
    rule=AnimGraph.AnimGraphNode_StateMachine_1.Idle.AnimStateTransitionNode_1.AnimationTransitionGraph_1
    src=Conduit
    via=AnimStateTransitionNode_1
   }
   {
    dst=StandingPose
    rule=AnimGraph.AnimGraphNode_StateMachine_1.Idle.AnimStateTransitionNode_2.AnimationTransitionGraph_1
    src=Conduit
    via=AnimStateTransitionNode_2
   }
   {
    dst=StandingPose
    rule=AnimGraph.AnimGraphNode_StateMachine_1.Idle.AnimStateTransitionNode_6.AnimationTransitionGraph_1
    src=SitToStand
    via=AnimStateTransitionNode_6
   }
   {
    dst=SitPose
    rule=AnimGraph.AnimGraphNode_StateMachine_1.Idle.AnimStateTransitionNode_0.AnimationTransitionGraph_1
    src=StandToSit
    via=AnimStateTransitionNode_0
   }
   {
    dst=Conduit
    rule=AnimGraph.AnimGraphNode_StateMachine_1.Idle.AnimStateTransitionNode_4.AnimationTransitionGraph_1
    src=StandingPose
    via=AnimStateTransitionNode_4
   }
  ]
 }
 Conduit {path=AnimGraph.AnimGraphNode_StateMachine_1.Idle.AnimStateConduitNode_0.Conduit, nodes [Result(bCanEnterTransition=true)]}
 StandToSit {
  path=AnimGraph.AnimGraphNode_StateMachine_1.Idle.AnimStateNode_0.StandToSit
  nodes [
   AnimGraphNode_ChooserPlayer_3 = Chooser Player: PA_COR_Bored_Idle_In!() -> (Pose:Pose Link Structure) {
    is=title AnimGraphNode_ChooserPlayer
    evaluationFrequency=OnInitialUpdate
    chooser {
     _structType=&LookupProxy
     proxy=/CosmeticCompanions/Chooser/AnimationProxyAssets/Core/PA_COR_Bored_Idle_In
     proxyTable {
      _structType=&ProxyTableContextProperty
      binding {contextIndex=0, displayName=Anim Table, isBoundToRoot=false, propertyBindingChain [AnimTable]}
     }
    }
    chooserContextDefinition [
     {
      _structType=&ContextObjectTypeClass
      class=&Dragon_Base_AnimBP_C
      direction=ReadWrite
     }
     {struct=&ChooserPlayerSettings, _structType=&ContextObjectTypeStruct, direction=Read}
    ]
   }
   Output Animation Pose(Result=Modify Curve(SourcePose=@AnimGraphNode_ChooserPlayer_3, CurveValues_0=1, Alpha=1) {curveNames [SitTail], curveValues [0]})
  ]
 }
 AnimationBlendStackGraph_0 {
  path=AnimGraph.AnimGraphNode_StateMachine_1.Idle.AnimStateNode_0.StandToSit.AnimGraphNode_ChooserPlayer_3.AnimationBlendStackGraph_0
  nodes [Output Pose(Result=Blend Stack Input())]
 }
 SitPose {
  path=AnimGraph.AnimGraphNode_StateMachine_1.Idle.AnimStateNode_1.SitPose
  nodes [
   AnimGraphNode_ChooserPlayer_3 = Chooser Player: PA_COR_Bored_Idle_Lp!() -> (Pose:Pose Link Structure) {
    is=title AnimGraphNode_ChooserPlayer
    evaluationFrequency=OnInitialUpdate
    chooser {
     _structType=&LookupProxy
     proxy=/CosmeticCompanions/Chooser/AnimationProxyAssets/Core/PA_COR_Bored_Idle_Lp
     proxyTable {
      _structType=&ProxyTableContextProperty
      binding {contextIndex=0, displayName=Anim Table, isBoundToRoot=false, propertyBindingChain [AnimTable]}
     }
    }
    chooserContextDefinition [
     {
      _structType=&ContextObjectTypeClass
      class=&Dragon_Base_AnimBP_C
      direction=ReadWrite
     }
     {struct=&ChooserPlayerSettings, _structType=&ContextObjectTypeStruct, direction=Read}
    ]
   }
   Output Animation Pose(Result=Modify Curve!(SourcePose=@AnimGraphNode_ChooserPlayer_3, CurveValues_0=1, CurveValues_1:float=1, CurveValues_2:float=1, Alpha=1) -> (Pose) {curveNames [idle,SitWeight,SitTail], curveValues [0,0,0]})
  ]
 }
 AnimationBlendStackGraph_0 {
  path=AnimGraph.AnimGraphNode_StateMachine_1.Idle.AnimStateNode_1.SitPose.AnimGraphNode_ChooserPlayer_3.AnimationBlendStackGraph_0
  nodes [Output Pose(Result=Blend Stack Input())]
 }
 StandingPose {
  path=AnimGraph.AnimGraphNode_StateMachine_1.Idle.AnimStateNode_2.StandingPose
  nodes [
   AnimGraphNode_StateMachine_0 = StandingStateMachine() {is=title AnimGraphNode_StateMachine, bAllowConduitEntryStates=true}
   Output Animation Pose(Result=Modify Curve!(SourcePose=@AnimGraphNode_StateMachine_0, CurveValues_0=1, Alpha=1) -> (Pose) {curveNames [idle], curveValues [0]})
  ]
 }
 StandingStateMachine {
  entry=Idle
  path=AnimGraph.AnimGraphNode_StateMachine_1.Idle.AnimStateNode_2.StandingPose.AnimGraphNode_StateMachine_0.StandingStateMachine
  ids=id
  nodes [
   AnimStateNode_0 = Idle() {is=title AnimStateNode}
   AnimStateNode_1 = AngryIdleIntro() {is=title AnimStateNode}
   AnimStateNode_2 = AngryIdle() {is=title AnimStateNode}
   AnimStateNode_3 = AngryIdleExit() {
    is=title AnimStateNode
    stateEntered {guid=#2}
    stateFullyBlended {guid=#3}
    stateLeft {guid=#4}
   }
   AnimStateTransitionNode_0 = AnimStateTransitionNode() {bAllowInertializationForSelfTransitions=true}
   AnimStateTransitionNode_2 = AnimStateTransitionNode() {bAllowInertializationForSelfTransitions=true, crossfadeDuration=0.1, logicType=TLT_Inertialization}
   AnimStateTransitionNode_3 = AnimStateTransitionNode() {
    automaticRuleTriggerTime=0
    bAllowInertializationForSelfTransitions=true
    bAutomaticRuleBasedOnSequencePlayerInState=true
    logicType=TLT_Inertialization
   }
   AnimStateTransitionNode_4 = AnimStateTransitionNode() {
    bAllowInertializationForSelfTransitions=true
    crossfadeDuration=0.1
    logicType=TLT_Inertialization
    priorityOrder=2
    transitionEnd {guid=#5}
    transitionInterrupt {guid=#6}
    transitionStart {guid=#7}
   }
   AnimStateTransitionNode_5 = AnimStateTransitionNode() {automaticRuleTriggerTime=0, bAllowInertializationForSelfTransitions=true, bAutomaticRuleBasedOnSequencePlayerInState=true}
   AnimStateTransitionNode_6 = AnimStateTransitionNode() {bAllowInertializationForSelfTransitions=true, bDisabled=true, crossfadeDuration=0.3, priorityOrder=2}
  ]
  links [
   {
    dst=AngryIdleExit
    rule=AnimGraph.AnimGraphNode_StateMachine_1.Idle.AnimStateNode_2.StandingPose.AnimGraphNode_StateMachine_0.StandingStateMachine.AnimStateTransitionNode_4.AnimationTransitionGraph_4
    src=AngryIdle
    via=AnimStateTransitionNode_4
   }
   {
    dst=Idle
    rule=AnimGraph.AnimGraphNode_StateMachine_1.Idle.AnimStateNode_2.StandingPose.AnimGraphNode_StateMachine_0.StandingStateMachine.AnimStateTransitionNode_0.Transition
    src=AngryIdle
    via=AnimStateTransitionNode_0
   }
   {
    dst=Idle
    rule=AnimGraph.AnimGraphNode_StateMachine_1.Idle.AnimStateNode_2.StandingPose.AnimGraphNode_StateMachine_0.StandingStateMachine.AnimStateTransitionNode_5.AnimationTransitionGraph_4
    src=AngryIdleExit
    via=AnimStateTransitionNode_5
   }
   {
    dst=AngryIdle
    rule=AnimGraph.AnimGraphNode_StateMachine_1.Idle.AnimStateNode_2.StandingPose.AnimGraphNode_StateMachine_0.StandingStateMachine.AnimStateTransitionNode_3.AnimationTransitionGraph_5
    src=AngryIdleIntro
    via=AnimStateTransitionNode_3
   }
   {
    dst=AngryIdle
    rule=AnimGraph.AnimGraphNode_StateMachine_1.Idle.AnimStateNode_2.StandingPose.AnimGraphNode_StateMachine_0.StandingStateMachine.AnimStateTransitionNode_6.Transition
    src=Idle
    via=AnimStateTransitionNode_6
   }
   {
    dst=AngryIdleIntro
    rule=AnimGraph.AnimGraphNode_StateMachine_1.Idle.AnimStateNode_2.StandingPose.AnimGraphNode_StateMachine_0.StandingStateMachine.AnimStateTransitionNode_2.Transition
    src=Idle
    via=AnimStateTransitionNode_2
   }
  ]
 }
 Idle {
  path=AnimGraph.AnimGraphNode_StateMachine_1.Idle.AnimStateNode_2.StandingPose.AnimGraphNode_StateMachine_0.StandingStateMachine.AnimStateNode_0.Idle
  nodes [
   Modify Curve(CurveValues_0=1, Alpha=1) {curveNames [idle], curveValues [0]}
   AnimGraphNode_ChooserPlayer_3 = Chooser Player: PA_COR_Stand_Pose!() -> (Pose:Pose Link Structure) {
    is=title AnimGraphNode_ChooserPlayer
    evaluationFrequency=OnInitialUpdate
    chooser {
     _structType=&LookupProxy
     proxy=/CosmeticCompanions/Chooser/AnimationProxyAssets/Core/PA_COR_Stand_Pose
     proxyTable {
      _structType=&ProxyTableContextProperty
      binding {contextIndex=0, displayName=Anim Table, isBoundToRoot=false, propertyBindingChain [AnimTable]}
     }
    }
    chooserContextDefinition [
     {
      _structType=&ContextObjectTypeClass
      class=&Dragon_Base_AnimBP_C
      direction=ReadWrite
     }
     {struct=&ChooserPlayerSettings, _structType=&ContextObjectTypeStruct, direction=Read}
    ]
   }
   Output Animation Pose(Result=@AnimGraphNode_ChooserPlayer_3)
  ]
 }
 AnimationBlendStackGraph_0 {
  path=AnimGraph.AnimGraphNode_StateMachine_1.Idle.AnimStateNode_2.StandingPose.AnimGraphNode_StateMachine_0.StandingStateMachine.AnimStateNode_0.Idle.AnimGraphNode_ChooserPlayer_3.AnimationBlendStackGraph_0
  nodes [Output Pose(Result=Blend Stack Input())]
 }
 AngryIdleIntro {
  path=AnimGraph.AnimGraphNode_StateMachine_1.Idle.AnimStateNode_2.StandingPose.AnimGraphNode_StateMachine_0.StandingStateMachine.AnimStateNode_1.AngryIdleIntro
  nodes [
   AnimGraphNode_ChooserPlayer_3 = Chooser Player: PA_COR_Angry_Idle_In!() -> (Pose:Pose Link Structure) {
    is=title AnimGraphNode_ChooserPlayer
    evaluationFrequency=OnInitialUpdate
    chooser {
     _structType=&LookupProxy
     proxy=/CosmeticCompanions/Chooser/AnimationProxyAssets/Core/PA_COR_Angry_Idle_In
     proxyTable {
      _structType=&ProxyTableContextProperty
      binding {contextIndex=0, displayName=Anim Table, isBoundToRoot=false, propertyBindingChain [AnimTable]}
     }
    }
    chooserContextDefinition [
     {
      _structType=&ContextObjectTypeClass
      class=&Dragon_Base_AnimBP_C
      direction=ReadWrite
     }
     {struct=&ChooserPlayerSettings, _structType=&ContextObjectTypeStruct, direction=Read}
    ]
   }
   Output Animation Pose(Result=@AnimGraphNode_ChooserPlayer_3)
  ]
 }
 AnimationBlendStackGraph_0 {
  path=AnimGraph.AnimGraphNode_StateMachine_1.Idle.AnimStateNode_2.StandingPose.AnimGraphNode_StateMachine_0.StandingStateMachine.AnimStateNode_1.AngryIdleIntro.AnimGraphNode_ChooserPlayer_3.AnimationBlendStackGraph_0
  nodes [Output Pose(Result=Blend Stack Input())]
 }
 AngryIdle {
  path=AnimGraph.AnimGraphNode_StateMachine_1.Idle.AnimStateNode_2.StandingPose.AnimGraphNode_StateMachine_0.StandingStateMachine.AnimStateNode_2.AngryIdle
  nodes [
   AnimGraphNode_ChooserPlayer_3 = Chooser Player: PA_COR_Angry_Idle_Lp!() -> (Pose:Pose Link Structure) {
    is=title AnimGraphNode_ChooserPlayer
    evaluationFrequency=OnInitialUpdate
    chooser {
     _structType=&LookupProxy
     proxy=/CosmeticCompanions/Chooser/AnimationProxyAssets/Core/PA_COR_Angry_Idle_Lp
     proxyTable {
      _structType=&ProxyTableContextProperty
      binding {contextIndex=0, displayName=Anim Table, isBoundToRoot=false, propertyBindingChain [AnimTable]}
     }
    }
    chooserContextDefinition [
     {
      _structType=&ContextObjectTypeClass
      class=&Dragon_Base_AnimBP_C
      direction=ReadWrite
     }
     {struct=&ChooserPlayerSettings, _structType=&ContextObjectTypeStruct, direction=Read}
    ]
    defaultSettings {bForceLooping=true}
   }
   Output Animation Pose(Result=@AnimGraphNode_ChooserPlayer_3)
  ]
 }
 AnimationBlendStackGraph_0 {
  path=AnimGraph.AnimGraphNode_StateMachine_1.Idle.AnimStateNode_2.StandingPose.AnimGraphNode_StateMachine_0.StandingStateMachine.AnimStateNode_2.AngryIdle.AnimGraphNode_ChooserPlayer_3.AnimationBlendStackGraph_0
  nodes [Output Pose(Result=Blend Stack Input())]
 }
 AngryIdleExit {
  path=AnimGraph.AnimGraphNode_StateMachine_1.Idle.AnimStateNode_2.StandingPose.AnimGraphNode_StateMachine_0.StandingStateMachine.AnimStateNode_3.AngryIdleExit
  nodes [
   AnimGraphNode_ChooserPlayer_3 = Chooser Player: PA_COR_Angry_Idle_Out!() -> (Pose:Pose Link Structure) {
    is=title AnimGraphNode_ChooserPlayer
    evaluationFrequency=OnInitialUpdate
    chooser {
     _structType=&LookupProxy
     proxy=/CosmeticCompanions/Chooser/AnimationProxyAssets/Core/PA_COR_Angry_Idle_Out
     proxyTable {
      _structType=&ProxyTableContextProperty
      binding {contextIndex=0, displayName=Anim Table, isBoundToRoot=false, propertyBindingChain [AnimTable]}
     }
    }
    chooserContextDefinition [
     {
      _structType=&ContextObjectTypeClass
      class=&Dragon_Base_AnimBP_C
      direction=ReadWrite
     }
     {struct=&ChooserPlayerSettings, _structType=&ContextObjectTypeStruct, direction=Read}
    ]
   }
   Output Animation Pose(Result=@AnimGraphNode_ChooserPlayer_3)
  ]
 }
 AnimationBlendStackGraph_0 {
  path=AnimGraph.AnimGraphNode_StateMachine_1.Idle.AnimStateNode_2.StandingPose.AnimGraphNode_StateMachine_0.StandingStateMachine.AnimStateNode_3.AngryIdleExit.AnimGraphNode_ChooserPlayer_3.AnimationBlendStackGraph_0
  nodes [Output Pose(Result=Blend Stack Input())]
 }
 Transition {
  path=AnimGraph.AnimGraphNode_StateMachine_1.Idle.AnimStateNode_2.StandingPose.AnimGraphNode_StateMachine_0.StandingStateMachine.AnimStateTransitionNode_0.Transition
  nodes [Result(bCanEnterTransition=Get bIsMoodWorried() -> (bIsMoodWorried:bool))]
 }
 Transition {
  path=AnimGraph.AnimGraphNode_StateMachine_1.Idle.AnimStateNode_2.StandingPose.AnimGraphNode_StateMachine_0.StandingStateMachine.AnimStateTransitionNode_2.Transition
  nodes [Result(bCanEnterTransition=Get bIsDelayedMoodCombat() -> (bIsDelayedMoodCombat:bool))]
 }
 AnimationTransitionGraph_5 {
  path=AnimGraph.AnimGraphNode_StateMachine_1.Idle.AnimStateNode_2.StandingPose.AnimGraphNode_StateMachine_0.StandingStateMachine.AnimStateTransitionNode_3.AnimationTransitionGraph_5
  nodes [Result()]
 }
 AnimationTransitionGraph_4 {
  path=AnimGraph.AnimGraphNode_StateMachine_1.Idle.AnimStateNode_2.StandingPose.AnimGraphNode_StateMachine_0.StandingStateMachine.AnimStateTransitionNode_4.AnimationTransitionGraph_4
  nodes [Result(bCanEnterTransition=NOT Boolean(A=Get bIsMoodCombat() -> (bIsMoodCombat:bool)))]
 }
 AnimationTransitionGraph_4 {
  path=AnimGraph.AnimGraphNode_StateMachine_1.Idle.AnimStateNode_2.StandingPose.AnimGraphNode_StateMachine_0.StandingStateMachine.AnimStateTransitionNode_5.AnimationTransitionGraph_4
  nodes [Result()]
 }
 Transition {
  path=AnimGraph.AnimGraphNode_StateMachine_1.Idle.AnimStateNode_2.StandingPose.AnimGraphNode_StateMachine_0.StandingStateMachine.AnimStateTransitionNode_6.Transition
  nodes [Result(bCanEnterTransition=Get bIsMoodCombat())]
 }
 SitToStand {
  path=AnimGraph.AnimGraphNode_StateMachine_1.Idle.AnimStateNode_3.SitToStand
  nodes [
   AnimGraphNode_ChooserPlayer_3 = Chooser Player: PA_COR_Bored_Idle_Out!() -> (Pose:Pose Link Structure) {
    is=title AnimGraphNode_ChooserPlayer
    evaluationFrequency=OnInitialUpdate
    chooser {
     _structType=&LookupProxy
     proxy=/CosmeticCompanions/Chooser/AnimationProxyAssets/Core/PA_COR_Bored_Idle_Out
     proxyTable {
      _structType=&ProxyTableContextProperty
      binding {contextIndex=0, displayName=Anim Table, isBoundToRoot=false, propertyBindingChain [AnimTable]}
     }
    }
    chooserContextDefinition [
     {
      _structType=&ContextObjectTypeClass
      class=&Dragon_Base_AnimBP_C
      direction=ReadWrite
     }
     {struct=&ChooserPlayerSettings, _structType=&ContextObjectTypeStruct, direction=Read}
    ]
   }
   Output Animation Pose(Result=@AnimGraphNode_ChooserPlayer_3)
  ]
 }
 AnimationBlendStackGraph_0 {
  path=AnimGraph.AnimGraphNode_StateMachine_1.Idle.AnimStateNode_3.SitToStand.AnimGraphNode_ChooserPlayer_3.AnimationBlendStackGraph_0
  nodes [Output Pose(Result=Blend Stack Input())]
 }
 AnimationTransitionGraph_1 {path=AnimGraph.AnimGraphNode_StateMachine_1.Idle.AnimStateTransitionNode_0.AnimationTransitionGraph_1, nodes [Result()]}
 AnimationTransitionGraph_1 {
  path=AnimGraph.AnimGraphNode_StateMachine_1.Idle.AnimStateTransitionNode_1.AnimationTransitionGraph_1
  nodes [Result(bCanEnterTransition=Get bIsStandToSit() -> (bIsStandToSit:bool))]
 }
 AnimationTransitionGraph_1 {
  path=AnimGraph.AnimGraphNode_StateMachine_1.Idle.AnimStateTransitionNode_2.AnimationTransitionGraph_1
  nodes [Result(bCanEnterTransition=true)]
 }
 AnimationTransitionGraph_1 {
  path=AnimGraph.AnimGraphNode_StateMachine_1.Idle.AnimStateTransitionNode_3.AnimationTransitionGraph_1
  nodes [Result(bCanEnterTransition=Get bIsSitting() -> (bIsSitting:bool))]
 }
 AnimationTransitionGraph_1 {
  path=AnimGraph.AnimGraphNode_StateMachine_1.Idle.AnimStateTransitionNode_4.AnimationTransitionGraph_1
  nodes [Result(bCanEnterTransition=true)]
 }
 AnimationTransitionGraph_1 {path=AnimGraph.AnimGraphNode_StateMachine_1.Idle.AnimStateTransitionNode_6.AnimationTransitionGraph_1, nodes [Result()]}
 AnimationTransitionGraph_1 {
  path=AnimGraph.AnimGraphNode_StateMachine_1.Idle.AnimStateTransitionNode_8.AnimationTransitionGraph_1
  nodes [Result(bCanEnterTransition=Get bIsSitToStand() -> (bIsSitToStand:bool))]
 }
 Tail {
  entry=Conduit
  path=AnimGraph.AnimGraphNode_StateMachine_2.Tail
  ids=id
  nodes [
   AnimStateConduitNode_0 = Conduit()
   AnimStateNode_0 = Default() {is=title AnimStateNode}
   AnimStateNode_2 = Combat() {is=title AnimStateNode}
   AnimStateTransitionNode_0 = AnimStateTransitionNode() {priorityOrder=100}
   AnimStateTransitionNode_1 = AnimStateTransitionNode()
   AnimStateTransitionNode_4 = AnimStateTransitionNode() {crossfadeDuration=0.5, priorityOrder=2}
   AnimStateTransitionNode_5 = AnimStateTransitionNode()
  ]
  links [
   {
    dst=Conduit
    rule=AnimGraph.AnimGraphNode_StateMachine_2.Tail.AnimStateTransitionNode_5.Transition
    src=Combat
    via=AnimStateTransitionNode_5
   }
   {
    dst=Combat
    rule=AnimGraph.AnimGraphNode_StateMachine_2.Tail.AnimStateTransitionNode_4.Transition
    src=Conduit
    via=AnimStateTransitionNode_4
   }
   {
    dst=Default
    rule=AnimGraph.AnimGraphNode_StateMachine_2.Tail.AnimStateTransitionNode_0.Transition
    src=Conduit
    via=AnimStateTransitionNode_0
   }
   {
    dst=Conduit
    rule=AnimGraph.AnimGraphNode_StateMachine_2.Tail.AnimStateTransitionNode_1.Transition
    src=Default
    via=AnimStateTransitionNode_1
   }
  ]
 }
 Conduit {path=AnimGraph.AnimGraphNode_StateMachine_2.Tail.AnimStateConduitNode_0.Conduit, nodes [Result(bCanEnterTransition=true)]}
 Default {
  path=AnimGraph.AnimGraphNode_StateMachine_2.Tail.AnimStateNode_0.Default
  nodes [Output Animation Pose(Result=Use cached pose 'DefaultTailWagPose'())]
 }
 Combat {
  path=AnimGraph.AnimGraphNode_StateMachine_2.Tail.AnimStateNode_2.Combat
  nodes [
   AnimGraphNode_ChooserPlayer_3 = Chooser Player: PA_COR_Tail_Angry_ADD!() -> (Pose:Pose Link Structure) {
    is=title AnimGraphNode_ChooserPlayer
    evaluationFrequency=OnInitialUpdate
    chooser {
     _structType=&LookupProxy
     proxy=/CosmeticCompanions/Chooser/AnimationProxyAssets/Core/Tail/PA_COR_Tail_Angry_ADD
     proxyTable {
      _structType=&ProxyTableContextProperty
      binding {contextIndex=0, displayName=Anim Table, isBoundToRoot=false, propertyBindingChain [AnimTable]}
     }
    }
    chooserContextDefinition [
     {
      _structType=&ContextObjectTypeClass
      class=&Dragon_Base_AnimBP_C
      direction=ReadWrite
     }
     {struct=&ChooserPlayerSettings, _structType=&ContextObjectTypeStruct, direction=Read}
    ]
   }
   Output Animation Pose(Result=@AnimGraphNode_ChooserPlayer_3)
  ]
 }
 AnimationBlendStackGraph_0 {
  path=AnimGraph.AnimGraphNode_StateMachine_2.Tail.AnimStateNode_2.Combat.AnimGraphNode_ChooserPlayer_3.AnimationBlendStackGraph_0
  nodes [Output Pose(Result=Blend Stack Input())]
 }
 Transition {path=AnimGraph.AnimGraphNode_StateMachine_2.Tail.AnimStateTransitionNode_0.Transition, nodes [Result(bCanEnterTransition=true)]}
 Transition {path=AnimGraph.AnimGraphNode_StateMachine_2.Tail.AnimStateTransitionNode_1.Transition, nodes [Result(bCanEnterTransition=true)]}
 Transition {
  path=AnimGraph.AnimGraphNode_StateMachine_2.Tail.AnimStateTransitionNode_4.Transition
  nodes [Result(bCanEnterTransition=Get bIsDelayedMoodCombat())]
 }
 Transition {path=AnimGraph.AnimGraphNode_StateMachine_2.Tail.AnimStateTransitionNode_5.Transition, nodes [Result(bCanEnterTransition=true)]}
 LowFidelityStateMachine {
  entry=Idle
  path=AnimGraph.AnimGraphNode_StateMachine_3.LowFidelityStateMachine
  ids=id
  nodes [
   AnimStateNode_0 = Idle() {
    bAlwaysResetOnEntry=true
    stateEntered {guid=#8}
    stateFullyBlended {guid=#9}
    stateLeft {guid=#10}
   }
   AnimStateNode_1 = Stop() {is=title AnimStateNode}
   AnimStateNode_3 = Locomotion() {is=title AnimStateNode, bAlwaysResetOnEntry=true}
   AnimStateTransitionNode_0 = AnimStateTransitionNode() {bAllowInertializationForSelfTransitions=true, bAutomaticRuleBasedOnSequencePlayerInState=true}
   AnimStateTransitionNode_1 = AnimStateTransitionNode() {bAllowInertializationForSelfTransitions=true}
   AnimStateTransitionNode_2 = AnimStateTransitionNode() {bAllowInertializationForSelfTransitions=true}
   AnimStateTransitionNode_4 = AnimStateTransitionNode() {crossfadeDuration=0.4, logicType=TLT_Inertialization}
   AnimStateTransitionNode_8 = AnimStateTransitionNode() {crossfadeDuration=0.4, priorityOrder=3}
  ]
  links [
   {
    dst=Locomotion
    rule=AnimGraph.AnimGraphNode_StateMachine_3.LowFidelityStateMachine.AnimStateTransitionNode_8.Transition
    src=Idle
    via=AnimStateTransitionNode_8
   }
   {
    dst=Idle
    rule=AnimGraph.AnimGraphNode_StateMachine_3.LowFidelityStateMachine.AnimStateTransitionNode_2.Transition
    src=Locomotion
    via=AnimStateTransitionNode_2
   }
   {
    dst=Stop
    rule=AnimGraph.AnimGraphNode_StateMachine_3.LowFidelityStateMachine.AnimStateTransitionNode_4.Transition
    src=Locomotion
    via=AnimStateTransitionNode_4
   }
   {
    dst=Idle
    rule=AnimGraph.AnimGraphNode_StateMachine_3.LowFidelityStateMachine.AnimStateTransitionNode_0.Transition
    src=Stop
    via=AnimStateTransitionNode_0
   }
   {
    dst=Locomotion
    rule=AnimGraph.AnimGraphNode_StateMachine_3.LowFidelityStateMachine.AnimStateTransitionNode_1.Transition
    src=Stop
    via=AnimStateTransitionNode_1
   }
  ]
 }
 Idle {
  path=AnimGraph.AnimGraphNode_StateMachine_3.LowFidelityStateMachine.AnimStateNode_0.Idle
  nodes [Output Animation Pose(Result=Use cached pose 'IdlePose'())]
 }
 Stop {
  path=AnimGraph.AnimGraphNode_StateMachine_3.LowFidelityStateMachine.AnimStateNode_1.Stop
  nodes [
   AnimGraphNode_ChooserPlayer_0 = Chooser Player: PRA_Companion_Stops!() -> (Pose:Pose Link Structure) {
    is=title AnimGraphNode_ChooserPlayer
    chooser {
     _structType=&LookupProxy
     proxy=/CosmeticCompanions/Chooser/PRA_Companion_Stops
     proxyTable {
      _structType=&ProxyTableContextProperty
      binding {contextIndex=0, displayName=Anim Table, isBoundToRoot=false, propertyBindingChain [AnimTable]}
     }
    }
    chooserContextDefinition [
     {
      _structType=&ContextObjectTypeClass
      class=&Dragon_Base_AnimBP_C
      direction=ReadWrite
     }
     {struct=&ChooserPlayerSettings, _structType=&ContextObjectTypeStruct, direction=Read}
    ]
   }
   Output Animation Pose(Result=@AnimGraphNode_ChooserPlayer_0)
  ]
 }
 AnimationBlendStackGraph_0 {
  path=AnimGraph.AnimGraphNode_StateMachine_3.LowFidelityStateMachine.AnimStateNode_1.Stop.AnimGraphNode_ChooserPlayer_0.AnimationBlendStackGraph_0
  nodes [Output Pose(Result=Blend Stack Input())]
 }
 Locomotion {
  path=AnimGraph.AnimGraphNode_StateMachine_3.LowFidelityStateMachine.AnimStateNode_3.Locomotion
  nodes [Output Animation Pose(Result=Use cached pose 'MovementPose'())]
 }
 Transition {path=AnimGraph.AnimGraphNode_StateMachine_3.LowFidelityStateMachine.AnimStateTransitionNode_0.Transition, nodes [Result()]}
 Transition {
  path=AnimGraph.AnimGraphNode_StateMachine_3.LowFidelityStateMachine.AnimStateTransitionNode_1.Transition
  nodes [Result(bCanEnterTransition=Get bIsStopToLocomotion() -> (bIsStopToLocomotion:bool))]
 }
 Transition {
  path=AnimGraph.AnimGraphNode_StateMachine_3.LowFidelityStateMachine.AnimStateTransitionNode_2.Transition
  nodes [Result(bCanEnterTransition=Get bExitLocomotion() -> (bExitLocomotion:bool))]
 }
 Transition {
  path=AnimGraph.AnimGraphNode_StateMachine_3.LowFidelityStateMachine.AnimStateTransitionNode_4.Transition
  nodes [Result(bCanEnterTransition=Get bIsLocomotionStop() -> (bIsLocomotionStop:bool))]
 }
 Transition {
  path=AnimGraph.AnimGraphNode_StateMachine_3.LowFidelityStateMachine.AnimStateTransitionNode_8.Transition
  nodes [Result(bCanEnterTransition=Get bIsMoving() -> (bIsMoving:bool))]
 }
 Emote SM {
  entry=Default
  path=AnimGraph.AnimGraphNode_StateMachine_4.Emote SM
  ids=id
  nodes [
   ToDefault() {is=title AnimStateAliasNode, also {aliases [EmoteLoops,ReactToIdle]}}
   AnimStateNode_1 = EmoteLoops() {is=title AnimStateNode}
   AnimStateNode_2 = ReactToIdle() {is=title AnimStateNode}
   AnimStateNode_6 = Default()
   AnimStateTransitionNode_0 = AnimStateTransitionNode() {bAllowInertializationForSelfTransitions=true}
   AnimStateTransitionNode_12 = AnimStateTransitionNode() {logicType=TLT_Inertialization}
   AnimStateTransitionNode_13 = AnimStateTransitionNode() {automaticRuleTriggerTime=0.2, bAutomaticRuleBasedOnSequencePlayerInState=true, crossfadeDuration=0.3}
   AnimStateTransitionNode_14 = AnimStateTransitionNode() {logicType=TLT_Inertialization, priorityOrder=2}
   AnimStateTransitionNode_7 = AnimStateTransitionNode() {logicType=TLT_Inertialization}
  ]
  links [
   {
    dst=EmoteLoops
    rule=AnimGraph.AnimGraphNode_StateMachine_4.Emote SM.AnimStateTransitionNode_7.AnimationTransitionGraph_5
    src=Default
    via=AnimStateTransitionNode_7
   }
   {
    dst=ReactToIdle
    rule=AnimGraph.AnimGraphNode_StateMachine_4.Emote SM.AnimStateTransitionNode_14.AnimationTransitionGraph_5
    src=Default
    via=AnimStateTransitionNode_14
   }
   {
    dst=ReactToIdle
    rule=AnimGraph.AnimGraphNode_StateMachine_4.Emote SM.AnimStateTransitionNode_12.AnimationTransitionGraph_5
    src=EmoteLoops
    via=AnimStateTransitionNode_12
   }
   {
    dst=Default
    rule=AnimGraph.AnimGraphNode_StateMachine_4.Emote SM.AnimStateTransitionNode_13.AnimationTransitionGraph_5
    src=ReactToIdle
    via=AnimStateTransitionNode_13
   }
   {
    dst=Default
    rule=AnimGraph.AnimGraphNode_StateMachine_4.Emote SM.AnimStateTransitionNode_0.Transition
    src=ToDefault
    via=AnimStateTransitionNode_0
   }
  ]
 }
 EmoteLoops {
  path=AnimGraph.AnimGraphNode_StateMachine_4.Emote SM.AnimStateNode_1.EmoteLoops
  nodes [
   AnimGraphNode_ChooserPlayer_0 = Chooser Player: PRA_Companion_SpecialIdle!() -> (Pose:Pose Link Structure) {
    is=title AnimGraphNode_ChooserPlayer
    chooser {
     _structType=&LookupProxy
     proxy=/CosmeticCompanions/Chooser/PRA_Companion_SpecialIdle
     proxyTable {
      _structType=&ProxyTableContextProperty
      binding {contextIndex=0, displayName=Anim Table, isBoundToRoot=false, propertyBindingChain [AnimTable]}
     }
    }
    chooserContextDefinition [
     {
      _structType=&ContextObjectTypeClass
      class=&Dragon_Base_AnimBP_C
      direction=ReadWrite
     }
     {struct=&ChooserPlayerSettings, _structType=&ContextObjectTypeStruct, direction=Read}
    ]
    defaultSettings {blendOption=Linear}
   }
   Output Animation Pose(Result=@AnimGraphNode_ChooserPlayer_0)
  ]
 }
 AnimationBlendStackGraph_0 {
  path=AnimGraph.AnimGraphNode_StateMachine_4.Emote SM.AnimStateNode_1.EmoteLoops.AnimGraphNode_ChooserPlayer_0.AnimationBlendStackGraph_0
  nodes [Output Pose(Result=Blend Stack Input())]
 }
 ReactToIdle {
  path=AnimGraph.AnimGraphNode_StateMachine_4.Emote SM.AnimStateNode_2.ReactToIdle
  nodes [
   AnimGraphNode_ChooserPlayer_0 = Chooser Player: PRA_Companion_ReactToIdle!() -> (Pose:Pose Link Structure) {
    is=title AnimGraphNode_ChooserPlayer
    chooser {
     _structType=&LookupProxy
     proxy=/CosmeticCompanions/Chooser/PRA_Companion_ReactToIdle
     proxyTable {
      _structType=&ProxyTableContextProperty
      binding {contextIndex=0, displayName=Anim Table, isBoundToRoot=false, propertyBindingChain [AnimTable]}
     }
    }
    chooserContextDefinition [
     {
      _structType=&ContextObjectTypeClass
      class=&Dragon_Base_AnimBP_C
      direction=ReadWrite
     }
     {struct=&ChooserPlayerSettings, _structType=&ContextObjectTypeStruct, direction=Read}
    ]
    defaultSettings {blendOption=Linear}
   }
   Output Animation Pose(Result=@AnimGraphNode_ChooserPlayer_0) {
    node {stateEntryFunction {functionName=SpecialIdle_StateEntry}, stateExitFunction {functionName=SpecialIdle_StateExit}}
    stateEntryFunction {bSelfContext=true, memberGuid=#11, memberName=SpecialIdle_StateEntry}
    stateExitFunction {bSelfContext=true, memberGuid=#12, memberName=SpecialIdle_StateExit}
   }
  ]
 }
 AnimationBlendStackGraph_0 {
  path=AnimGraph.AnimGraphNode_StateMachine_4.Emote SM.AnimStateNode_2.ReactToIdle.AnimGraphNode_ChooserPlayer_0.AnimationBlendStackGraph_0
  nodes [Output Pose(Result=Blend Stack Input())]
 }
 Default {
  path=AnimGraph.AnimGraphNode_StateMachine_4.Emote SM.AnimStateNode_6.Default
  nodes [Output Animation Pose(Result=Use cached pose 'PreEmotePose'())]
 }
 Transition {
  path=AnimGraph.AnimGraphNode_StateMachine_4.Emote SM.AnimStateTransitionNode_0.Transition
  nodes [Result(bCanEnterTransition=Get bIsMoving())]
 }
 AnimationTransitionGraph_5 {
  path=AnimGraph.AnimGraphNode_StateMachine_4.Emote SM.AnimStateTransitionNode_12.AnimationTransitionGraph_5
  nodes [Result(bCanEnterTransition=NOT Boolean(A=Get bIsPlayingEmote()))]
 }
 AnimationTransitionGraph_5 {path=AnimGraph.AnimGraphNode_StateMachine_4.Emote SM.AnimStateTransitionNode_13.AnimationTransitionGraph_5, nodes [Result()]}
 AnimationTransitionGraph_5 {
  path=AnimGraph.AnimGraphNode_StateMachine_4.Emote SM.AnimStateTransitionNode_14.AnimationTransitionGraph_5
  nodes [Result(bCanEnterTransition=Get bIsReactionHappyPlaying() -> (bIsReactionHappyPlaying:bool))]
 }
 AnimationTransitionGraph_5 {
  path=AnimGraph.AnimGraphNode_StateMachine_4.Emote SM.AnimStateTransitionNode_7.AnimationTransitionGraph_5
  nodes [Result(bCanEnterTransition=Get bIsSpecialIdleEnter() -> (bIsSpecialIdleEnter:bool))]
 }
 Locomotion {
  entry=Idle
  path=AnimGraph.AnimGraphNode_StateMachine_5.Locomotion
  ids=id
  nodes [
   IsTurning() {is=title AnimStateAliasNode, also {aliases [Idle, Locomotion, LocomotionStop]}}
   AnimStateNode_0 = Idle() {bAlwaysResetOnEntry=true}
   AnimStateNode_3 = Locomotion() {
    bAlwaysResetOnEntry=true
    stateEntered {guid=#13}
    stateFullyBlended {guid=#14}
    stateLeft {guid=#15}
   }
   AnimStateNode_4 = LocomotionStop() {
    is=title AnimStateNode
    stateEntered {guid=#16}
    stateFullyBlended {guid=#17}
    stateLeft {guid=#18}
   }
   AnimStateNode_5 = Start() {is=title AnimStateNode}
   AnimStateTransitionNode_1 = AnimStateTransitionNode() {blendMode=Linear, crossfadeDuration=0.15}
   AnimStateTransitionNode_10 = AnimStateTransitionNode() {
    priorityOrder=100
    transitionEnd {guid=#19}
    transitionInterrupt {guid=#20}
    transitionStart {guid=#21}
   }
   AnimStateTransitionNode_18 = AnimStateTransitionNode() {
    automaticRuleTriggerTime=0.1
    bAllowInertializationForSelfTransitions=true
    bAutomaticRuleBasedOnSequencePlayerInState=true
    blendMode=CubicInOut
   }
   AnimStateTransitionNode_3 = AnimStateTransitionNode() {blendMode=Linear, crossfadeDuration=0.1, logicType=TLT_Inertialization, priorityOrder=10}
   AnimStateTransitionNode_4 = AnimStateTransitionNode()
   AnimStateTransitionNode_5 = AnimStateTransitionNode() {bAutomaticRuleBasedOnSequencePlayerInState=true}
   AnimStateTransitionNode_8 = AnimStateTransitionNode() {blendMode=Linear, priorityOrder=3}
  ]
  links [
   {
    dst=Locomotion
    rule=AnimGraph.AnimGraphNode_StateMachine_5.Locomotion.AnimStateTransitionNode_8.AnimationTransitionGraph_0
    src=Idle
    via=AnimStateTransitionNode_8
   }
   {
    dst=Start
    rule=AnimGraph.AnimGraphNode_StateMachine_5.Locomotion.AnimStateTransitionNode_3.AnimationTransitionGraph_1
    src=IsTurning
    via=AnimStateTransitionNode_3
   }
   {
    dst=Idle
    rule=AnimGraph.AnimGraphNode_StateMachine_5.Locomotion.AnimStateTransitionNode_4.AnimationTransitionGraph_0
    src=Locomotion
    via=AnimStateTransitionNode_4
   }
   {
    dst=LocomotionStop
    rule=AnimGraph.AnimGraphNode_StateMachine_5.Locomotion.AnimStateTransitionNode_1.AnimationTransitionGraph_1
    src=Locomotion
    via=AnimStateTransitionNode_1
   }
   {
    dst=Idle
    rule=AnimGraph.AnimGraphNode_StateMachine_5.Locomotion.AnimStateTransitionNode_5.AnimationTransitionGraph_0
    src=LocomotionStop
    via=AnimStateTransitionNode_5
   }
   {
    dst=Locomotion
    rule=AnimGraph.AnimGraphNode_StateMachine_5.Locomotion.AnimStateTransitionNode_10.AnimationTransitionGraph_1
    src=LocomotionStop
    via=AnimStateTransitionNode_10
   }
   {
    dst=Locomotion
    rule=AnimGraph.AnimGraphNode_StateMachine_5.Locomotion.AnimStateTransitionNode_18.AnimationTransitionGraph_1
    src=Start
    via=AnimStateTransitionNode_18
   }
  ]
 }
 Idle {
  path=AnimGraph.AnimGraphNode_StateMachine_5.Locomotion.AnimStateNode_0.Idle
  nodes [Output Animation Pose(Result=Use cached pose 'IdlePose'())]
 }
 Locomotion {
  path=AnimGraph.AnimGraphNode_StateMachine_5.Locomotion.AnimStateNode_3.Locomotion
  nodes [Output Animation Pose(Result=Use cached pose 'MovementPose'())]
 }
 LocomotionStop {
  path=AnimGraph.AnimGraphNode_StateMachine_5.Locomotion.AnimStateNode_4.LocomotionStop
  nodes [
   AnimGraphNode_ChooserPlayer_0 = Chooser Player: PRA_Companion_Stops() {
    groupName=FlyStop
    method=SyncGroup
    chooser {
     _structType=&LookupProxy
     proxy=&PRA_Companion_Stops
     proxyTable {
      _structType=&ProxyTableContextProperty
      binding {contextIndex=0, displayName=Anim Table, isBoundToRoot=false, propertyBindingChain [AnimTable]}
     }
    }
    chooserContextDefinition [
     {
      _structType=&ContextObjectTypeClass
      class=&Dragon_Base_AnimBP_C
      direction=ReadWrite
     }
     {struct=&ChooserPlayerSettings, _structType=&ContextObjectTypeStruct, direction=Read}
    ]
   }
   Output Animation Pose(Result=@AnimGraphNode_ChooserPlayer_0)
  ]
 }
 AnimationBlendStackGraph_0 {
  path=AnimGraph.AnimGraphNode_StateMachine_5.Locomotion.AnimStateNode_4.LocomotionStop.AnimGraphNode_ChooserPlayer_0.AnimationBlendStackGraph_0
  nodes [Output Pose(Result=Blend Stack Input())]
 }
 Start {
  path=AnimGraph.AnimGraphNode_StateMachine_5.Locomotion.AnimStateNode_5.Start
  nodes [
   AnimGraphNode_ChooserPlayer_3 = Chooser Player: PA_COR_Loco_Start!() -> (Pose:Pose Link Structure) {
    is=title AnimGraphNode_ChooserPlayer
    evaluationFrequency=OnInitialUpdate
    chooser {
     _structType=&LookupProxy
     proxy=/CosmeticCompanions/Chooser/AnimationProxyAssets/Core/Locomotion/PA_COR_Loco_Start
     proxyTable {
      _structType=&ProxyTableContextProperty
      binding {contextIndex=0, displayName=Anim Table, isBoundToRoot=false, propertyBindingChain [AnimTable]}
     }
    }
    chooserContextDefinition [
     {
      _structType=&ContextObjectTypeClass
      class=&Dragon_Base_AnimBP_C
      direction=ReadWrite
     }
     {struct=&ChooserPlayerSettings, _structType=&ContextObjectTypeStruct, direction=Read}
    ]
   }
   Output Animation Pose(Result=Modify Curve(SourcePose=@AnimGraphNode_ChooserPlayer_3, CurveValues_0=1, Alpha=1) {curveNames [walk], curveValues [0]}) {
    node {stateEntryFunction {functionName=OnStateEntry}, stateExitFunction {functionName=OnStateExit}}
    stateEntryFunction {bSelfContext=true, memberGuid=#22, memberName=LocoStart_Entry}
    stateExitFunction {bSelfContext=true, memberGuid=#23, memberName=StartTurn_StateExit}
   }
  ]
 }
 AnimationBlendStackGraph_0 {
  path=AnimGraph.AnimGraphNode_StateMachine_5.Locomotion.AnimStateNode_5.Start.AnimGraphNode_ChooserPlayer_3.AnimationBlendStackGraph_0
  nodes [Output Pose(Result=Blend Stack Input())]
 }
 AnimationTransitionGraph_1 {
  path=AnimGraph.AnimGraphNode_StateMachine_5.Locomotion.AnimStateTransitionNode_1.AnimationTransitionGraph_1
  nodes [
   Result(bCanEnterTransition=Get bIsLocomotionStop())
   n1 = Get bIsMoving()
   AND Boolean(A=OR Boolean(A=NOT Boolean(A=@n1), B=Get bIsCompanionStopping() -> (bIsCompanionStopping:bool)), B=NOT Boolean(A=Get bIsMoodCombat()))
  ]
 }
 AnimationTransitionGraph_1 {
  path=AnimGraph.AnimGraphNode_StateMachine_5.Locomotion.AnimStateTransitionNode_10.AnimationTransitionGraph_1
  nodes [Result(bCanEnterTransition=Get bIsStopToLocomotion())]
 }
 AnimationTransitionGraph_1 {path=AnimGraph.AnimGraphNode_StateMachine_5.Locomotion.AnimStateTransitionNode_18.AnimationTransitionGraph_1, nodes [Result()]}
 AnimationTransitionGraph_1 {
  path=AnimGraph.AnimGraphNode_StateMachine_5.Locomotion.AnimStateTransitionNode_3.AnimationTransitionGraph_1
  nodes [Result(bCanEnterTransition=Get bTriggerStartTurn() -> (bTriggerStartTurn:bool))]
 }
 AnimationTransitionGraph_0 {
  path=AnimGraph.AnimGraphNode_StateMachine_5.Locomotion.AnimStateTransitionNode_4.AnimationTransitionGraph_0
  nodes [Result(bCanEnterTransition=Get bExitLocomotion())]
 }
 AnimationTransitionGraph_0 {path=AnimGraph.AnimGraphNode_StateMachine_5.Locomotion.AnimStateTransitionNode_5.AnimationTransitionGraph_0, nodes [Result()]}
 AnimationTransitionGraph_0 {
  path=AnimGraph.AnimGraphNode_StateMachine_5.Locomotion.AnimStateTransitionNode_8.AnimationTransitionGraph_0
  nodes [Result(bCanEnterTransition=Get bIsMoving())]
 }
 BodyAdditives {
  entry=BodyAdditiveTempState
  path=AnimGraph.AnimGraphNode_StateMachine_6.BodyAdditives
  ids=id
  nodes [
   AnimStateNode_0 = BodyAdditiveTempState() {is=title AnimStateNode}
   AnimStateNode_1 = Worried() {is=title AnimStateNode}
   AnimStateNode_2 = Combat()
   AnimStateTransitionNode_0 = AnimStateTransitionNode() {crossfadeDuration=0.35, logicType=TLT_Inertialization}
   AnimStateTransitionNode_1 = AnimStateTransitionNode() {crossfadeDuration=0.35, logicType=TLT_Inertialization}
   AnimStateTransitionNode_2 = AnimStateTransitionNode() {logicType=TLT_Inertialization, priorityOrder=2}
   AnimStateTransitionNode_3 = AnimStateTransitionNode() {crossfadeDuration=0.35}
  ]
  links [
   {
    dst=Combat
    rule=AnimGraph.AnimGraphNode_StateMachine_6.BodyAdditives.AnimStateTransitionNode_2.Transition
    src=BodyAdditiveTempState
    via=AnimStateTransitionNode_2
   }
   {
    dst=Worried
    rule=AnimGraph.AnimGraphNode_StateMachine_6.BodyAdditives.AnimStateTransitionNode_0.Transition
    src=BodyAdditiveTempState
    via=AnimStateTransitionNode_0
   }
   {
    dst=BodyAdditiveTempState
    rule=AnimGraph.AnimGraphNode_StateMachine_6.BodyAdditives.AnimStateTransitionNode_3.Transition
    src=Combat
    via=AnimStateTransitionNode_3
   }
   {
    dst=BodyAdditiveTempState
    rule=AnimGraph.AnimGraphNode_StateMachine_6.BodyAdditives.AnimStateTransitionNode_1.Transition
    src=Worried
    via=AnimStateTransitionNode_1
   }
  ]
 }
 BodyAdditiveTempState {
  path=AnimGraph.AnimGraphNode_StateMachine_6.BodyAdditives.AnimStateNode_0.BodyAdditiveTempState
  nodes [Output Animation Pose(Result=Additive Identity Pose())]
 }
 Worried {
  path=AnimGraph.AnimGraphNode_StateMachine_6.BodyAdditives.AnimStateNode_1.Worried
  nodes [
   AnimGraphNode_ChooserPlayer_3 = Chooser Player: PA_COR_Loco_Worried_ADD!() -> (Pose:Pose Link Structure) {
    is=title AnimGraphNode_ChooserPlayer
    evaluationFrequency=OnInitialUpdate
    chooser {
     _structType=&LookupProxy
     proxy=/CosmeticCompanions/Chooser/AnimationProxyAssets/Core/Locomotion/PA_COR_Loco_Worried_ADD
     proxyTable {
      _structType=&ProxyTableContextProperty
      binding {contextIndex=0, displayName=Anim Table, isBoundToRoot=false, propertyBindingChain [AnimTable]}
     }
    }
    chooserContextDefinition [
     {
      _structType=&ContextObjectTypeClass
      class=&Dragon_Base_AnimBP_C
      direction=ReadWrite
     }
     {struct=&ChooserPlayerSettings, _structType=&ContextObjectTypeStruct, direction=Read}
    ]
   }
   Output Animation Pose(Result=Two Way Blend!(A=Additive Identity Pose(), B=@AnimGraphNode_ChooserPlayer_3, Alpha:float=Get WorriedAddPoseAlpha() -> (WorriedAddPoseAlpha:double)) -> (Pose) {blendNode {alphaBoolBlend {blendOption=Linear}}})
  ]
 }
 AnimationBlendStackGraph_0 {
  path=AnimGraph.AnimGraphNode_StateMachine_6.BodyAdditives.AnimStateNode_1.Worried.AnimGraphNode_ChooserPlayer_3.AnimationBlendStackGraph_0
  nodes [Output Pose(Result=Blend Stack Input())]
 }
 Combat {
  path=AnimGraph.AnimGraphNode_StateMachine_6.BodyAdditives.AnimStateNode_2.Combat
  nodes [
   AnimGraphNode_ChooserPlayer_0 = Chooser Player: PA_COR_Loco_Angry_ADD!() -> (Pose:Pose Link Structure) {
    is=title AnimGraphNode_ChooserPlayer
    evaluationFrequency=OnInitialUpdate
    chooser {
     _structType=&LookupProxy
     proxy=/CosmeticCompanions/Chooser/AnimationProxyAssets/Core/Locomotion/PA_COR_Loco_Angry_ADD
     proxyTable {
      _structType=&ProxyTableContextProperty
      binding {contextIndex=0, displayName=Anim Table, isBoundToRoot=false, propertyBindingChain [AnimTable]}
     }
    }
    chooserContextDefinition [
     {
      _structType=&ContextObjectTypeClass
      class=&Dragon_Base_AnimBP_C
      direction=ReadWrite
     }
     {struct=&ChooserPlayerSettings, _structType=&ContextObjectTypeStruct, direction=Read}
    ]
   }
   AnimGraphNode_ChooserPlayer_3 = Chooser Player: PA_COR_Loco_AngryFly_ADD!() -> (Pose:Pose Link Structure) {
    is=title AnimGraphNode_ChooserPlayer
    evaluationFrequency=OnInitialUpdate
    chooser {
     _structType=&LookupProxy
     proxy=/CosmeticCompanions/Chooser/AnimationProxyAssets/Core/Locomotion/PA_COR_Loco_AngryFly_ADD
     proxyTable {
      _structType=&ProxyTableContextProperty
      binding {contextIndex=0, displayName=Anim Table, isBoundToRoot=false, propertyBindingChain [AnimTable]}
     }
    }
    chooserContextDefinition [
     {
      _structType=&ContextObjectTypeClass
      class=&Dragon_Base_AnimBP_C
      direction=ReadWrite
     }
     {struct=&ChooserPlayerSettings, _structType=&ContextObjectTypeStruct, direction=Read}
    ]
   }
   Output Animation Pose(Result=Blend Poses by bool(
     BlendPose_0=@AnimGraphNode_ChooserPlayer_3
     BlendPose_1=@AnimGraphNode_ChooserPlayer_0
     BlendTime_0=0.2
     BlendTime_1=0.2
     bActiveValue=Get bFlyStateActive()
    ))
  ]
 }
 AnimationBlendStackGraph_0 {
  path=AnimGraph.AnimGraphNode_StateMachine_6.BodyAdditives.AnimStateNode_2.Combat.AnimGraphNode_ChooserPlayer_0.AnimationBlendStackGraph_0
  nodes [Output Pose(Result=Blend Stack Input())]
 }
 AnimationBlendStackGraph_0 {
  path=AnimGraph.AnimGraphNode_StateMachine_6.BodyAdditives.AnimStateNode_2.Combat.AnimGraphNode_ChooserPlayer_3.AnimationBlendStackGraph_0
  nodes [Output Pose(Result=Blend Stack Input())]
 }
 Transition {
  path=AnimGraph.AnimGraphNode_StateMachine_6.BodyAdditives.AnimStateTransitionNode_0.Transition
  nodes [Result(bCanEnterTransition=Get bIsMoodWorried())]
 }
 Transition {
  path=AnimGraph.AnimGraphNode_StateMachine_6.BodyAdditives.AnimStateTransitionNode_1.Transition
  nodes [Result(bCanEnterTransition=NOT Boolean(A=Get bIsMoodWorried()))]
 }
 Transition {
  path=AnimGraph.AnimGraphNode_StateMachine_6.BodyAdditives.AnimStateTransitionNode_2.Transition
  nodes [Result(bCanEnterTransition=Get bPlayCombatBodyAdditive() -> (bPlayCombatBodyAdditive:bool))]
 }
 Transition {
  path=AnimGraph.AnimGraphNode_StateMachine_6.BodyAdditives.AnimStateTransitionNode_3.Transition
  nodes [Result(bCanEnterTransition=NOT Boolean(A=Get bPlayCombatBodyAdditive()))]
 }
 MovementSM {
  entry=Ground
  path=AnimGraph.AnimGraphNode_StateMachine_7.MovementSM
  ids=id
  nodes [
   AnimStateNode_0 = FlyStart() {is=title AnimStateNode}
   AnimStateNode_11 = Fly() {is=title AnimStateNode}
   AnimStateNode_2 = Ground() {is=title AnimStateNode}
   AnimStateNode_3 = Land() {is=title AnimStateNode}
   AnimStateTransitionNode_0 = AnimStateTransitionNode() {automaticRuleTriggerTime=0, bAllowInertializationForSelfTransitions=true, bAutomaticRuleBasedOnSequencePlayerInState=true}
   AnimStateTransitionNode_1 = AnimStateTransitionNode() {bAllowInertializationForSelfTransitions=true}
   AnimStateTransitionNode_2 = AnimStateTransitionNode() {bAllowInertializationForSelfTransitions=true, priorityOrder=2}
   AnimStateTransitionNode_4 = AnimStateTransitionNode() {bAllowInertializationForSelfTransitions=true}
   AnimStateTransitionNode_5 = AnimStateTransitionNode() {bAllowInertializationForSelfTransitions=true}
   AnimStateTransitionNode_7 = AnimStateTransitionNode() {automaticRuleTriggerTime=0.5, bAutomaticRuleBasedOnSequencePlayerInState=true, crossfadeDuration=0.3}
   AnimStateTransitionNode_8 = AnimStateTransitionNode() {bAllowInertializationForSelfTransitions=true, crossfadeDuration=0.4}
  ]
  links [
   {
    dst=Land
    rule=AnimGraph.AnimGraphNode_StateMachine_7.MovementSM.AnimStateTransitionNode_4.Transition
    src=Fly
    via=AnimStateTransitionNode_4
   }
   {
    dst=Fly
    rule=AnimGraph.AnimGraphNode_StateMachine_7.MovementSM.AnimStateTransitionNode_0.Transition
    src=FlyStart
    via=AnimStateTransitionNode_0
   }
   {
    dst=Ground
    rule=AnimGraph.AnimGraphNode_StateMachine_7.MovementSM.AnimStateTransitionNode_1.Transition
    src=FlyStart
    via=AnimStateTransitionNode_1
   }
   {
    dst=Fly
    rule=AnimGraph.AnimGraphNode_StateMachine_7.MovementSM.AnimStateTransitionNode_5.Transition
    src=Ground
    via=AnimStateTransitionNode_5
   }
   {
    dst=FlyStart
    rule=AnimGraph.AnimGraphNode_StateMachine_7.MovementSM.AnimStateTransitionNode_2.Transition
    src=Ground
    via=AnimStateTransitionNode_2
   }
   {
    dst=Fly
    rule=AnimGraph.AnimGraphNode_StateMachine_7.MovementSM.AnimStateTransitionNode_8.Transition
    src=Land
    via=AnimStateTransitionNode_8
   }
   {
    dst=Ground
    rule=AnimGraph.AnimGraphNode_StateMachine_7.MovementSM.AnimStateTransitionNode_7.Transition
    src=Land
    via=AnimStateTransitionNode_7
   }
  ]
 }
 FlyStart {
  path=AnimGraph.AnimGraphNode_StateMachine_7.MovementSM.AnimStateNode_0.FlyStart
  nodes [
   AnimGraphNode_ChooserPlayer_3 = Chooser Player: PA_COR_Loco_FlyStart!() -> (Pose:Pose Link Structure) {
    is=title AnimGraphNode_ChooserPlayer
    evaluationFrequency=OnInitialUpdate
    chooser {
     _structType=&LookupProxy
     proxy=/CosmeticCompanions/Chooser/AnimationProxyAssets/Core/Locomotion/PA_COR_Loco_FlyStart
     proxyTable {
      _structType=&ProxyTableContextProperty
      binding {contextIndex=0, displayName=Anim Table, isBoundToRoot=false, propertyBindingChain [AnimTable]}
     }
    }
    chooserContextDefinition [
     {
      _structType=&ContextObjectTypeClass
      class=&Dragon_Base_AnimBP_C
      direction=ReadWrite
     }
     {struct=&ChooserPlayerSettings, _structType=&ContextObjectTypeStruct, direction=Read}
    ]
   }
   Output Animation Pose(Result=@AnimGraphNode_ChooserPlayer_3)
  ]
 }
 AnimationBlendStackGraph_0 {
  path=AnimGraph.AnimGraphNode_StateMachine_7.MovementSM.AnimStateNode_0.FlyStart.AnimGraphNode_ChooserPlayer_3.AnimationBlendStackGraph_0
  nodes [Output Pose(Result=Blend Stack Input())]
 }
 Fly {
  path=AnimGraph.AnimGraphNode_StateMachine_7.MovementSM.AnimStateNode_11.Fly
  nodes [
   AnimGraphNode_ChooserPlayer_1 = Chooser Player: PRA_Companion_WingFlaps!() -> (Pose:Pose Link Structure) {
    is=title AnimGraphNode_ChooserPlayer
    evaluationFrequency=OnLoop
    chooser {
     _structType=&LookupProxy
     proxy=/CosmeticCompanions/Chooser/PRA_Companion_WingFlaps
     proxyTable {
      _structType=&ProxyTableContextProperty
      binding {contextIndex=0, displayName=Anim Table, isBoundToRoot=false, propertyBindingChain [AnimTable]}
     }
    }
    chooserContextDefinition [
     {
      _structType=&ContextObjectTypeClass
      class=&Dragon_Base_AnimBP_C
      direction=ReadWrite
     }
     {struct=&ChooserPlayerSettings, _structType=&ContextObjectTypeStruct, direction=Read}
    ]
   }
   AnimGraphNode_StateMachine_1 = WingFlapSM() {is=title AnimGraphNode_StateMachine}
   n1 = "Blend Poses (ELocomotionMovementType)"!(
    BlendPose_0:Pose Link Structure=@AnimGraphNode_StateMachine_1
    BlendPose_1:Pose Link Structure=@AnimGraphNode_ChooserPlayer_1
    BlendTime_0:float=0.2
    BlendTime_1:float=0.2
    ActiveEnumValue:Byte=Get CurrentLocomotionMovementType() -> (CurrentLocomotionMovementType:ELocomotionMovementType Enum=Walk)
   ) -> (Pose:Pose Link Structure) {is=title AnimGraphNode_BlendListByEnum, blendPose [{linkId=-1, sourceLinkId=-1}, {linkId=-1, sourceLinkId=-1}], blendTime [0.1,0.1]}
   AnimGraphNode_ChooserPlayer_0 = Chooser Player: PA_COR_Loco_Glide!() -> (Pose:Pose Link Structure) {
    is=title AnimGraphNode_ChooserPlayer
    evaluationFrequency=OnInitialUpdate
    chooser {
     _structType=&LookupProxy
     proxy=/CosmeticCompanions/Chooser/AnimationProxyAssets/Core/Locomotion/PA_COR_Loco_Glide
     proxyTable {
      _structType=&ProxyTableContextProperty
      binding {contextIndex=0, displayName=Anim Table, isBoundToRoot=false, propertyBindingChain [AnimTable]}
     }
    }
    chooserContextDefinition [
     {
      _structType=&ContextObjectTypeClass
      class=&Dragon_Base_AnimBP_C
      direction=ReadWrite
     }
     {struct=&ChooserPlayerSettings, _structType=&ContextObjectTypeStruct, direction=Read}
    ]
    defaultSettings {bForceLooping=true}
   }
   Output Animation Pose(Result=Modify Curve(
     SourcePose=Apply Additive!(Base=@AnimGraphNode_ChooserPlayer_0, Additive=@n1, Alpha:float) -> (Pose) {alphaScaleBiasClamp {interpSpeedIncreasing=3}}
     CurveValues_0=1
     Alpha=1
    ) {curveNames [Flying], curveValues [0]})
  ]
 }
 AnimationBlendStackGraph_0 {
  path=AnimGraph.AnimGraphNode_StateMachine_7.MovementSM.AnimStateNode_11.Fly.AnimGraphNode_ChooserPlayer_0.AnimationBlendStackGraph_0
  nodes [Output Pose(Result=Blend Stack Input())]
 }
 AnimationBlendStackGraph_0 {
  path=AnimGraph.AnimGraphNode_StateMachine_7.MovementSM.AnimStateNode_11.Fly.AnimGraphNode_ChooserPlayer_1.AnimationBlendStackGraph_0
  nodes [Output Pose(Result=Blend Stack Input())]
 }
 WingFlapSM {
  entry=Nil
  path=AnimGraph.AnimGraphNode_StateMachine_7.MovementSM.AnimStateNode_11.Fly.AnimGraphNode_StateMachine_1.WingFlapSM
  ids=id
  nodes [
   AnimStateNode_0 = Nil() {is=title AnimStateNode}
   AnimStateNode_1 = Flap() {is=title AnimStateNode}
   AnimStateTransitionNode_5 = AnimStateTransitionNode() {crossfadeDuration=0.1}
   AnimStateTransitionNode_6 = AnimStateTransitionNode() {
    automaticRuleTriggerTime=0
    bAllowInertializationForSelfTransitions=true
    bAutomaticRuleBasedOnSequencePlayerInState=true
    crossfadeDuration=0.1
   }
  ]
  links [
   {
    dst=Nil
    rule=AnimGraph.AnimGraphNode_StateMachine_7.MovementSM.AnimStateNode_11.Fly.AnimGraphNode_StateMachine_1.WingFlapSM.AnimStateTransitionNode_6.Transition
    src=Flap
    via=AnimStateTransitionNode_6
   }
   {
    dst=Flap
    rule=AnimGraph.AnimGraphNode_StateMachine_7.MovementSM.AnimStateNode_11.Fly.AnimGraphNode_StateMachine_1.WingFlapSM.AnimStateTransitionNode_5.Transition
    src=Nil
    via=AnimStateTransitionNode_5
   }
  ]
 }
 Nil {
  path=AnimGraph.AnimGraphNode_StateMachine_7.MovementSM.AnimStateNode_11.Fly.AnimGraphNode_StateMachine_1.WingFlapSM.AnimStateNode_0.Nil
  nodes [Output Animation Pose()]
 }
 Flap {
  path=AnimGraph.AnimGraphNode_StateMachine_7.MovementSM.AnimStateNode_11.Fly.AnimGraphNode_StateMachine_1.WingFlapSM.AnimStateNode_1.Flap
  nodes [
   AnimGraphNode_ChooserPlayer_1 = Chooser Player: PRA_Companion_WingFlaps() {
    chooser {
     _structType=&LookupProxy
     proxy=&PRA_Companion_WingFlaps
     proxyTable {
      _structType=&ProxyTableContextProperty
      binding {contextIndex=0, displayName=Anim Table, isBoundToRoot=false, propertyBindingChain [AnimTable]}
     }
    }
    chooserContextDefinition [
     {
      _structType=&ContextObjectTypeClass
      class=&Dragon_Base_AnimBP_C
      direction=ReadWrite
     }
     {struct=&ChooserPlayerSettings, _structType=&ContextObjectTypeStruct, direction=Read}
    ]
   }
   Output Animation Pose(Result=@AnimGraphNode_ChooserPlayer_1)
  ]
 }
 AnimationBlendStackGraph_0 {
  path=AnimGraph.AnimGraphNode_StateMachine_7.MovementSM.AnimStateNode_11.Fly.AnimGraphNode_StateMachine_1.WingFlapSM.AnimStateNode_1.Flap.AnimGraphNode_ChooserPlayer_1.AnimationBlendStackGraph_0
  nodes [Output Pose(Result=Blend Stack Input())]
 }
 Transition {
  path=AnimGraph.AnimGraphNode_StateMachine_7.MovementSM.AnimStateNode_11.Fly.AnimGraphNode_StateMachine_1.WingFlapSM.AnimStateTransitionNode_5.Transition
  nodes [Result(bCanEnterTransition=Break Companion Random Number Timer!(CompanionRandomNumberTimer:"Companion Random Number Timer Structure (by ref)"=Get AuxiliaryTimer4() -> (AuxiliaryTimer4:Companion Random Number Timer Structure)) -> (ValueInt:int, ValueFloat:float=~, bJustChanged:bool=~) {is=title K2Node_BreakStruct}.bJustChanged)]
 }
 Transition {
  path=AnimGraph.AnimGraphNode_StateMachine_7.MovementSM.AnimStateNode_11.Fly.AnimGraphNode_StateMachine_1.WingFlapSM.AnimStateTransitionNode_6.Transition
  nodes [Result()]
 }
 Ground {
  path=AnimGraph.AnimGraphNode_StateMachine_7.MovementSM.AnimStateNode_2.Ground
  nodes [
   AnimGraphNode_ChooserPlayer_0 = Chooser Player: PA_COR_Loco_Trot!() -> (Pose:Pose Link Structure) {
    is=title AnimGraphNode_ChooserPlayer
    evaluationFrequency=OnInitialUpdate
    groupName=Locomotion
    method=SyncGroup
    chooser {
     _structType=&LookupProxy
     proxy=/CosmeticCompanions/Chooser/AnimationProxyAssets/Core/Locomotion/PA_COR_Loco_Trot
     proxyTable {
      _structType=&ProxyTableContextProperty
      binding {contextIndex=0, displayName=Anim Table, isBoundToRoot=false, propertyBindingChain [AnimTable]}
     }
    }
    chooserContextDefinition [
     {
      _structType=&ContextObjectTypeClass
      class=&Dragon_Base_AnimBP_C
      direction=ReadWrite
     }
     {struct=&ChooserPlayerSettings, _structType=&ContextObjectTypeStruct, direction=Read}
    ]
    defaultSettings {bForceLooping=true}
   }
   AnimGraphNode_ChooserPlayer_3 = Chooser Player: PA_COR_Loco_Walk!() -> (Pose:Pose Link Structure) {
    is=title AnimGraphNode_ChooserPlayer
    evaluationFrequency=OnInitialUpdate
    groupName=Locomotion
    method=SyncGroup
    chooser {
     _structType=&LookupProxy
     proxy=/CosmeticCompanions/Chooser/AnimationProxyAssets/Core/Locomotion/PA_COR_Loco_Walk
     proxyTable {
      _structType=&ProxyTableContextProperty
      binding {contextIndex=0, displayName=Anim Table, isBoundToRoot=false, propertyBindingChain [AnimTable]}
     }
    }
    chooserContextDefinition [
     {
      _structType=&ContextObjectTypeClass
      class=&Dragon_Base_AnimBP_C
      direction=ReadWrite
     }
     {struct=&ChooserPlayerSettings, _structType=&ContextObjectTypeStruct, direction=Read}
    ]
    defaultSettings {bForceLooping=true}
   }
   n1 = Get CurrentLocomotionMovementType() -> (CurrentLocomotionMovementType=Walk)
   Output Animation Pose(Result=Modify Curve!(
     SourcePose="Blend Poses (ELocomotionMovementType)"(
      BlendPose_0=@AnimGraphNode_ChooserPlayer_0
      BlendPose_1=@AnimGraphNode_ChooserPlayer_3
      BlendTime_0=0.2
      BlendTime_1=0.2
      ActiveEnumValue=@n1
     ) {blendPose [{linkId=-1, sourceLinkId=-1}, {linkId=-1, sourceLinkId=-1}], blendTime [0.1,0.1]}
     CurveValues_0=1
     CurveValues_1:float=1
     Alpha=1
    ) -> (Pose) {curveNames [Locomotion,Walk], curveValues [0,0]})
  ]
 }
 AnimationBlendStackGraph_0 {
  path=AnimGraph.AnimGraphNode_StateMachine_7.MovementSM.AnimStateNode_2.Ground.AnimGraphNode_ChooserPlayer_0.AnimationBlendStackGraph_0
  nodes [Output Pose(Result=Blend Stack Input!(PlayRate:float=Get TrotPlayRate() -> (TrotPlayRate:double)) -> (Pose) {bOverridePlayRate=true})]
 }
 AnimationBlendStackGraph_0 {
  path=AnimGraph.AnimGraphNode_StateMachine_7.MovementSM.AnimStateNode_2.Ground.AnimGraphNode_ChooserPlayer_3.AnimationBlendStackGraph_0
  nodes [Output Pose(Result=Blend Stack Input(PlayRate=Get WalkPlayRate() -> (WalkPlayRate:double)) {bOverridePlayRate=true})]
 }
 Land {
  path=AnimGraph.AnimGraphNode_StateMachine_7.MovementSM.AnimStateNode_3.Land
  nodes [
   AnimGraphNode_ChooserPlayer_1 = Chooser Player: PA_COR_Loco_StopLanding!() -> (Pose:Pose Link Structure) {
    is=title AnimGraphNode_ChooserPlayer
    evaluationFrequency=OnInitialUpdate
    groupName=FlyStop
    method=SyncGroup
    chooser {
     _structType=&LookupProxy
     proxy=/CosmeticCompanions/Chooser/AnimationProxyAssets/Core/Locomotion/PA_COR_Loco_StopLanding
     proxyTable {
      _structType=&ProxyTableContextProperty
      binding {contextIndex=0, displayName=Anim Table, isBoundToRoot=false, propertyBindingChain [AnimTable]}
     }
    }
    chooserContextDefinition [
     {
      _structType=&ContextObjectTypeClass
      class=&Dragon_Base_AnimBP_C
      direction=ReadWrite
     }
     {struct=&ChooserPlayerSettings, _structType=&ContextObjectTypeStruct, direction=Read}
    ]
   }
   Output Animation Pose(Result=@AnimGraphNode_ChooserPlayer_1)
  ]
 }
 AnimationBlendStackGraph_0 {
  path=AnimGraph.AnimGraphNode_StateMachine_7.MovementSM.AnimStateNode_3.Land.AnimGraphNode_ChooserPlayer_1.AnimationBlendStackGraph_0
  nodes [Output Pose(Result=Blend Stack Input!() -> (Pose))]
 }
 Transition {path=AnimGraph.AnimGraphNode_StateMachine_7.MovementSM.AnimStateTransitionNode_0.Transition, nodes [Result()]}
 Transition {
  path=AnimGraph.AnimGraphNode_StateMachine_7.MovementSM.AnimStateTransitionNode_1.Transition
  nodes [Result(bCanEnterTransition=NOT Boolean(A=Get bFlyStateActive()))]
 }
 Transition {
  path=AnimGraph.AnimGraphNode_StateMachine_7.MovementSM.AnimStateTransitionNode_2.Transition
  nodes [Result(bCanEnterTransition=Get bFlyStateActive())]
 }
 Transition {
  path=AnimGraph.AnimGraphNode_StateMachine_7.MovementSM.AnimStateTransitionNode_4.Transition
  nodes [Result(bCanEnterTransition=NOT Boolean(A=Get bFlyStateActive()))]
 }
 Transition {
  path=AnimGraph.AnimGraphNode_StateMachine_7.MovementSM.AnimStateTransitionNode_5.Transition
  nodes [Result(bCanEnterTransition=Get bIsFlying())]
 }
 Transition {path=AnimGraph.AnimGraphNode_StateMachine_7.MovementSM.AnimStateTransitionNode_7.Transition, nodes [Result()]}
 Transition {
  path=AnimGraph.AnimGraphNode_StateMachine_7.MovementSM.AnimStateTransitionNode_8.Transition
  nodes [Result(bCanEnterTransition=Get bFlyStateActive())]
 }
 IdleAntics {
  entry=Conduit
  path=AnimGraph.AnimGraphNode_StateMachine_8.IdleAntics
  ids=id
  nodes [
   AnimStateConduitNode_0 = Conduit()
   AnimStateNode_0 = Default()
   AnimStateNode_7 = FrontEndSpawnAnticChooser() {is=title AnimStateNode}
   AnimStateNode_9 = AnticChooser() {is=title AnimStateNode}
   AnimStateTransitionNode_0 = AnimStateTransitionNode() {crossfadeDuration=0.25, priorityOrder=100}
   AnimStateTransitionNode_1 = AnimStateTransitionNode() {bAutomaticRuleBasedOnSequencePlayerInState=true}
   AnimStateTransitionNode_14 = AnimStateTransitionNode() {logicType=TLT_Inertialization}
   AnimStateTransitionNode_18 = AnimStateTransitionNode() {logicType=TLT_Inertialization, priorityOrder=2}
   AnimStateTransitionNode_2 = AnimStateTransitionNode()
   AnimStateTransitionNode_4 = AnimStateTransitionNode()
   AnimStateTransitionNode_5 = AnimStateTransitionNode() {bAutomaticRuleBasedOnSequencePlayerInState=true}
  ]
  links [
   {
    dst=Conduit
    rule=AnimGraph.AnimGraphNode_StateMachine_8.IdleAntics.AnimStateTransitionNode_2.Transition
    src=AnticChooser
    via=AnimStateTransitionNode_2
   }
   {
    dst=Conduit
    rule=AnimGraph.AnimGraphNode_StateMachine_8.IdleAntics.AnimStateTransitionNode_5.Transition
    src=AnticChooser
    via=AnimStateTransitionNode_5
   }
   {
    dst=AnticChooser
    rule=AnimGraph.AnimGraphNode_StateMachine_8.IdleAntics.AnimStateTransitionNode_18.Transition
    src=Conduit
    via=AnimStateTransitionNode_18
   }
   {
    dst=Default
    rule=AnimGraph.AnimGraphNode_StateMachine_8.IdleAntics.AnimStateTransitionNode_0.Transition
    src=Conduit
    via=AnimStateTransitionNode_0
   }
   {
    dst=FrontEndSpawnAnticChooser
    rule=AnimGraph.AnimGraphNode_StateMachine_8.IdleAntics.AnimStateTransitionNode_14.Transition
    src=Conduit
    via=AnimStateTransitionNode_14
   }
   {
    dst=Conduit
    rule=AnimGraph.AnimGraphNode_StateMachine_8.IdleAntics.AnimStateTransitionNode_4.Transition
    src=Default
    via=AnimStateTransitionNode_4
   }
   {
    dst=Conduit
    rule=AnimGraph.AnimGraphNode_StateMachine_8.IdleAntics.AnimStateTransitionNode_1.Transition
    src=FrontEndSpawnAnticChooser
    via=AnimStateTransitionNode_1
   }
  ]
 }
 Conduit {path=AnimGraph.AnimGraphNode_StateMachine_8.IdleAntics.AnimStateConduitNode_0.Conduit, nodes [Result(bCanEnterTransition=true)]}
 Default {
  path=AnimGraph.AnimGraphNode_StateMachine_8.IdleAntics.AnimStateNode_0.Default
  nodes [
   AnimGraphNode_ChooserPlayer_0 = Chooser Player: PRA_Companion_AnticsNeutral!() -> (Pose:Pose Link Structure) {
    is=title AnimGraphNode_ChooserPlayer
    evaluationFrequency=OnInitialUpdate
    chooser {
     _structType=&LookupProxy
     proxy=/CosmeticCompanions/Chooser/PRA_Companion_AnticsNeutral
     proxyTable {
      _structType=&ProxyTableContextProperty
      binding {contextIndex=0, displayName=Anim Table, isBoundToRoot=false, propertyBindingChain [AnimTable]}
     }
    }
    chooserContextDefinition [
     {
      _structType=&ContextObjectTypeClass
      class=&Dragon_Base_AnimBP_C
      direction=ReadWrite
     }
     {struct=&ChooserPlayerSettings, _structType=&ContextObjectTypeStruct, direction=Read}
    ]
    defaultSettings {bForceLooping=true, blendOption=Linear}
   }
   Output Animation Pose(Result=@AnimGraphNode_ChooserPlayer_0)
  ]
 }
 AnimationBlendStackGraph_0 {
  path=AnimGraph.AnimGraphNode_StateMachine_8.IdleAntics.AnimStateNode_0.Default.AnimGraphNode_ChooserPlayer_0.AnimationBlendStackGraph_0
  nodes [Output Pose(Result=Blend Stack Input())]
 }
 FrontEndSpawnAnticChooser {
  path=AnimGraph.AnimGraphNode_StateMachine_8.IdleAntics.AnimStateNode_7.FrontEndSpawnAnticChooser
  nodes [
   n1 = Get OnSpawnedChooser() -> (OnSpawnedChooser:Async Sequence Chooser Structure)
   Output Animation Pose(Result=Sequence Player!(Sequence:Anim Sequence Base Object Reference=Break Async Sequence Chooser!(AsyncSequenceChooser:"Async Sequence Chooser Structure (by ref)"=@n1) -> (
      bIsLoading:bool=~
      bIsMontage:bool=~
      bIsReady:bool=~
      AnimSequence:Anim Sequence Object Reference
      AnimMontage:Anim Montage Object Reference
     ) {is=title K2Node_BreakStruct}.AnimSequence) -> (Pose) {bLoopAnimation=false}) {
    node {stateExitFunction {functionName=OnStateExit_0}}
    stateExitFunction {bSelfContext=true, memberGuid=#24, memberName=AllowedToSit_Exit}
   }
  ]
 }
 AnticChooser {
  path=AnimGraph.AnimGraphNode_StateMachine_8.IdleAntics.AnimStateNode_9.AnticChooser
  nodes [Output Animation Pose(Result=Sequence Player() {bLoopAnimation=false}) {
    node {stateEntryFunction {functionName=OnStateEntry}, stateExitFunction {functionName=OnStateExit_0}}
    stateEntryFunction {bSelfContext=true, memberGuid=#25, memberName=AnticPlaying_Enter}
    stateExitFunction {bSelfContext=true, memberGuid=#26, memberName=AnticPlaying_Exit}
   }]
 }
 Transition {path=AnimGraph.AnimGraphNode_StateMachine_8.IdleAntics.AnimStateTransitionNode_0.Transition, nodes [Result(bCanEnterTransition=true)]}
 Transition {path=AnimGraph.AnimGraphNode_StateMachine_8.IdleAntics.AnimStateTransitionNode_1.Transition, nodes [Result()]}
 Transition {
  path=AnimGraph.AnimGraphNode_StateMachine_8.IdleAntics.AnimStateTransitionNode_14.Transition
  nodes [Result(bCanEnterTransition=Get bTriggerSpawnedAntic() -> (bTriggerSpawnedAntic:bool))]
 }
 Transition {
  path=AnimGraph.AnimGraphNode_StateMachine_8.IdleAntics.AnimStateTransitionNode_18.Transition
  nodes [Result(bCanEnterTransition=Get bIsAllowAnticToPlay() -> (bIsAllowAnticToPlay:bool))]
 }
 Transition {
  path=AnimGraph.AnimGraphNode_StateMachine_8.IdleAntics.AnimStateTransitionNode_2.Transition
  nodes [Result(bCanEnterTransition=NOT Boolean(A=Get bIsCanPlayAntic() -> (bIsCanPlayAntic:bool)))]
 }
 Transition {path=AnimGraph.AnimGraphNode_StateMachine_8.IdleAntics.AnimStateTransitionNode_4.Transition, nodes [Result(bCanEnterTransition=true)]}
 Transition {path=AnimGraph.AnimGraphNode_StateMachine_8.IdleAntics.AnimStateTransitionNode_5.Transition, nodes [Result()]}
 AnticPlaying_Enter {nodes [
   AnticPlaying_Enter!() -> (then, Context:"Anim Update Context Structure (by ref)", Node:"Anim Node Reference Structure (by ref)") {is=title K2Node_FunctionEntry}
   > Set bIsPlayingAntic(bIsPlayingAntic:bool=true) -> (Output_Get:bool)
   > Return Node!(execute)
  ]}
 AnticPlaying_Exit {nodes [
   AnticPlaying_Exit!() -> (then, Context:"Anim Update Context Structure (by ref)", Node:"Anim Node Reference Structure (by ref)") {is=title K2Node_FunctionEntry}
   > Set bIsPlayingAntic(bIsPlayingAntic=false)
   > Return Node()
  ]}
 DefaultTailWag {nodes [
   AnimGraphNode_ChooserPlayer_0 = Chooser Player: PA_COR_Tail_BS!(BlendSpaceX:float=Get TailWagSpeed() -> (TailWagSpeed:float)) -> (Pose:Pose Link Structure) {
    is=title AnimGraphNode_ChooserPlayer
    evaluationFrequency=OnInitialUpdate
    chooser {
     _structType=&LookupProxy
     proxy=/CosmeticCompanions/Chooser/AnimationProxyAssets/Core/Tail/PA_COR_Tail_BS
     proxyTable {
      _structType=&ProxyTableContextProperty
      binding {contextIndex=0, displayName=Anim Table, isBoundToRoot=false, propertyBindingChain [AnimTable]}
     }
    }
    chooserContextDefinition [
     {
      _structType=&ContextObjectTypeClass
      class=&Dragon_Base_AnimBP_C
      direction=ReadWrite
     }
     {struct=&ChooserPlayerSettings, _structType=&ContextObjectTypeStruct, direction=Read}
    ]
    defaultSettings {bForceLooping=true}
   }
   AnimGraphNode_ChooserPlayer_1 = Chooser Player: PA_COR_Tail_Sit_ADD!() -> (Pose:Pose Link Structure) {
    is=title AnimGraphNode_ChooserPlayer
    evaluationFrequency=OnInitialUpdate
    chooser {
     _structType=&LookupProxy
     proxy=/CosmeticCompanions/Chooser/AnimationProxyAssets/Core/Tail/PA_COR_Tail_Sit_ADD
     proxyTable {
      _structType=&ProxyTableContextProperty
      binding {contextIndex=0, displayName=Anim Table, isBoundToRoot=false, propertyBindingChain [AnimTable]}
     }
    }
    chooserContextDefinition [
     {
      _structType=&ContextObjectTypeClass
      class=&Dragon_Base_AnimBP_C
      direction=ReadWrite
     }
     {struct=&ChooserPlayerSettings, _structType=&ContextObjectTypeStruct, direction=Read}
    ]
    defaultSettings {bForceLooping=true}
   }
   Output Pose(Result=Blend Poses by bool(
     BlendPose_0=@AnimGraphNode_ChooserPlayer_1
     BlendPose_1=@AnimGraphNode_ChooserPlayer_0
     BlendTime_0=0.3
     BlendTime_1=0.3
     bActiveValue=Get SitTail() -> (SitTail:bool)
    ) {blendType=Linear, transitionType=Inertialization}) {is=title AnimGraphNode_Root}
  ]}
 AnimationBlendStackGraph_0 {
  path=DefaultTailWag.AnimGraphNode_ChooserPlayer_0.AnimationBlendStackGraph_0
  nodes [Output Pose(Result=Blend Stack Input()) {is=title AnimGraphNode_BlendStackResult}]
 }
 AnimationBlendStackGraph_0 {path=DefaultTailWag.AnimGraphNode_ChooserPlayer_1.AnimationBlendStackGraph_0, nodes [Output Pose(Result=Blend Stack Input())]}
 EventGraph {links [], nodes []}
 GetSequencerAnimSlotNames {nodes [
   GetSequencerAnimSlotNames() {is=title K2Node_FunctionEntry}
   > Return Node!(execute, ReturnValue:Array of Names=Make Array("[0]":Name=SequencerDefault) -> (Array:Array of Names))
  ]}
 LocoStart_Entry {nodes [
   LocoStart_Entry!() -> (then, Context:"Anim Update Context Structure (by ref)", Node:"Anim Node Reference Structure (by ref)") {is=title K2Node_FunctionEntry}
   > Set bIsLocoStart(bIsLocoStart:bool=true) -> (Output_Get:bool)
   > Return Node!(execute)
  ]}
 OnInitialUpdate {nodes [
   n1 = OnInitialUpdate!() -> (then, Context:"Anim Update Context Structure (by ref)", Node:"Anim Node Reference Structure (by ref)") {is=title K2Node_FunctionEntry}
   > n2 = Evaluate Proxy: PA_COR_AO!(execute:Exec, ProxyTable:Proxy Table Object Reference=Get AnimTable() -> (AnimTable:Proxy Table Object Reference)) -> (execute:Exec, Result:Animation Asset Object Reference) {is=title K2Node_EvaluateProxy2, proxy=/CosmeticCompanions/Chooser/AnimationProxyAssets/Core/PA_COR_AO, structOutputMode=Default}
   > n3 = Cast To AimOffsetBlendSpace!(execute, Object=@n2) -> (then, CastFailed, AsAim Offset Blend Space:Aim Offset Blend Space Object Reference) {is=title K2Node_DynamicCast}
   > SetBlendSpace!(
    execute
    BlendSpacePlayer:"Blend Space Player Reference Structure (by ref)"=Convert to Blend Space Player!(Node:"Anim Node Reference Structure (by ref)"=@n1.Node) -> (BlendSpacePlayer:Blend Space Player Reference Structure, Result:bool)
    BlendSpace:Blend Space Object Reference=@n3
   ) -> (then, ReturnValue:Blend Space Player Reference Structure)
   > Return Node()
  ]}
 OnUpdate {nodes [
   n1 = Convert to Sequence Player!(Node:"Anim Node Reference Structure (by ref)"=OnUpdate!() -> (then, Context:"Anim Update Context Structure (by ref)", Node:"Anim Node Reference Structure (by ref)") {is=title K2Node_FunctionEntry}.Node) -> (SequencePlayer:Sequence Player Reference Structure, Result:bool)
   n2 = Branch(Condition=Contains Item(
     TargetArray:Array of Anim Sequence Base Object References
     ItemToFind:"Anim Sequence Base Object Reference (by ref)"=Get Sequence!(SequencePlayer:"Sequence Player Reference Structure (by ref)"=@n1) -> (ReturnValue:Anim Sequence Base Object Reference) {is=title K2Node_CallFunction}
    ))
   > Return Node()
   SetSequence!(execute=@n2.else, SequencePlayer:"Sequence Player Reference Structure (by ref)"=@n1, Sequence:Anim Sequence Base Object Reference) -> (then, ReturnValue:Sequence Player Reference Structure)
  ]}
 SpecialIdle_StateEntry {nodes [
   SpecialIdle_StateEntry!() -> (then, Context:"Anim Update Context Structure (by ref)", Node:"Anim Node Reference Structure (by ref)") {is=title K2Node_FunctionEntry}
   > Set bIsSpecialIdlePlaying(bIsSpecialIdlePlaying:bool=true) -> (Output_Get:bool)
   > Return Node()
  ]}
 SpecialIdle_StateExit {nodes [
   SpecialIdle_StateExit!() -> (then, Context:"Anim Update Context Structure (by ref)", Node:"Anim Node Reference Structure (by ref)") {is=title K2Node_FunctionEntry}
   > Set bIsSpecialIdlePlaying(bIsSpecialIdlePlaying=false)
   > Return Node()
  ]}
 StartTurn_StateExit {nodes [
   StartTurn_StateExit!() -> (then, Context:"Anim Update Context Structure (by ref)", Node:"Anim Node Reference Structure (by ref)") {is=title K2Node_FunctionEntry}
   > Set UpdatedGoalRotation(UpdatedGoalRotation:float=0) -> (Output_Get:float)
   > Set bIsLocoStart(bIsLocoStart=false)
   > Return Node()
  ]}
]
variables [
 RandomizationContext {type=ChooserRandomizationContext, default {}}
 SitTail {type=boolean}
]
