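# Base Animation Blueprint for quadruped cosmetic companions, pulling animations from a swappable Quadruped_Base_Struct and proxy-table choosers.
# High- and low-fidelity state machine paths, layered with additives for body, antics and tail, blend by inertialization, then gain body turn and breathing.
# Idle handles sitting and angry combat stances, while linked layers supply locomotion and default tail wag.
# Look-at aim offset, chooser tail additive and Dead Blending precede quadruped slope warping and four-leg IK.
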AnimBlueprint /CosmeticCompanions/Assets/Quadruped/QuadrupedBaseBP/Quadruped_Base_AnimBP
target_skeleton=/CosmeticCompanions/Assets/Quadruped/SitPlant/Meshes/SitPlant_Quadruped_Skeleton
graphs [
 AllowedToSit_Exit {nodes [
   AllowedToSit_Exit!() -> (then, Context:"Anim Update Context Structure (by ref)", Node:"Anim Node Reference Structure (by ref)") {is=title K2Node_FunctionEntry}
   > Set bIsAllowedToSit(bIsAllowedToSit:bool=true) -> (Output_Get:bool)
   > Return Node()
  ]}
 AllowsCinematicOverride {nodes [
   AllowsCinematicOverride() {is=title K2Node_FunctionEntry}
   > Return Node!(execute, ReturnValue:bool=false)
  ]}
 AnimGraph {nodes [
   n1 = Quadruped Slope Warping!(
    ComponentPose:Component Space Pose Link Structure=Local To Component(LocalPose=Use cached pose 'PreSlopeWarpingPose'())
    Alpha:float=Get SlopeWarpingAlpha() -> (SlopeWarpingAlpha:float)
    MaxStepHeight:float=50
    bOnlyAffectSkeletonRootBone:bool=true
    MaxSkeletonRootBoneRotationAngle:float=0
   ) -> (Pose:Component Space Pose Link Structure) {
    is=title FortAnimGraphNode_QuadrupedSlopeWarping
    bUseMovementComponentFloorInfoInsteadOfTraces=true
    alphaBoolBlend {blendOption=Linear}
    alphaScaleBiasClamp {bias=1, scale=-1}
    armsFeetDefinitions [
     {numBonesInLimb=3, fKFootBone {boneName=ball_fr_l}, iKFootBone {boneName=IK_Foot_Front_L}}
     {numBonesInLimb=3, fKFootBone {boneName=ball_fr_r}, iKFootBone {boneName=IK_Foot_Front_R}}
    ]
    armsIKFootRootBone {boneName=IK_Foot_Front_Root}
    armsRootBone {boneName=spine_03}
    feetDefinitions [
     {numBonesInLimb=3, fKFootBone {boneName=ball_bk_l}, iKFootBone {boneName=IK_Foot_Rear_L}}
     {numBonesInLimb=3, fKFootBone {boneName=ball_bk_r}, iKFootBone {boneName=IK_Foot_Rear_R}}
    ]
    iKFootRootBone {boneName=IK_Foot_Rear_Root}
    pelvisBone {boneName=hips}
    skeletonRootBone {boneName=root}
   }
   n2 = Component To Local(ComponentPose=@n1)
   n3 = Get SlopeWarpingAlpha()
   n4 = Root Slope Tilt!(
    ComponentPose:Component Space Pose Link Structure=Local To Component(LocalPose=Use cached pose 'PreSlopeWarpingPose'())
    Alpha:float=@n3
    RootSlopeCompensationAlpha:float=0.25
   ) -> (Pose:Component Space Pose Link Structure) {
    is=title FortAnimGraphNode_RootSlopeTilt
    alphaScaleBiasClamp {bias=1, interpSpeedDecreasing=8, interpSpeedIncreasing=8, scale=-1}
    skeletonRootBone {boneName=root}
   }
   n5 = Quadruped Slope Warping(ComponentPose=@n4, Alpha=@n3, MaxStepHeight=50, bOnlyAffectSkeletonRootBone=False, MaxSkeletonRootBoneRotationAngle=0) {
    bUseMovementComponentFloorInfoInsteadOfTraces=true
    alphaBoolBlend {blendOption=Linear}
    alphaScaleBiasClamp {bias=1, scale=-1}
    armsFeetDefinitions [
     {numBonesInLimb=3, fKFootBone {boneName=ball_fr_l}, iKFootBone {boneName=IK_Foot_Front_L}}
     {numBonesInLimb=3, fKFootBone {boneName=ball_fr_r}, iKFootBone {boneName=IK_Foot_Front_R}}
    ]
    armsIKFootRootBone {boneName=IK_Foot_Front_Root}
    armsRootBone {boneName=spine_03}
    feetDefinitions [
     {numBonesInLimb=3, fKFootBone {boneName=ball_bk_l}, iKFootBone {boneName=IK_Foot_Rear_L}}
     {numBonesInLimb=3, fKFootBone {boneName=ball_bk_r}, iKFootBone {boneName=IK_Foot_Rear_R}}
    ]
    iKFootRootBone {boneName=IK_Foot_Rear_Root}
    pelvisBone {boneName=hips}
    skeletonRootBone {boneName=root}
   }
   n6 = Component To Local(ComponentPose=@n5)
   n7 = Get bUseRootSlopeOrientation() -> (bUseRootSlopeOrientation:bool)
   n8 = Leg IK!(
    ComponentPose:Component Space Pose Link Structure=Local To Component(LocalPose=Blend Poses by bool(BlendPose_0=@n2, BlendPose_1=@n6, BlendTime_0=0.3, BlendTime_1=0.3, bActiveValue=@n7) {blendType=Linear})
    Alpha:float=Get IKAlpha() -> (IKAlpha:float)
   ) -> (Pose:Component Space Pose Link Structure) {
    is=title AnimGraphNode_LegIK
    alphaBoolBlend {blendOption=Linear}
    alphaScaleBiasClamp {bias=1, scale=-1}
    legsDefinition [
     {
      bEnableKneeTwistCorrection=true
      bEnableRotationLimit=false
      footBoneForwardAxis=Y
      hingeRotationAxis=None
      minRotationAngle=15
      numBonesInLimb=3
      twistOffsetCurveName=None
      fKFootBone {boneName=ball_bk_l}
      iKFootBone {boneName=IK_Foot_Rear_L}
     }
     {
      bEnableKneeTwistCorrection=true
      bEnableRotationLimit=false
      footBoneForwardAxis=Y
      hingeRotationAxis=None
      minRotationAngle=15
      numBonesInLimb=3
      twistOffsetCurveName=None
      fKFootBone {boneName=ball_bk_r}
      iKFootBone {boneName=IK_Foot_Rear_R}
     }
     {
      bEnableKneeTwistCorrection=true
      bEnableRotationLimit=false
      footBoneForwardAxis=Y
      hingeRotationAxis=None
      minRotationAngle=15
      numBonesInLimb=3
      twistOffsetCurveName=None
      fKFootBone {boneName=ball_fr_r}
      iKFootBone {boneName=IK_Foot_Front_R}
     }
     {
      bEnableKneeTwistCorrection=true
      bEnableRotationLimit=false
      footBoneForwardAxis=Y
      hingeRotationAxis=None
      minRotationAngle=15
      numBonesInLimb=3
      twistOffsetCurveName=None
      fKFootBone {boneName=ball_fr_l}
      iKFootBone {boneName=IK_Foot_Front_L}
     }
    ]
   }
   Output Pose(Result=Component To Local(ComponentPose=@n8))
   AnimGraphNode_StateMachine_4 = Emote SM() {is=title AnimGraphNode_StateMachine}
   n9 = Get Quadruped_Base_Struct() -> (Quadruped_Base_Struct:Quadruped Base Struct Structure)
   Save cached pose 'PreVO'(Pose="AimOffset Player '(None)'"!(
     X:float=Get LookAtYawAdjusted() -> (LookAtYawAdjusted:double)
     Y:float=Get LookAtPitchActual() -> (LookAtPitchActual:float)
     BlendSpace:Blend Space Object Reference=Break Quadruped Base Struct!(Quadruped_Base_Struct:"Quadruped Base Struct Structure (by ref)"=@n9) -> (
      LocomotionStart_4_#1:Anim Sequence Object Reference
      LocomotionStop_Slow_6_#2:Anim Sequence Object Reference
      LocomotionStop_Fast_8_#3:Anim Sequence Object Reference
      IdlePose_10_#4:Anim Sequence Object Reference
      IdleAngry_Loop_12_#5:Anim Sequence Object Reference
      BodyTurn_BS_51_#6:Blend Space 1D Object Reference
      Breathing_Loop_16_#7:Anim Sequence Object Reference
      Locomotion_Sprint_18_#8:Anim Sequence Object Reference
      Locomotion_Run_20_#9:Anim Sequence Object Reference
      Locomotion_Trot_22_#10:Anim Sequence Object Reference
      Locomotion_Walk_24_#11:Anim Sequence Object Reference
      Tall_Wag_BS_32_#12:Blend Space 1D Object Reference
      TallWag_Sit_28_#13:Anim Sequence Object Reference
      Tail_Wag_Angry_30_#14:Anim Sequence Object Reference
      BodyPose_Worried_35_#15:Anim Sequence Object Reference
      BodyPose_Combat_37_#16:Anim Sequence Object Reference
      SittingToStand_39_#17:Anim Sequence Object Reference
      Sitting_41_#18:Anim Sequence Object Reference
      StandToSitting_43_#19:Anim Sequence Object Reference
      AngryIdle_Intro_54_#20:Anim Sequence Object Reference
      AngryIdle_Exit_45_#21:Anim Sequence Object Reference
      IdleAntic_47_#22:Anim Sequence Object Reference
      AimOffset_50_#23:Blend Space Object Reference
      Chonky_Sit_ADD_60_#24:Anim Sequence Object Reference
     ) {is=title K2Node_BreakStruct}.AimOffset_50_#23
     BasePose:Pose Link Structure=Slot 'DefaultSlot'(Source=@AnimGraphNode_StateMachine_4)
     Alpha:float=Get AimOffsetAlpha() -> (AimOffsetAlpha:float)
    ) -> (Pose:Pose Link Structure) {alphaScaleBiasClamp {bias=1, scale=-1}}) {cacheName=PreVO}
   AnimGraphNode_StateMachine_3 = LowFidelityStateMachine() {is=title AnimGraphNode_StateMachine, bAllowConduitEntryStates=true}
   n10 = Apply Additive(Base=@AnimGraphNode_StateMachine_3, Additive=Use cached pose 'BodyAdditivePose'()) {alphaBoolBlend {blendOption=Linear}}
   n11 = Sequence Player!(Sequence:Anim Sequence Base Object Reference=Break Quadruped Base Struct(Quadruped_Base_Struct=Get Quadruped_Base_Struct()).IdleAntic_47_#22) -> (Pose)
   n12 = Get IdleAnticAlpha() -> (IdleAnticAlpha:double)
   Save cached pose 'LowFidelityPose'(Pose=Apply Additive(
     Base=Apply Additive(Base=@n10, Additive=@n11, Alpha=@n12) {alphaBoolBlend {blendOption=Linear}}
     Additive=Use cached pose 'DefaultTailWagPose'()
     Alpha=Get TailWagAlpha() -> (TailWagAlpha:float)
    ) {alphaCurveName=Appendage01, alphaBoolBlend {blendOption=Linear}}) {cacheName=LowFidelityPose}
   AnimGraphNode_StateMachine_5 = Locomotion() {is=title AnimGraphNode_StateMachine, bAllowConduitEntryStates=true}
   n13 = Use cached pose 'BodyAdditivePose'()
   AnimGraphNode_StateMachine_0 = IdleAntics() {is=title AnimGraphNode_StateMachine, bAllowConduitEntryStates=true}
   AnimGraphNode_StateMachine_2 = Tail() {is=title AnimGraphNode_StateMachine, bAllowConduitEntryStates=true}
   n14 = Apply Additive(
    Base=Apply Additive(
     Base=Apply Additive!(Base=@AnimGraphNode_StateMachine_5, Additive=@n13) -> (Pose) {alphaBoolBlend {blendInTime=0.25, blendOption=Sinusoidal}}
     Additive=@AnimGraphNode_StateMachine_0
     Alpha=Get IdleAnticAlpha()
    ) {alphaBoolBlend {blendOption=Linear}}
    Additive=@AnimGraphNode_StateMachine_2
    Alpha=Get TailWagAlpha()
   ) {
    alphaCurveName=Appendage01
    alphaBoolBlend {blendOption=Linear}
    alphaScaleBiasClamp {bInterpResult=true, interpSpeedDecreasing=0, interpSpeedIncreasing=8}
   }
   Save cached pose 'HighFidelityPose'(Pose=@n14) {cacheName=HighFidelityPose}
   AnimGraphNode_ChooserPlayer_1 = Chooser Player: PRA_Companion_TailAdditive!() -> (Pose:Pose Link Structure) {
    is=title AnimGraphNode_ChooserPlayer
    evaluationFrequency=OnUpdate
    chooser {
     _structType=/Script/ProxyTable.LookupProxy
     proxy=/CosmeticCompanions/Chooser/PRA_Companion_TailAdditive
     proxyTable {
      _structType=/Script/ProxyTable.ProxyTableContextProperty
      binding {contextIndex=0, displayName=Chooser Tables, isBoundToRoot=false, propertyBindingChain [ChooserTables]}
     }
    }
    chooserContextDefinition [
     {
      _structType=/Script/Chooser.ContextObjectTypeClass
      class=&/Quadruped_Base_AnimBP.Quadruped_Base_AnimBP_C
      direction=ReadWrite
     }
     {struct=/Script/Chooser.ChooserPlayerSettings, _structType=/Script/Chooser.ContextObjectTypeStruct, direction=Read}
    ]
    defaultSettings {blendTime=0.3}
   }
   n15 = Modify Curve(SourcePose=Use cached pose 'PreVO'(), CurveValues_0=1, Alpha=Get AudioEnvelopeValue() -> (AudioEnvelopeValue:float)) {curveNames [VOActivity], curveValues [0]}
   n16 = Use cached pose 'PreVO'()
   n17 = Get bIsVOActive() -> (bIsVOActive:bool)
   n18 = Apply Additive!(
    Base=Slot 'SequencerDefault'(Source=Blend Poses by bool(BlendPose_0=@n15, BlendPose_1=@n16, bActiveValue=@n17)) {slotName=SequencerDefault}
    Additive=@AnimGraphNode_ChooserPlayer_1
    Alpha:float=Get TailAdditiveAlpha() -> (TailAdditiveAlpha:float)
   ) -> (Pose) {
    alphaBoolBlend {blendInTime=0.3, blendOutTime=0.2}
    alphaScaleBiasClamp {bInterpResult=true, interpSpeedDecreasing=8, interpSpeedIncreasing=8}
   }
   Save cached pose 'PreSlopeWarpingPose'(Pose=Dead Blending!(Source:Pose Link Structure=@n18) -> (Pose:Pose Link Structure) {is=title AnimGraphNode_DeadBlending, defaultBlendMode=Linear, filteredCurves [Reaction_Happy, Reaction_Happy_01, EmoteCurveA, EmoteCurveB]}) {cacheName=PreSlopeWarpingPose}
   n19 = Break Quadruped Base Struct!(Quadruped_Base_Struct=Get Quadruped_Base_Struct()) -> (Chonky_Sit_ADD_60_#24)
   AnimGraphNode_StateMachine_1 = Idle() {is=title AnimGraphNode_StateMachine, bAllowConduitEntryStates=true}
   Save cached pose 'IdlePose'(Pose=Apply Additive(
     Base=@AnimGraphNode_StateMachine_1
     Additive=Sequence Player(Sequence=@n19)
     Alpha=Get ChonkyAdditiveWeight() -> (ChonkyAdditiveWeight:float)
    )) {cacheName=IdlePose}
   Save cached pose 'MovementPose'(Pose=LocomotionLayer!() -> (Pose:Pose Link Structure) {is=title AnimGraphNode_LinkedAnimLayer, layer=LocomotionLayer}) {cacheName=MovementPose}
   AnimGraphNode_StateMachine_6 = BodyAdditives() {is=title AnimGraphNode_StateMachine}
   Save cached pose 'BodyAdditivePose'(Pose=@AnimGraphNode_StateMachine_6) {cacheName=BodyAdditivePose}
   Save cached pose 'DefaultTailWagPose'(Pose=DefaultTailWag!() -> (Pose:Pose Link Structure) {is=title AnimGraphNode_LinkedAnimLayer, layer=DefaultTailWag}) {cacheName=DefaultTailWagPose}
   n20 = Blend Poses by bool(
    BlendPose_0=Use cached pose 'LowFidelityPose'()
    BlendPose_1=Use cached pose 'HighFidelityPose'()
    BlendTime_0=0.5
    BlendTime_1=0.5
    bActiveValue=Get bUseLowFidelity() -> (bUseLowFidelity:bool)
   ) {
    blendProfile=&SitPlant_Quadruped_Skeleton:TimeBlend_FastFeetSlowBody
    blendType=Linear
    transitionType=Inertialization
   }
   n21 = Break Quadruped Base Struct!(Quadruped_Base_Struct=Get Quadruped_Base_Struct()) -> (
    LocomotionStart_4_#1:Anim Sequence Object Reference
    LocomotionStop_Slow_6_#2:Anim Sequence Object Reference
    LocomotionStop_Fast_8_#3:Anim Sequence Object Reference
    IdlePose_10_#4:Anim Sequence Object Reference
    IdleAngry_Loop_12_#5:Anim Sequence Object Reference
    BodyTurn_BS_51_#6:Blend Space 1D Object Reference
    Breathing_Loop_16_#7:Anim Sequence Object Reference
    Locomotion_Sprint_18_#8:Anim Sequence Object Reference
    Locomotion_Run_20_#9:Anim Sequence Object Reference
    Locomotion_Trot_22_#10:Anim Sequence Object Reference
    Locomotion_Walk_24_#11:Anim Sequence Object Reference
    Tall_Wag_BS_32_#12:Blend Space 1D Object Reference
    TallWag_Sit_28_#13:Anim Sequence Object Reference
    Tail_Wag_Angry_30_#14:Anim Sequence Object Reference
    BodyPose_Worried_35_#15:Anim Sequence Object Reference
    BodyPose_Combat_37_#16:Anim Sequence Object Reference
    SittingToStand_39_#17:Anim Sequence Object Reference
    Sitting_41_#18:Anim Sequence Object Reference
    StandToSitting_43_#19:Anim Sequence Object Reference
    AngryIdle_Intro_54_#20:Anim Sequence Object Reference
    AngryIdle_Exit_45_#21:Anim Sequence Object Reference
    IdleAntic_47_#22:Anim Sequence Object Reference
    AimOffset_50_#23:Blend Space Object Reference
    Chonky_Sit_ADD_60_#24
   )
   n22 = Blendspace Player!(
    X=Get CompanionDirection() -> (CompanionDirection:float)
    Y=0
    BlendSpace:Blend Space Object Reference=@n21.BodyTurn_BS_51_#6
   ) -> (Pose)
   n23 = Get BreathingPlayRate() -> (BreathingPlayRate:float)
   Save cached pose 'PreEmotePose'(Pose=Apply Additive(
     Base=Apply Additive(Base=@n20, Additive=@n22) {alphaBoolBlend {blendOption=Linear}}
     Additive=Sequence Player!(Sequence=@n21.Breathing_Loop_16_#7, PlayRate:float=@n23) -> (Pose)
     Alpha=Get BreathingAlpha() -> (BreathingAlpha:double)
    ) {alphaBoolBlend {blendOption=Linear}}) {cacheName=PreEmotePose}
  ]}
 AnimationBlendStackGraph_0 {
  path=AnimGraph.AnimGraphNode_ChooserPlayer_1.AnimationBlendStackGraph_0
  nodes [Output Pose!(Result:Pose Link Structure=Blend Stack Input!() -> (Pose:Pose Link Structure) {is=title AnimGraphNode_BlendStackInput}) {is=title AnimGraphNode_BlendStackResult}]
 }
 IdleAntics {
  entry=Default
  path=AnimGraph.AnimGraphNode_StateMachine_0.IdleAntics
  ids=id
  nodes [
   AnimStateNode_0 = Default() {is=title AnimStateNode}
   AnimStateNode_7 = FrontEndSpawnAnticChooser() {
    is=title AnimStateNode
    stateEntered {guid=#25}
    stateFullyBlended {guid=#26}
    stateLeft {guid=#27}
   }
   AnimStateNode_9 = AnticChooser() {is=title AnimStateNode}
   AnimStateTransitionNode_1 = AnimStateTransitionNode() {bAutomaticRuleBasedOnSequencePlayerInState=true}
   AnimStateTransitionNode_14 = AnimStateTransitionNode() {crossfadeDuration=0.15}
   AnimStateTransitionNode_18 = AnimStateTransitionNode() {crossfadeDuration=0.15, minTimeBeforeReentry=0.15, priorityOrder=2}
   AnimStateTransitionNode_2 = AnimStateTransitionNode()
   AnimStateTransitionNode_5 = AnimStateTransitionNode() {bAutomaticRuleBasedOnSequencePlayerInState=true}
  ]
  links [
   {
    dst=Default
    rule=AnimGraph.AnimGraphNode_StateMachine_0.IdleAntics.AnimStateTransitionNode_2.Transition
    src=AnticChooser
    via=AnimStateTransitionNode_2
   }
   {
    dst=Default
    rule=AnimGraph.AnimGraphNode_StateMachine_0.IdleAntics.AnimStateTransitionNode_5.Transition
    src=AnticChooser
    via=AnimStateTransitionNode_5
   }
   {
    dst=AnticChooser
    rule=AnimGraph.AnimGraphNode_StateMachine_0.IdleAntics.AnimStateTransitionNode_18.Transition
    src=Default
    via=AnimStateTransitionNode_18
   }
   {
    dst=FrontEndSpawnAnticChooser
    rule=AnimGraph.AnimGraphNode_StateMachine_0.IdleAntics.AnimStateTransitionNode_14.Transition
    src=Default
    via=AnimStateTransitionNode_14
   }
   {
    dst=Default
    rule=AnimGraph.AnimGraphNode_StateMachine_0.IdleAntics.AnimStateTransitionNode_1.Transition
    src=FrontEndSpawnAnticChooser
    via=AnimStateTransitionNode_1
   }
  ]
 }
 Default {
  path=AnimGraph.AnimGraphNode_StateMachine_0.IdleAntics.AnimStateNode_0.Default
  nodes [
   AnimGraphNode_ChooserPlayer_1 = Chooser Player: PRA_Companion_AnticsNeutral!() -> (Pose:Pose Link Structure) {
    is=title AnimGraphNode_ChooserPlayer
    evaluationFrequency=OnLoop
    chooser {
     _structType=&LookupProxy
     proxy=/CosmeticCompanions/Chooser/PRA_Companion_AnticsNeutral
     proxyTable {
      _structType=&ProxyTableContextProperty
      binding {contextIndex=0, displayName=Chooser Tables, isBoundToRoot=false, propertyBindingChain [ChooserTables]}
     }
    }
    chooserContextDefinition [
     {
      _structType=&ContextObjectTypeClass
      class=&Quadruped_Base_AnimBP_C
      direction=ReadWrite
     }
     {struct=&ChooserPlayerSettings, _structType=&ContextObjectTypeStruct, direction=Read}
    ]
   }
   Output Animation Pose(Result=@AnimGraphNode_ChooserPlayer_1)
  ]
 }
 AnimationBlendStackGraph_0 {
  path=AnimGraph.AnimGraphNode_StateMachine_0.IdleAntics.AnimStateNode_0.Default.AnimGraphNode_ChooserPlayer_1.AnimationBlendStackGraph_0
  nodes [Output Pose(Result=Blend Stack Input())]
 }
 FrontEndSpawnAnticChooser {
  path=AnimGraph.AnimGraphNode_StateMachine_0.IdleAntics.AnimStateNode_7.FrontEndSpawnAnticChooser
  nodes [
   n1 = Get OnSpawnedChooser() -> (OnSpawnedChooser:Async Sequence Chooser Structure)
   Output Animation Pose(Result=Sequence Player!(Sequence=Break Async Sequence Chooser!(AsyncSequenceChooser:"Async Sequence Chooser Structure (by ref)"=@n1) -> (
      bIsLoading:bool=~
      bIsMontage:bool=~
      bIsReady:bool=~
      AnimSequence:Anim Sequence Object Reference
      AnimMontage:Anim Montage Object Reference
     ) {is=title K2Node_BreakStruct}.AnimSequence) -> (Pose) {bLoopAnimation=false}) {
    node {stateExitFunction {functionName=OnStateExit_0}}
    stateExitFunction {bSelfContext=true, memberGuid=#28, memberName=AllowedToSit_Exit}
   }
  ]
 }
 AnticChooser {
  path=AnimGraph.AnimGraphNode_StateMachine_0.IdleAntics.AnimStateNode_9.AnticChooser
  nodes [Output Animation Pose(Result=Sequence Player() {bLoopAnimation=false}) {
    node {stateEntryFunction {functionName=OnStateEntry}, stateExitFunction {functionName=OnStateExit_0}}
    stateEntryFunction {bSelfContext=true, memberGuid=#29, memberName=AnticPlaying_Enter}
    stateExitFunction {bSelfContext=true, memberGuid=#30, memberName=AnticPlaying_Exit}
   }]
 }
 Transition {path=AnimGraph.AnimGraphNode_StateMachine_0.IdleAntics.AnimStateTransitionNode_1.Transition, nodes [Result()]}
 Transition {
  path=AnimGraph.AnimGraphNode_StateMachine_0.IdleAntics.AnimStateTransitionNode_14.Transition
  nodes [
   Result(bCanEnterTransition=Get bTriggerSpawnedAntic() -> (bTriggerSpawnedAntic:bool))
   AND Boolean(
    A=Break Async Sequence Chooser(AsyncSequenceChooser=Get OnSpawnedChooser()) -> (bIsLoading=~, bIsMontage=~, bIsReady=~).bIsReady
    B=Get bIsPreviewActor() -> (bIsPreviewActor:bool)
   )
  ]
 }
 Transition {
  path=AnimGraph.AnimGraphNode_StateMachine_0.IdleAntics.AnimStateTransitionNode_18.Transition
  nodes [Result(bCanEnterTransition=Get bIsAllowAnticToPlay() -> (bIsAllowAnticToPlay:bool))]
 }
 Transition {
  path=AnimGraph.AnimGraphNode_StateMachine_0.IdleAntics.AnimStateTransitionNode_2.Transition
  nodes [Result(bCanEnterTransition=NOT Boolean(A=Get bIsCanPlayAntic() -> (bIsCanPlayAntic:bool)))]
 }
 Transition {path=AnimGraph.AnimGraphNode_StateMachine_0.IdleAntics.AnimStateTransitionNode_5.Transition, nodes [Result()]}
 Idle {
  entry=Conduit
  path=AnimGraph.AnimGraphNode_StateMachine_1.Idle
  ids=id
  nodes [
   Alias() {is=title AnimStateAliasNode, also {aliases [SitPose,StandToSit]}}
   AnimStateConduitNode_0 = Conduit()
   AnimStateNode_1 = EmoteLoops() {is=title AnimStateNode}
   AnimStateNode_2 = ReactToIdle() {is=title AnimStateNode}
   AnimStateNode_3 = SitToStand() {is=title AnimStateNode}
   AnimStateNode_4 = StandToSit() {is=title AnimStateNode}
   AnimStateNode_5 = SitPose() {is=title AnimStateNode}
   AnimStateNode_6 = StandingPose() {is=title AnimStateNode}
   AnimStateTransitionNode_0 = AnimStateTransitionNode() {automaticRuleTriggerTime=0, bAutomaticRuleBasedOnSequencePlayerInState=true, logicType=TLT_Inertialization}
   AnimStateTransitionNode_1 = AnimStateTransitionNode() {logicType=TLT_Inertialization, priorityOrder=2}
   AnimStateTransitionNode_12 = AnimStateTransitionNode() {logicType=TLT_Inertialization}
   AnimStateTransitionNode_13 = AnimStateTransitionNode() {bAutomaticRuleBasedOnSequencePlayerInState=true, bDisabled=true}
   AnimStateTransitionNode_14 = AnimStateTransitionNode() {bDisabled=true, logicType=TLT_Inertialization, priorityOrder=2}
   AnimStateTransitionNode_2 = AnimStateTransitionNode() {logicType=TLT_Inertialization, priorityOrder=100}
   AnimStateTransitionNode_3 = AnimStateTransitionNode() {bDisabled=true, logicType=TLT_Inertialization}
   AnimStateTransitionNode_4 = AnimStateTransitionNode()
   AnimStateTransitionNode_6 = AnimStateTransitionNode() {automaticRuleTriggerTime=0, bAutomaticRuleBasedOnSequencePlayerInState=true, crossfadeDuration=0.1, logicType=TLT_Inertialization}
   AnimStateTransitionNode_7 = AnimStateTransitionNode() {bDisabled=true, logicType=TLT_Inertialization}
   AnimStateTransitionNode_8 = AnimStateTransitionNode() {logicType=TLT_Inertialization}
  ]
  links [
   {
    dst=SitToStand
    rule=AnimGraph.AnimGraphNode_StateMachine_1.Idle.AnimStateTransitionNode_8.AnimationTransitionGraph_1
    src=Alias
    via=AnimStateTransitionNode_8
   }
   {
    dst=EmoteLoops
    rule=AnimGraph.AnimGraphNode_StateMachine_1.Idle.AnimStateTransitionNode_7.AnimationTransitionGraph_5
    src=Conduit
    via=AnimStateTransitionNode_7
   }
   {
    dst=ReactToIdle
    rule=AnimGraph.AnimGraphNode_StateMachine_1.Idle.AnimStateTransitionNode_14.AnimationTransitionGraph_5
    src=Conduit
    via=AnimStateTransitionNode_14
   }
   {
    dst=SitPose
    rule=AnimGraph.AnimGraphNode_StateMachine_1.Idle.AnimStateTransitionNode_3.AnimationTransitionGraph_1
    src=Conduit
    via=AnimStateTransitionNode_3
   }
   {
    dst=StandToSit
    rule=AnimGraph.AnimGraphNode_StateMachine_1.Idle.AnimStateTransitionNode_1.AnimationTransitionGraph_1
    src=Conduit
    via=AnimStateTransitionNode_1
   }
   {
    dst=StandingPose
    rule=AnimGraph.AnimGraphNode_StateMachine_1.Idle.AnimStateTransitionNode_2.AnimationTransitionGraph_5
    src=Conduit
    via=AnimStateTransitionNode_2
   }
   {
    dst=ReactToIdle
    rule=AnimGraph.AnimGraphNode_StateMachine_1.Idle.AnimStateTransitionNode_12.AnimationTransitionGraph_5
    src=EmoteLoops
    via=AnimStateTransitionNode_12
   }
   {
    dst=StandingPose
    rule=AnimGraph.AnimGraphNode_StateMachine_1.Idle.AnimStateTransitionNode_13.AnimationTransitionGraph_5
    src=ReactToIdle
    via=AnimStateTransitionNode_13
   }
   {
    dst=StandingPose
    rule=AnimGraph.AnimGraphNode_StateMachine_1.Idle.AnimStateTransitionNode_6.AnimationTransitionGraph_1
    src=SitToStand
    via=AnimStateTransitionNode_6
   }
   {
    dst=SitPose
    rule=AnimGraph.AnimGraphNode_StateMachine_1.Idle.AnimStateTransitionNode_0.AnimationTransitionGraph_1
    src=StandToSit
    via=AnimStateTransitionNode_0
   }
   {
    dst=Conduit
    rule=AnimGraph.AnimGraphNode_StateMachine_1.Idle.AnimStateTransitionNode_4.AnimationTransitionGraph_5
    src=StandingPose
    via=AnimStateTransitionNode_4
   }
  ]
 }
 Conduit {path=AnimGraph.AnimGraphNode_StateMachine_1.Idle.AnimStateConduitNode_0.Conduit, nodes [Result(bCanEnterTransition=true)]}
 EmoteLoops {
  path=AnimGraph.AnimGraphNode_StateMachine_1.Idle.AnimStateNode_1.EmoteLoops
  nodes [
   AnimGraphNode_ChooserPlayer_0 = Chooser Player: PRA_Companion_SpecialIdle!() -> (Pose:Pose Link Structure) {
    is=title AnimGraphNode_ChooserPlayer
    chooser {
     _structType=&LookupProxy
     proxy=/CosmeticCompanions/Chooser/PRA_Companion_SpecialIdle
     proxyTable {
      _structType=&ProxyTableContextProperty
      binding {contextIndex=0, displayName=Chooser Tables, isBoundToRoot=false, propertyBindingChain [ChooserTables]}
     }
    }
    chooserContextDefinition [
     {
      _structType=&ContextObjectTypeClass
      class=&Quadruped_Base_AnimBP_C
      direction=ReadWrite
     }
     {struct=&ChooserPlayerSettings, _structType=&ContextObjectTypeStruct, direction=Read}
    ]
    defaultSettings {blendOption=Linear}
   }
   Output Animation Pose(Result=@AnimGraphNode_ChooserPlayer_0)
  ]
 }
 AnimationBlendStackGraph_0 {
  path=AnimGraph.AnimGraphNode_StateMachine_1.Idle.AnimStateNode_1.EmoteLoops.AnimGraphNode_ChooserPlayer_0.AnimationBlendStackGraph_0
  nodes [Output Pose(Result=Blend Stack Input())]
 }
 ReactToIdle {
  path=AnimGraph.AnimGraphNode_StateMachine_1.Idle.AnimStateNode_2.ReactToIdle
  nodes [
   AnimGraphNode_ChooserPlayer_0 = Chooser Player: PRA_Companion_ReactToIdle!() -> (Pose:Pose Link Structure) {
    is=title AnimGraphNode_ChooserPlayer
    chooser {
     _structType=&LookupProxy
     proxy=/CosmeticCompanions/Chooser/PRA_Companion_ReactToIdle
     proxyTable {
      _structType=&ProxyTableContextProperty
      binding {contextIndex=0, displayName=Chooser Tables, isBoundToRoot=false, propertyBindingChain [ChooserTables]}
     }
    }
    chooserContextDefinition [
     {
      _structType=&ContextObjectTypeClass
      class=&Quadruped_Base_AnimBP_C
      direction=ReadWrite
     }
     {struct=&ChooserPlayerSettings, _structType=&ContextObjectTypeStruct, direction=Read}
    ]
    defaultSettings {blendOption=Linear}
   }
   Output Animation Pose(Result=@AnimGraphNode_ChooserPlayer_0) {
    node {stateEntryFunction {functionName=SpecialIdle_StateEntry}, stateExitFunction {functionName=SpecialIdle_StateExit}}
    stateEntryFunction {bSelfContext=true, memberGuid=#31, memberName=SpecialIdle_StateEntry}
    stateExitFunction {bSelfContext=true, memberGuid=#32, memberName=SpecialIdle_StateExit}
   }
  ]
 }
 AnimationBlendStackGraph_0 {
  path=AnimGraph.AnimGraphNode_StateMachine_1.Idle.AnimStateNode_2.ReactToIdle.AnimGraphNode_ChooserPlayer_0.AnimationBlendStackGraph_0
  nodes [Output Pose(Result=Blend Stack Input())]
 }
 SitToStand {
  path=AnimGraph.AnimGraphNode_StateMachine_1.Idle.AnimStateNode_3.SitToStand
  nodes [
   n1 = Get Quadruped_Base_Struct()
   Output Animation Pose(Result=Sequence Player(Sequence=Break Quadruped Base Struct(Quadruped_Base_Struct=@n1).SittingToStand_39_#17) {bLoopAnimation=false})
  ]
 }
 StandToSit {
  path=AnimGraph.AnimGraphNode_StateMachine_1.Idle.AnimStateNode_4.StandToSit
  nodes [
   n1 = Break Quadruped Base Struct(Quadruped_Base_Struct=Get Quadruped_Base_Struct())
   Output Animation Pose(Result=Modify Curve(SourcePose=Sequence Player(Sequence=@n1.StandToSitting_43_#19) {bLoopAnimation=false}, CurveValues_0=1, Alpha=1) {curveNames [SitTail], curveValues [0]})
  ]
 }
 SitPose {
  path=AnimGraph.AnimGraphNode_StateMachine_1.Idle.AnimStateNode_5.SitPose
  nodes [
   n1 = Break Quadruped Base Struct(Quadruped_Base_Struct=Get Quadruped_Base_Struct())
   Output Animation Pose(Result=Modify Curve!(
     SourcePose=Sequence Player(Sequence=@n1.Sitting_41_#18)
     CurveValues_0=1
     CurveValues_1:float=1
     CurveValues_2:float=1
     Alpha=1
    ) -> (Pose) {curveNames [idle,SitWeight,SitTail], curveValues [0,0,0]})
  ]
 }
 StandingPose {
  path=AnimGraph.AnimGraphNode_StateMachine_1.Idle.AnimStateNode_6.StandingPose
  nodes [
   AnimGraphNode_StateMachine_0 = StandingStateMachine() {is=title AnimGraphNode_StateMachine, bAllowConduitEntryStates=true}
   Output Animation Pose(Result=Modify Curve!(SourcePose=@AnimGraphNode_StateMachine_0, CurveValues_0=1, Alpha=1) -> (Pose) {curveNames [idle], curveValues [0]})
  ]
 }
 StandingStateMachine {
  entry=Conduit
  path=AnimGraph.AnimGraphNode_StateMachine_1.Idle.AnimStateNode_6.StandingPose.AnimGraphNode_StateMachine_0.StandingStateMachine
  ids=id
  nodes [
   AnimStateConduitNode_0 = Conduit()
   AnimStateNode_0 = Idle() {is=title AnimStateNode}
   AnimStateNode_1 = AngryIdleIntro() {is=title AnimStateNode}
   AnimStateNode_2 = AngryIdle() {is=title AnimStateNode}
   AnimStateNode_3 = AngryIdleExit() {
    is=title AnimStateNode
    stateEntered {guid=#33}
    stateFullyBlended {guid=#34}
    stateLeft {guid=#35}
   }
   AnimStateTransitionNode_0 = AnimStateTransitionNode() {logicType=TLT_Inertialization, priorityOrder=100}
   AnimStateTransitionNode_1 = AnimStateTransitionNode() {bAllowInertializationForSelfTransitions=true}
   AnimStateTransitionNode_2 = AnimStateTransitionNode() {bAllowInertializationForSelfTransitions=true, crossfadeDuration=0.1, logicType=TLT_Inertialization}
   AnimStateTransitionNode_3 = AnimStateTransitionNode() {
    automaticRuleTriggerTime=0
    bAllowInertializationForSelfTransitions=true
    bAutomaticRuleBasedOnSequencePlayerInState=true
    logicType=TLT_Inertialization
   }
   AnimStateTransitionNode_4 = AnimStateTransitionNode() {
    bAllowInertializationForSelfTransitions=true
    crossfadeDuration=0.1
    logicType=TLT_Inertialization
    transitionEnd {guid=#36}
    transitionInterrupt {guid=#37}
    transitionStart {guid=#38}
   }
   AnimStateTransitionNode_5 = AnimStateTransitionNode() {automaticRuleTriggerTime=0, bAllowInertializationForSelfTransitions=true, bAutomaticRuleBasedOnSequencePlayerInState=true}
  ]
  links [
   {
    dst=AngryIdleExit
    rule=AnimGraph.AnimGraphNode_StateMachine_1.Idle.AnimStateNode_6.StandingPose.AnimGraphNode_StateMachine_0.StandingStateMachine.AnimStateTransitionNode_4.AnimationTransitionGraph_4
    src=AngryIdle
    via=AnimStateTransitionNode_4
   }
   {
    dst=Conduit
    rule=AnimGraph.AnimGraphNode_StateMachine_1.Idle.AnimStateNode_6.StandingPose.AnimGraphNode_StateMachine_0.StandingStateMachine.AnimStateTransitionNode_5.AnimationTransitionGraph_4
    src=AngryIdleExit
    via=AnimStateTransitionNode_5
   }
   {
    dst=AngryIdle
    rule=AnimGraph.AnimGraphNode_StateMachine_1.Idle.AnimStateNode_6.StandingPose.AnimGraphNode_StateMachine_0.StandingStateMachine.AnimStateTransitionNode_3.AnimationTransitionGraph_5
    src=AngryIdleIntro
    via=AnimStateTransitionNode_3
   }
   {
    dst=AngryIdleIntro
    rule=AnimGraph.AnimGraphNode_StateMachine_1.Idle.AnimStateNode_6.StandingPose.AnimGraphNode_StateMachine_0.StandingStateMachine.AnimStateTransitionNode_2.Transition
    src=Conduit
    via=AnimStateTransitionNode_2
   }
   {
    dst=Idle
    rule=AnimGraph.AnimGraphNode_StateMachine_1.Idle.AnimStateNode_6.StandingPose.AnimGraphNode_StateMachine_0.StandingStateMachine.AnimStateTransitionNode_0.Transition
    src=Conduit
    via=AnimStateTransitionNode_0
   }
   {
    dst=Conduit
    rule=AnimGraph.AnimGraphNode_StateMachine_1.Idle.AnimStateNode_6.StandingPose.AnimGraphNode_StateMachine_0.StandingStateMachine.AnimStateTransitionNode_1.Transition
    src=Idle
    via=AnimStateTransitionNode_1
   }
  ]
 }
 Conduit {
  path=AnimGraph.AnimGraphNode_StateMachine_1.Idle.AnimStateNode_6.StandingPose.AnimGraphNode_StateMachine_0.StandingStateMachine.AnimStateConduitNode_0.Conduit
  nodes [Result(bCanEnterTransition=true)]
 }
 Idle {
  path=AnimGraph.AnimGraphNode_StateMachine_1.Idle.AnimStateNode_6.StandingPose.AnimGraphNode_StateMachine_0.StandingStateMachine.AnimStateNode_0.Idle
  nodes [
   Modify Curve(CurveValues_0=1, Alpha=1) {curveNames [idle], curveValues [0]}
   n1 = Get Quadruped_Base_Struct()
   Output Animation Pose(Result=Sequence Player(Sequence=Break Quadruped Base Struct(Quadruped_Base_Struct=@n1).IdlePose_10_#4) {bLoopAnimation=false, groupName=Locomotion, method=SyncGroup})
  ]
 }
 AngryIdleIntro {
  path=AnimGraph.AnimGraphNode_StateMachine_1.Idle.AnimStateNode_6.StandingPose.AnimGraphNode_StateMachine_0.StandingStateMachine.AnimStateNode_1.AngryIdleIntro
  nodes [
   n1 = Get Quadruped_Base_Struct()
   Output Animation Pose(Result=Sequence Player(Sequence=Break Quadruped Base Struct(Quadruped_Base_Struct=@n1).AngryIdle_Intro_54_#20) {bLoopAnimation=false})
  ]
 }
 AngryIdle {
  path=AnimGraph.AnimGraphNode_StateMachine_1.Idle.AnimStateNode_6.StandingPose.AnimGraphNode_StateMachine_0.StandingStateMachine.AnimStateNode_2.AngryIdle
  nodes [
   n1 = Get Quadruped_Base_Struct()
   Output Animation Pose(Result=Sequence Player(Sequence=Break Quadruped Base Struct(Quadruped_Base_Struct=@n1).IdleAngry_Loop_12_#5))
  ]
 }
 AngryIdleExit {
  path=AnimGraph.AnimGraphNode_StateMachine_1.Idle.AnimStateNode_6.StandingPose.AnimGraphNode_StateMachine_0.StandingStateMachine.AnimStateNode_3.AngryIdleExit
  nodes [
   n1 = Get Quadruped_Base_Struct()
   Output Animation Pose(Result=Sequence Player(Sequence=Break Quadruped Base Struct(Quadruped_Base_Struct=@n1).AngryIdle_Exit_45_#21) {bLoopAnimation=false})
  ]
 }
 Transition {
  path=AnimGraph.AnimGraphNode_StateMachine_1.Idle.AnimStateNode_6.StandingPose.AnimGraphNode_StateMachine_0.StandingStateMachine.AnimStateTransitionNode_0.Transition
  nodes [Result(bCanEnterTransition=true)]
 }
 Transition {
  path=AnimGraph.AnimGraphNode_StateMachine_1.Idle.AnimStateNode_6.StandingPose.AnimGraphNode_StateMachine_0.StandingStateMachine.AnimStateTransitionNode_1.Transition
  nodes [Result(bCanEnterTransition=true)]
 }
 Transition {
  path=AnimGraph.AnimGraphNode_StateMachine_1.Idle.AnimStateNode_6.StandingPose.AnimGraphNode_StateMachine_0.StandingStateMachine.AnimStateTransitionNode_2.Transition
  nodes [Result(bCanEnterTransition=Get bIsDelayedMoodCombat() -> (bIsDelayedMoodCombat:bool))]
 }
 AnimationTransitionGraph_5 {
  path=AnimGraph.AnimGraphNode_StateMachine_1.Idle.AnimStateNode_6.StandingPose.AnimGraphNode_StateMachine_0.StandingStateMachine.AnimStateTransitionNode_3.AnimationTransitionGraph_5
  nodes [Result()]
 }
 AnimationTransitionGraph_4 {
  path=AnimGraph.AnimGraphNode_StateMachine_1.Idle.AnimStateNode_6.StandingPose.AnimGraphNode_StateMachine_0.StandingStateMachine.AnimStateTransitionNode_4.AnimationTransitionGraph_4
  nodes [Result(bCanEnterTransition=NOT Boolean(A=Get bIsMoodCombat() -> (bIsMoodCombat:bool)))]
 }
 AnimationTransitionGraph_4 {
  path=AnimGraph.AnimGraphNode_StateMachine_1.Idle.AnimStateNode_6.StandingPose.AnimGraphNode_StateMachine_0.StandingStateMachine.AnimStateTransitionNode_5.AnimationTransitionGraph_4
  nodes [Result()]
 }
 AnimationTransitionGraph_1 {path=AnimGraph.AnimGraphNode_StateMachine_1.Idle.AnimStateTransitionNode_0.AnimationTransitionGraph_1, nodes [Result()]}
 AnimationTransitionGraph_1 {
  path=AnimGraph.AnimGraphNode_StateMachine_1.Idle.AnimStateTransitionNode_1.AnimationTransitionGraph_1
  nodes [Result(bCanEnterTransition=Get bIsStandToSit() -> (bIsStandToSit:bool))]
 }
 AnimationTransitionGraph_5 {
  path=AnimGraph.AnimGraphNode_StateMachine_1.Idle.AnimStateTransitionNode_12.AnimationTransitionGraph_5
  nodes [Result(bCanEnterTransition=NOT Boolean(A=Get bIsPlayingEmote()))]
 }
 AnimationTransitionGraph_5 {path=AnimGraph.AnimGraphNode_StateMachine_1.Idle.AnimStateTransitionNode_13.AnimationTransitionGraph_5, nodes [Result()]}
 AnimationTransitionGraph_5 {
  path=AnimGraph.AnimGraphNode_StateMachine_1.Idle.AnimStateTransitionNode_14.AnimationTransitionGraph_5
  nodes [Result(bCanEnterTransition=Get bIsReactionHappyPlaying() -> (bIsReactionHappyPlaying:bool))]
 }
 AnimationTransitionGraph_5 {
  path=AnimGraph.AnimGraphNode_StateMachine_1.Idle.AnimStateTransitionNode_2.AnimationTransitionGraph_5
  nodes [Result(bCanEnterTransition=true)]
 }
 AnimationTransitionGraph_1 {
  path=AnimGraph.AnimGraphNode_StateMachine_1.Idle.AnimStateTransitionNode_3.AnimationTransitionGraph_1
  nodes [Result(bCanEnterTransition=Get bIsSitting() -> (bIsSitting:bool))]
 }
 AnimationTransitionGraph_5 {
  path=AnimGraph.AnimGraphNode_StateMachine_1.Idle.AnimStateTransitionNode_4.AnimationTransitionGraph_5
  nodes [Result(bCanEnterTransition=true)]
 }
 AnimationTransitionGraph_1 {path=AnimGraph.AnimGraphNode_StateMachine_1.Idle.AnimStateTransitionNode_6.AnimationTransitionGraph_1, nodes [Result()]}
 AnimationTransitionGraph_5 {
  path=AnimGraph.AnimGraphNode_StateMachine_1.Idle.AnimStateTransitionNode_7.AnimationTransitionGraph_5
  nodes [Result(bCanEnterTransition=Get bIsSpecialIdleEnter() -> (bIsSpecialIdleEnter:bool))]
 }
 AnimationTransitionGraph_1 {
  path=AnimGraph.AnimGraphNode_StateMachine_1.Idle.AnimStateTransitionNode_8.AnimationTransitionGraph_1
  nodes [Result(bCanEnterTransition=Get bIsSitToStand() -> (bIsSitToStand:bool))]
 }
 Tail {
  entry=Conduit
  path=AnimGraph.AnimGraphNode_StateMachine_2.Tail
  ids=id
  nodes [
   AnimStateConduitNode_0 = Conduit()
   AnimStateNode_0 = Default()
   AnimStateNode_2 = Combat() {is=title AnimStateNode}
   AnimStateTransitionNode_0 = AnimStateTransitionNode() {priorityOrder=100}
   AnimStateTransitionNode_1 = AnimStateTransitionNode()
   AnimStateTransitionNode_4 = AnimStateTransitionNode() {crossfadeDuration=0.5, priorityOrder=2}
   AnimStateTransitionNode_5 = AnimStateTransitionNode()
  ]
  links [
   {
    dst=Conduit
    rule=AnimGraph.AnimGraphNode_StateMachine_2.Tail.AnimStateTransitionNode_5.Transition
    src=Combat
    via=AnimStateTransitionNode_5
   }
   {
    dst=Combat
    rule=AnimGraph.AnimGraphNode_StateMachine_2.Tail.AnimStateTransitionNode_4.Transition
    src=Conduit
    via=AnimStateTransitionNode_4
   }
   {
    dst=Default
    rule=AnimGraph.AnimGraphNode_StateMachine_2.Tail.AnimStateTransitionNode_0.Transition
    src=Conduit
    via=AnimStateTransitionNode_0
   }
   {
    dst=Conduit
    rule=AnimGraph.AnimGraphNode_StateMachine_2.Tail.AnimStateTransitionNode_1.Transition
    src=Default
    via=AnimStateTransitionNode_1
   }
  ]
 }
 Conduit {path=AnimGraph.AnimGraphNode_StateMachine_2.Tail.AnimStateConduitNode_0.Conduit, nodes [Result(bCanEnterTransition=true)]}
 Default {
  path=AnimGraph.AnimGraphNode_StateMachine_2.Tail.AnimStateNode_0.Default
  nodes [Output Animation Pose(Result=Use cached pose 'DefaultTailWagPose'())]
 }
 Combat {
  path=AnimGraph.AnimGraphNode_StateMachine_2.Tail.AnimStateNode_2.Combat
  nodes [
   n1 = Get Quadruped_Base_Struct()
   Output Animation Pose(Result=Sequence Player(Sequence=Break Quadruped Base Struct(Quadruped_Base_Struct=@n1).Tail_Wag_Angry_30_#14))
  ]
 }
 Transition {path=AnimGraph.AnimGraphNode_StateMachine_2.Tail.AnimStateTransitionNode_0.Transition, nodes [Result(bCanEnterTransition=true)]}
 Transition {path=AnimGraph.AnimGraphNode_StateMachine_2.Tail.AnimStateTransitionNode_1.Transition, nodes [Result(bCanEnterTransition=true)]}
 Transition {
  path=AnimGraph.AnimGraphNode_StateMachine_2.Tail.AnimStateTransitionNode_4.Transition
  nodes [Result(bCanEnterTransition=Get bIsDelayedMoodCombat())]
 }
 Transition {path=AnimGraph.AnimGraphNode_StateMachine_2.Tail.AnimStateTransitionNode_5.Transition, nodes [Result(bCanEnterTransition=true)]}
 LowFidelityStateMachine {
  entry=Idle
  path=AnimGraph.AnimGraphNode_StateMachine_3.LowFidelityStateMachine
  ids=id
  nodes [
   AnimStateNode_0 = Idle() {bAlwaysResetOnEntry=true}
   AnimStateNode_3 = Locomotion() {is=title AnimStateNode, bAlwaysResetOnEntry=true}
   AnimStateTransitionNode_4 = AnimStateTransitionNode() {crossfadeDuration=0.4, logicType=TLT_Inertialization}
   AnimStateTransitionNode_8 = AnimStateTransitionNode() {
    crossfadeDuration=0.4
    priorityOrder=3
    blendProfileWrapper {blendProfile=&SitPlant_Quadruped_Skeleton:TimeBlend_FastFeetSlowBody}
   }
  ]
  links [
   {
    dst=Locomotion
    rule=AnimGraph.AnimGraphNode_StateMachine_3.LowFidelityStateMachine.AnimStateTransitionNode_8.Transition
    src=Idle
    via=AnimStateTransitionNode_8
   }
   {
    dst=Idle
    rule=AnimGraph.AnimGraphNode_StateMachine_3.LowFidelityStateMachine.AnimStateTransitionNode_4.Transition
    src=Locomotion
    via=AnimStateTransitionNode_4
   }
  ]
 }
 Idle {
  path=AnimGraph.AnimGraphNode_StateMachine_3.LowFidelityStateMachine.AnimStateNode_0.Idle
  nodes [Output Animation Pose(Result=Use cached pose 'IdlePose'())]
 }
 Locomotion {
  path=AnimGraph.AnimGraphNode_StateMachine_3.LowFidelityStateMachine.AnimStateNode_3.Locomotion
  nodes [Output Animation Pose(Result=Use cached pose 'MovementPose'())]
 }
 Transition {
  path=AnimGraph.AnimGraphNode_StateMachine_3.LowFidelityStateMachine.AnimStateTransitionNode_4.Transition
  nodes [Result(bCanEnterTransition=NOT Boolean(A=Get bIsMoving() -> (bIsMoving:bool)))]
 }
 Transition {
  path=AnimGraph.AnimGraphNode_StateMachine_3.LowFidelityStateMachine.AnimStateTransitionNode_8.Transition
  nodes [Result(bCanEnterTransition=Get bIsMoving())]
 }
 Emote SM {
  entry=Default
  path=AnimGraph.AnimGraphNode_StateMachine_4.Emote SM
  ids=id
  nodes [
   ToDefault() {is=title AnimStateAliasNode, also {aliases [EmoteLoops,ReactToIdle]}}
   AnimStateNode_1 = EmoteLoops()
   AnimStateNode_2 = ReactToIdle()
   AnimStateNode_6 = Default()
   AnimStateTransitionNode_0 = AnimStateTransitionNode() {bAllowInertializationForSelfTransitions=true}
   AnimStateTransitionNode_12 = AnimStateTransitionNode() {logicType=TLT_Inertialization}
   AnimStateTransitionNode_13 = AnimStateTransitionNode() {automaticRuleTriggerTime=0.2, bAutomaticRuleBasedOnSequencePlayerInState=true, crossfadeDuration=0.3}
   AnimStateTransitionNode_14 = AnimStateTransitionNode() {bDisabled=true, logicType=TLT_Inertialization, priorityOrder=2}
   AnimStateTransitionNode_7 = AnimStateTransitionNode() {logicType=TLT_Inertialization}
  ]
  links [
   {
    dst=EmoteLoops
    rule=AnimGraph.AnimGraphNode_StateMachine_4.Emote SM.AnimStateTransitionNode_7.AnimationTransitionGraph_5
    src=Default
    via=AnimStateTransitionNode_7
   }
   {
    dst=ReactToIdle
    rule=AnimGraph.AnimGraphNode_StateMachine_4.Emote SM.AnimStateTransitionNode_14.AnimationTransitionGraph_5
    src=Default
    via=AnimStateTransitionNode_14
   }
   {
    dst=ReactToIdle
    rule=AnimGraph.AnimGraphNode_StateMachine_4.Emote SM.AnimStateTransitionNode_12.AnimationTransitionGraph_5
    src=EmoteLoops
    via=AnimStateTransitionNode_12
   }
   {
    dst=Default
    rule=AnimGraph.AnimGraphNode_StateMachine_4.Emote SM.AnimStateTransitionNode_13.AnimationTransitionGraph_5
    src=ReactToIdle
    via=AnimStateTransitionNode_13
   }
   {
    dst=Default
    rule=AnimGraph.AnimGraphNode_StateMachine_4.Emote SM.AnimStateTransitionNode_0.Transition
    src=ToDefault
    via=AnimStateTransitionNode_0
   }
  ]
 }
 EmoteLoops {
  path=AnimGraph.AnimGraphNode_StateMachine_4.Emote SM.AnimStateNode_1.EmoteLoops
  nodes [
   AnimGraphNode_ChooserPlayer_0 = Chooser Player: PRA_Companion_SpecialIdle() {
    chooser {
     _structType=&LookupProxy
     proxy=&PRA_Companion_SpecialIdle
     proxyTable {
      _structType=&ProxyTableContextProperty
      binding {contextIndex=0, displayName=Chooser Tables, isBoundToRoot=false, propertyBindingChain [ChooserTables]}
     }
    }
    chooserContextDefinition [
     {
      _structType=&ContextObjectTypeClass
      class=&Quadruped_Base_AnimBP_C
      direction=ReadWrite
     }
     {struct=&ChooserPlayerSettings, _structType=&ContextObjectTypeStruct, direction=Read}
    ]
    defaultSettings {bForceLooping=true, blendOption=Linear}
   }
   Output Animation Pose(Result=@AnimGraphNode_ChooserPlayer_0)
  ]
 }
 AnimationBlendStackGraph_0 {
  path=AnimGraph.AnimGraphNode_StateMachine_4.Emote SM.AnimStateNode_1.EmoteLoops.AnimGraphNode_ChooserPlayer_0.AnimationBlendStackGraph_0
  nodes [Output Pose(Result=Blend Stack Input())]
 }
 ReactToIdle {
  path=AnimGraph.AnimGraphNode_StateMachine_4.Emote SM.AnimStateNode_2.ReactToIdle
  nodes [
   AnimGraphNode_ChooserPlayer_0 = Chooser Player: PRA_Companion_ReactToIdle() {
    chooser {
     _structType=&LookupProxy
     proxy=&PRA_Companion_ReactToIdle
     proxyTable {
      _structType=&ProxyTableContextProperty
      binding {contextIndex=0, displayName=Chooser Tables, isBoundToRoot=false, propertyBindingChain [ChooserTables]}
     }
    }
    chooserContextDefinition [
     {
      _structType=&ContextObjectTypeClass
      class=&Quadruped_Base_AnimBP_C
      direction=ReadWrite
     }
     {struct=&ChooserPlayerSettings, _structType=&ContextObjectTypeStruct, direction=Read}
    ]
    defaultSettings {blendOption=Linear}
   }
   Output Animation Pose(Result=@AnimGraphNode_ChooserPlayer_0) {
    node {stateEntryFunction {functionName=SpecialIdle_StateEntry}, stateExitFunction {functionName=SpecialIdle_StateExit}}
    stateEntryFunction {bSelfContext=true, memberGuid=#31, memberName=SpecialIdle_StateEntry}
    stateExitFunction {bSelfContext=true, memberGuid=#32, memberName=SpecialIdle_StateExit}
   }
  ]
 }
 AnimationBlendStackGraph_0 {
  path=AnimGraph.AnimGraphNode_StateMachine_4.Emote SM.AnimStateNode_2.ReactToIdle.AnimGraphNode_ChooserPlayer_0.AnimationBlendStackGraph_0
  nodes [Output Pose(Result=Blend Stack Input())]
 }
 Default {
  path=AnimGraph.AnimGraphNode_StateMachine_4.Emote SM.AnimStateNode_6.Default
  nodes [Output Animation Pose(Result=Use cached pose 'PreEmotePose'())]
 }
 Transition {
  path=AnimGraph.AnimGraphNode_StateMachine_4.Emote SM.AnimStateTransitionNode_0.Transition
  nodes [Result(bCanEnterTransition=Get bIsMoving())]
 }
 AnimationTransitionGraph_5 {
  path=AnimGraph.AnimGraphNode_StateMachine_4.Emote SM.AnimStateTransitionNode_12.AnimationTransitionGraph_5
  nodes [Result(bCanEnterTransition=NOT Boolean(A=Get bIsPlayingEmote()))]
 }
 AnimationTransitionGraph_5 {path=AnimGraph.AnimGraphNode_StateMachine_4.Emote SM.AnimStateTransitionNode_13.AnimationTransitionGraph_5, nodes [Result()]}
 AnimationTransitionGraph_5 {
  path=AnimGraph.AnimGraphNode_StateMachine_4.Emote SM.AnimStateTransitionNode_14.AnimationTransitionGraph_5
  nodes [Result(bCanEnterTransition=Get bIsReactionHappyPlaying())]
 }
 AnimationTransitionGraph_5 {
  path=AnimGraph.AnimGraphNode_StateMachine_4.Emote SM.AnimStateTransitionNode_7.AnimationTransitionGraph_5
  nodes [Result(bCanEnterTransition=Get bIsSpecialIdleEnter())]
 }
 Locomotion {
  entry=Idle
  path=AnimGraph.AnimGraphNode_StateMachine_5.Locomotion
  ids=id
  nodes [
   IsTurning() {is=title AnimStateAliasNode, also {aliases [Idle, Locomotion, LocomotionStop]}}
   AnimStateNode_0 = Idle() {bAlwaysResetOnEntry=true}
   AnimStateNode_3 = Locomotion() {
    bAlwaysResetOnEntry=true
    stateEntered {guid=#39}
    stateFullyBlended {guid=#40}
    stateLeft {guid=#41}
   }
   AnimStateNode_4 = LocomotionStop() {
    is=title AnimStateNode
    bAlwaysResetOnEntry=true
    stateEntered {guid=#42}
    stateFullyBlended {guid=#43}
    stateLeft {guid=#44}
   }
   AnimStateNode_5 = Start Turn() {is=title AnimStateNode}
   AnimStateTransitionNode_1 = AnimStateTransitionNode() {blendMode=Linear, crossfadeDuration=0.15, logicType=TLT_Inertialization}
   AnimStateTransitionNode_10 = AnimStateTransitionNode() {
    logicType=TLT_Inertialization
    priorityOrder=100
    blendProfileWrapper {blendProfile=&SitPlant_Quadruped_Skeleton:TimeBlend_FastFeetSlowBody}
    transitionEnd {guid=#45}
    transitionInterrupt {guid=#46}
    transitionStart {guid=#47}
   }
   AnimStateTransitionNode_11 = AnimStateTransitionNode() {bAllowInertializationForSelfTransitions=true}
   AnimStateTransitionNode_18 = AnimStateTransitionNode() {
    automaticRuleTriggerTime=0.05
    bAllowInertializationForSelfTransitions=true
    bAutomaticRuleBasedOnSequencePlayerInState=true
    blendMode=Linear
   }
   AnimStateTransitionNode_3 = AnimStateTransitionNode() {blendMode=Linear, crossfadeDuration=0.1, logicType=TLT_Inertialization, priorityOrder=10}
   AnimStateTransitionNode_4 = AnimStateTransitionNode()
   AnimStateTransitionNode_5 = AnimStateTransitionNode() {bAutomaticRuleBasedOnSequencePlayerInState=true}
  ]
  links [
   {
    dst=Locomotion
    rule=AnimGraph.AnimGraphNode_StateMachine_5.Locomotion.AnimStateTransitionNode_11.Transition
    src=Idle
    via=AnimStateTransitionNode_11
   }
   {
    dst=Start Turn
    rule=AnimGraph.AnimGraphNode_StateMachine_5.Locomotion.AnimStateTransitionNode_3.Transition
    src=IsTurning
    via=AnimStateTransitionNode_3
   }
   {
    dst=Idle
    rule=AnimGraph.AnimGraphNode_StateMachine_5.Locomotion.AnimStateTransitionNode_4.Transition
    src=Locomotion
    via=AnimStateTransitionNode_4
   }
   {
    dst=LocomotionStop
    rule=AnimGraph.AnimGraphNode_StateMachine_5.Locomotion.AnimStateTransitionNode_1.Transition
    src=Locomotion
    via=AnimStateTransitionNode_1
   }
   {
    dst=Idle
    rule=AnimGraph.AnimGraphNode_StateMachine_5.Locomotion.AnimStateTransitionNode_5.Transition
    src=LocomotionStop
    via=AnimStateTransitionNode_5
   }
   {
    dst=Locomotion
    rule=AnimGraph.AnimGraphNode_StateMachine_5.Locomotion.AnimStateTransitionNode_10.Transition
    src=LocomotionStop
    via=AnimStateTransitionNode_10
   }
   {
    dst=Locomotion
    rule=AnimGraph.AnimGraphNode_StateMachine_5.Locomotion.AnimStateTransitionNode_18.Transition
    src=Start Turn
    via=AnimStateTransitionNode_18
   }
  ]
 }
 Idle {
  path=AnimGraph.AnimGraphNode_StateMachine_5.Locomotion.AnimStateNode_0.Idle
  nodes [Output Animation Pose(Result=Use cached pose 'IdlePose'())]
 }
 Locomotion {
  path=AnimGraph.AnimGraphNode_StateMachine_5.Locomotion.AnimStateNode_3.Locomotion
  nodes [Output Animation Pose(Result=Use cached pose 'MovementPose'())]
 }
 LocomotionStop {
  path=AnimGraph.AnimGraphNode_StateMachine_5.Locomotion.AnimStateNode_4.LocomotionStop
  nodes [
   AnimGraphNode_StateMachine_0 = LocomotionStopStateMachine() {is=title AnimGraphNode_StateMachine, bAllowConduitEntryStates=true}
   Output Animation Pose(Result=@AnimGraphNode_StateMachine_0)
  ]
 }
 LocomotionStopStateMachine {
  entry=Conduit
  path=AnimGraph.AnimGraphNode_StateMachine_5.Locomotion.AnimStateNode_4.LocomotionStop.AnimGraphNode_StateMachine_0.LocomotionStopStateMachine
  ids=id
  nodes [
   AnimStateConduitNode_0 = Conduit()
   AnimStateNode_1 = SlowStop() {is=title AnimStateNode}
   AnimStateNode_4 = FastStop() {is=title AnimStateNode}
   AnimStateTransitionNode_1 = AnimStateTransitionNode() {
    crossfadeDuration=0.4
    logicType=TLT_Inertialization
    blendProfileWrapper {blendProfile=&SitPlant_Quadruped_Skeleton:TimeBlend_FastFeetSlowBody}
   }
   AnimStateTransitionNode_4 = AnimStateTransitionNode() {priorityOrder=100}
  ]
  links [
   {
    dst=FastStop
    rule=AnimGraph.AnimGraphNode_StateMachine_5.Locomotion.AnimStateNode_4.LocomotionStop.AnimGraphNode_StateMachine_0.LocomotionStopStateMachine.AnimStateTransitionNode_4.Transition
    src=Conduit
    via=AnimStateTransitionNode_4
   }
   {
    dst=SlowStop
    rule=AnimGraph.AnimGraphNode_StateMachine_5.Locomotion.AnimStateNode_4.LocomotionStop.AnimGraphNode_StateMachine_0.LocomotionStopStateMachine.AnimStateTransitionNode_1.Transition
    src=Conduit
    via=AnimStateTransitionNode_1
   }
  ]
 }
 Conduit {
  path=AnimGraph.AnimGraphNode_StateMachine_5.Locomotion.AnimStateNode_4.LocomotionStop.AnimGraphNode_StateMachine_0.LocomotionStopStateMachine.AnimStateConduitNode_0.Conduit
  nodes [Result(bCanEnterTransition=true)]
 }
 SlowStop {
  path=AnimGraph.AnimGraphNode_StateMachine_5.Locomotion.AnimStateNode_4.LocomotionStop.AnimGraphNode_StateMachine_0.LocomotionStopStateMachine.AnimStateNode_1.SlowStop
  nodes [
   n1 = Get Quadruped_Base_Struct()
   Output Animation Pose(Result=Sequence Player(Sequence=Break Quadruped Base Struct(Quadruped_Base_Struct=@n1).LocomotionStop_Slow_6_#2) {bLoopAnimation=false})
  ]
 }
 FastStop {
  path=AnimGraph.AnimGraphNode_StateMachine_5.Locomotion.AnimStateNode_4.LocomotionStop.AnimGraphNode_StateMachine_0.LocomotionStopStateMachine.AnimStateNode_4.FastStop
  nodes [
   n1 = Get Quadruped_Base_Struct()
   Output Animation Pose(Result=Sequence Player(Sequence=Break Quadruped Base Struct(Quadruped_Base_Struct=@n1).LocomotionStop_Fast_8_#3) {bLoopAnimation=false})
  ]
 }
 Transition {
  path=AnimGraph.AnimGraphNode_StateMachine_5.Locomotion.AnimStateNode_4.LocomotionStop.AnimGraphNode_StateMachine_0.LocomotionStopStateMachine.AnimStateTransitionNode_1.Transition
  nodes [Result(bCanEnterTransition=Get bIsWalkCurveActive() -> (bIsWalkCurveActive:bool))]
 }
 Transition {
  path=AnimGraph.AnimGraphNode_StateMachine_5.Locomotion.AnimStateNode_4.LocomotionStop.AnimGraphNode_StateMachine_0.LocomotionStopStateMachine.AnimStateTransitionNode_4.Transition
  nodes [Result(bCanEnterTransition=true)]
 }
 Start Turn {
  path=AnimGraph.AnimGraphNode_StateMachine_5.Locomotion.AnimStateNode_5.Start Turn
  nodes [
   n1 = Break Quadruped Base Struct(Quadruped_Base_Struct=Get Quadruped_Base_Struct())
   Output Animation Pose(Result=Modify Curve(SourcePose=Sequence Player(Sequence=@n1) {bLoopAnimation=false}, CurveValues_0=1, Alpha=1) {curveNames [walk], curveValues [0]}) {
    node {stateEntryFunction {functionName=OnStateEntry}, stateExitFunction {functionName=OnStateExit}}
    stateEntryFunction {bSelfContext=true, memberGuid=#48, memberName=LocoStart_Entry}
    stateExitFunction {bSelfContext=true, memberGuid=#49, memberName=StartTurn_StateExit}
   }
  ]
 }
 Transition {
  path=AnimGraph.AnimGraphNode_StateMachine_5.Locomotion.AnimStateTransitionNode_1.Transition
  nodes [Result(bCanEnterTransition=Get bIsLocomotionStop() -> (bIsLocomotionStop:bool))]
 }
 Transition {
  path=AnimGraph.AnimGraphNode_StateMachine_5.Locomotion.AnimStateTransitionNode_10.Transition
  nodes [Result(bCanEnterTransition=Get bIsStopToLocomotion() -> (bIsStopToLocomotion:bool))]
 }
 Transition {
  path=AnimGraph.AnimGraphNode_StateMachine_5.Locomotion.AnimStateTransitionNode_11.Transition
  nodes [Result(bCanEnterTransition=Get bIsMoving())]
 }
 Transition {path=AnimGraph.AnimGraphNode_StateMachine_5.Locomotion.AnimStateTransitionNode_18.Transition, nodes [Result()]}
 Transition {
  path=AnimGraph.AnimGraphNode_StateMachine_5.Locomotion.AnimStateTransitionNode_3.Transition
  nodes [Result(bCanEnterTransition=Get bTriggerStartTurn() -> (bTriggerStartTurn:bool))]
 }
 Transition {
  path=AnimGraph.AnimGraphNode_StateMachine_5.Locomotion.AnimStateTransitionNode_4.Transition
  nodes [Result(bCanEnterTransition=Get bExitLocomotion() -> (bExitLocomotion:bool))]
 }
 Transition {path=AnimGraph.AnimGraphNode_StateMachine_5.Locomotion.AnimStateTransitionNode_5.Transition, nodes [Result()]}
 BodyAdditives {
  entry=BodyAdditiveTempState
  path=AnimGraph.AnimGraphNode_StateMachine_6.BodyAdditives
  ids=id
  nodes [
   AnimStateNode_0 = BodyAdditiveTempState() {is=title AnimStateNode}
   AnimStateNode_1 = Worried() {is=title AnimStateNode}
   AnimStateNode_2 = Combat()
   AnimStateTransitionNode_0 = AnimStateTransitionNode() {crossfadeDuration=0.35, logicType=TLT_Inertialization}
   AnimStateTransitionNode_1 = AnimStateTransitionNode() {crossfadeDuration=0.35, logicType=TLT_Inertialization}
   AnimStateTransitionNode_2 = AnimStateTransitionNode() {logicType=TLT_Inertialization, priorityOrder=2}
   AnimStateTransitionNode_3 = AnimStateTransitionNode()
  ]
  links [
   {
    dst=Combat
    rule=AnimGraph.AnimGraphNode_StateMachine_6.BodyAdditives.AnimStateTransitionNode_2.Transition
    src=BodyAdditiveTempState
    via=AnimStateTransitionNode_2
   }
   {
    dst=Worried
    rule=AnimGraph.AnimGraphNode_StateMachine_6.BodyAdditives.AnimStateTransitionNode_0.Transition
    src=BodyAdditiveTempState
    via=AnimStateTransitionNode_0
   }
   {
    dst=BodyAdditiveTempState
    rule=AnimGraph.AnimGraphNode_StateMachine_6.BodyAdditives.AnimStateTransitionNode_3.Transition
    src=Combat
    via=AnimStateTransitionNode_3
   }
   {
    dst=BodyAdditiveTempState
    rule=AnimGraph.AnimGraphNode_StateMachine_6.BodyAdditives.AnimStateTransitionNode_1.Transition
    src=Worried
    via=AnimStateTransitionNode_1
   }
  ]
 }
 BodyAdditiveTempState {
  path=AnimGraph.AnimGraphNode_StateMachine_6.BodyAdditives.AnimStateNode_0.BodyAdditiveTempState
  nodes [Output Animation Pose(Result=Additive Identity Pose())]
 }
 Worried {
  path=AnimGraph.AnimGraphNode_StateMachine_6.BodyAdditives.AnimStateNode_1.Worried
  nodes [
   n1 = Break Quadruped Base Struct(Quadruped_Base_Struct=Get Quadruped_Base_Struct())
   Output Animation Pose(Result=Two Way Blend(
     A=Additive Identity Pose()
     B=Sequence Player(Sequence=@n1.BodyPose_Worried_35_#15)
     Alpha=Get WorriedAddPoseAlpha() -> (WorriedAddPoseAlpha:double)
    ) {blendNode {alphaBoolBlend {blendOption=Linear}}})
  ]
 }
 Combat {
  path=AnimGraph.AnimGraphNode_StateMachine_6.BodyAdditives.AnimStateNode_2.Combat
  nodes [
   n1 = Break Quadruped Base Struct(Quadruped_Base_Struct=Get Quadruped_Base_Struct())
   Output Animation Pose(Result=Two Way Blend!(A=Sequence Player(Sequence=@n1.BodyPose_Combat_37_#16), B=Additive Identity Pose(), AlphaCurveName:Name=Idle) -> (Pose) {blendNode {alphaInputType=Curve, alphaBoolBlend {blendOption=Linear}}})
  ]
 }
 Transition {
  path=AnimGraph.AnimGraphNode_StateMachine_6.BodyAdditives.AnimStateTransitionNode_0.Transition
  nodes [Result(bCanEnterTransition=Get bIsMoodWorried() -> (bIsMoodWorried:bool))]
 }
 Transition {
  path=AnimGraph.AnimGraphNode_StateMachine_6.BodyAdditives.AnimStateTransitionNode_1.Transition
  nodes [Result(bCanEnterTransition=NOT Boolean(A=Get bIsMoodWorried()))]
 }
 Transition {
  path=AnimGraph.AnimGraphNode_StateMachine_6.BodyAdditives.AnimStateTransitionNode_2.Transition
  nodes [
   Result(bCanEnterTransition=Get bIsDelayedMoodCombat())
   Get bPlayCombatBodyAdditive() -> (bPlayCombatBodyAdditive:bool)
  ]
 }
 Transition {
  path=AnimGraph.AnimGraphNode_StateMachine_6.BodyAdditives.AnimStateTransitionNode_3.Transition
  nodes [
   Result(bCanEnterTransition=NOT Boolean(A=Get bIsMoodCombat()))
   Get bPlayCombatBodyAdditive()
  ]
 }
 AnticPlaying_Enter {nodes [
   AnticPlaying_Enter!() -> (then, Context:"Anim Update Context Structure (by ref)", Node:"Anim Node Reference Structure (by ref)") {is=title K2Node_FunctionEntry}
   > Set bIsPlayingAntic(bIsPlayingAntic:bool=true) -> (Output_Get:bool)
   > Return Node!(execute)
  ]}
 AnticPlaying_Exit {nodes [
   AnticPlaying_Exit!() -> (then, Context:"Anim Update Context Structure (by ref)", Node:"Anim Node Reference Structure (by ref)") {is=title K2Node_FunctionEntry}
   > Set bIsPlayingAntic(bIsPlayingAntic=false)
   > Set bIsAllowedToSit(bIsAllowedToSit=true)
   > Return Node()
  ]}
 CheckForTailChooser {nodes [
   CheckForTailChooser() {is=title K2Node_FunctionEntry}
   > n1 = Evaluate Proxy: PRA_Companion_TailAdditive!(execute:Exec, ProxyTable:Proxy Table Object Reference=Get ChooserTables() -> (ChooserTables:Proxy Table Object Reference)) -> (execute:Exec, Result:Object Reference) {is=title K2Node_EvaluateProxy2, proxy=&PRA_Companion_TailAdditive}
   > Set bHasTailAdditive(bHasTailAdditive:bool=IsValid(Object=@n1)) -> (Output_Get:bool)
  ]}
 DefaultTailWag {nodes [
   n1 = Break Quadruped Base Struct(Quadruped_Base_Struct=Get Quadruped_Base_Struct())
   n2 = Get TailWagSpeed() -> (TailWagSpeed:float)
   Output Pose(Result=Blend Poses by bool(
     BlendPose_0=Sequence Player(Sequence=@n1.TallWag_Sit_28_#13)
     BlendPose_1=Blendspace Player(X=@n2, Y=0, BlendSpace=@n1.Tall_Wag_BS_32_#12)
     BlendTime_0=0.3
     BlendTime_1=0.3
     bActiveValue=Get SitTail() -> (SitTail:bool)
    ) {blendType=Linear}) {is=title AnimGraphNode_Root}
   Two Way Blend!(A, B, Alpha:float=Get SitWeight() -> (SitWeight:float)) -> (Pose)
   Get bIsSitting()
  ]}
 EventGraph {nodes [
   Event BlueprintInitializeAnimation()
   > n1 = Sequence()
   n2 = GetAnimDataAsset!(self:Fort Player Companion Anim Instance Object Reference) -> (ReturnValue:Fort User Defined Data Asset Object Reference)
   > n3 = Branch(Condition=IsValid(Object=@n2))
   n4 = Get Data!(self:Fort User Defined Data Asset Object Reference=@n2) -> (bSucceed:bool, Dest:Quadruped Base Struct Structure) {is=title K2Node_CallFunction}
   > n5 = Branch(Condition=@n4)
   > Set Quadruped_Base_Struct(Quadruped_Base_Struct:Quadruped Base Struct Structure=@n4.Dest) -> (Output_Get:Quadruped Base Struct Structure)
   n6 = Get Data(self=Get Quadruped_DA() -> (Quadruped_DA:Fort User Defined Data Asset Object Reference))
   Branch(execute=[@n3.else, @n5.else], Condition=@n6)
   > Set Quadruped_Base_Struct(Quadruped_Base_Struct=@n6.Dest)
   n7 = Get Quadruped_Base_Struct()
   Set bHasChonkyAdditivePose(
    execute=@n1.then_1
    bHasChonkyAdditivePose:bool=IsValid(Object=Break Quadruped Base Struct!(Quadruped_Base_Struct=@n7) -> (Chonky_Sit_ADD_60_#24))
   ) -> (Output_Get:bool)
   > CheckForTailChooser() {is=title K2Node_CallFunction}
  ]}
 GetSequencerAnimSlotNames {nodes [
   GetSequencerAnimSlotNames() {is=title K2Node_FunctionEntry}
   > Return Node!(execute, ReturnValue:Array of Names=Make Array("[0]":Name=SequencerDefault) -> (Array:Array of Names))
  ]}
 LocoStart_Entry {nodes [
   LocoStart_Entry!() -> (then, Context:"Anim Update Context Structure (by ref)", Node:"Anim Node Reference Structure (by ref)") {is=title K2Node_FunctionEntry}
   > Set bIsLocoStart(bIsLocoStart:bool=true) -> (Output_Get:bool)
   > Return Node!(execute)
  ]}
 LocomotionLayer {nodes [
   AnimGraphNode_ChooserPlayer_0 = "Chooser Player: "!() -> (Pose:Pose Link Structure) {
    is=title AnimGraphNode_ChooserPlayer
    evaluationFrequency=OnUpdate
    chooser {_structType=/Script/Chooser.EvaluateChooser, chooser=null}
    chooserContextDefinition [
     {
      _structType=&ContextObjectTypeClass
      class=&Quadruped_Base_AnimBP_C
      direction=ReadWrite
     }
     {struct=&ChooserPlayerSettings, _structType=&ContextObjectTypeStruct, direction=Read}
    ]
    defaultSettings {bForceLooping=true, curveOverrides {hash=3351501126}}
   }
   n1 = Break Quadruped Base Struct!(Quadruped_Base_Struct=Get Quadruped_Base_Struct()) -> (
    LocomotionStart_4_#1:Anim Sequence Object Reference
    LocomotionStop_Slow_6_#2:Anim Sequence Object Reference
    LocomotionStop_Fast_8_#3:Anim Sequence Object Reference
    IdlePose_10_#4:Anim Sequence Object Reference
    IdleAngry_Loop_12_#5:Anim Sequence Object Reference
    BodyTurn_BS_51_#6:Blend Space 1D Object Reference
    Breathing_Loop_16_#7:Anim Sequence Object Reference
    Locomotion_Sprint_18_#8:Anim Sequence Object Reference
    Locomotion_Run_20_#9:Anim Sequence Object Reference
    Locomotion_Trot_22_#10:Anim Sequence Object Reference
    Locomotion_Walk_24_#11:Anim Sequence Object Reference
    Tall_Wag_BS_32_#12:Blend Space 1D Object Reference
    TallWag_Sit_28_#13:Anim Sequence Object Reference
    Tail_Wag_Angry_30_#14:Anim Sequence Object Reference
    BodyPose_Worried_35_#15:Anim Sequence Object Reference
    BodyPose_Combat_37_#16:Anim Sequence Object Reference
    SittingToStand_39_#17:Anim Sequence Object Reference
    Sitting_41_#18:Anim Sequence Object Reference
    StandToSitting_43_#19:Anim Sequence Object Reference
    AngryIdle_Intro_54_#20:Anim Sequence Object Reference
    AngryIdle_Exit_45_#21:Anim Sequence Object Reference
    IdleAntic_47_#22:Anim Sequence Object Reference
    AimOffset_50_#23:Blend Space Object Reference
    Chonky_Sit_ADD_60_#24
   )
   n2 = Get WalkPlayRate() -> (WalkPlayRate:double)
   n3 = Get TrotPlayRate() -> (TrotPlayRate:double)
   n4 = "Blend Poses (ELocomotionMovementType)"!(
    BlendPose_0:Pose Link Structure
    BlendPose_1:Pose Link Structure=Sequence Player!(Sequence=@n1.Locomotion_Sprint_18_#8, PlayRate:float=1) -> (Pose) {groupName=Locomotion, method=SyncGroup}
    BlendPose_2:Pose Link Structure=Sequence Player(Sequence=@n1.Locomotion_Run_20_#9, PlayRate=Get RunPlayRate() -> (RunPlayRate:double)) {groupName=Locomotion, method=SyncGroup}
    BlendPose_3:Pose Link Structure=Modify Curve(
     SourcePose=Sequence Player(Sequence=@n1.Locomotion_Trot_22_#10, PlayRate=@n3) {groupName=Locomotion, method=SyncGroup}
     CurveValues_0=1
     Alpha=1
    ) {curveNames [Walk], curveValues [0]}
    BlendPose_4:Pose Link Structure=Modify Curve(
     SourcePose=Sequence Player(Sequence=@n1.Locomotion_Walk_24_#11, PlayRate=@n2) {groupName=Locomotion, method=SyncGroup}
     CurveValues_0=1
     Alpha=1
    ) {curveNames [Walk], curveValues [0]}
    BlendTime_0:float=0.2
    BlendTime_1:float=0.2
    BlendTime_2:float=0.2
    BlendTime_3:float=0.2
    BlendTime_4:float=0.2
    ActiveEnumValue:Byte=Get CurrentLocomotionMovementType() -> (CurrentLocomotionMovementType:ELocomotionMovementType Enum=Walk)
   ) -> (Pose:Pose Link Structure) {
    is=title AnimGraphNode_BlendListByEnum
    blendPose [
     {linkId=-1, sourceLinkId=-1}
     {linkId=-1, sourceLinkId=-1}
     {linkId=-1, sourceLinkId=-1}
     {linkId=-1, sourceLinkId=-1}
     {linkId=-1, sourceLinkId=-1}
    ]
    blendTime [0.1,0.1,0.1,0.1,0.1]
   }
   Output Pose(Result=Modify Curve(SourcePose=@n4, CurveValues_0=1, Alpha=1) {curveNames [Locomotion], curveValues [0]})
  ]}
 AnimationBlendStackGraph_0 {
  path=LocomotionLayer.AnimGraphNode_ChooserPlayer_0.AnimationBlendStackGraph_0
  nodes [
   n1 = Blend Stack Input!(PlayRate:float=Get LocoPlayrate() -> (LocoPlayrate:double)) -> (Pose) {
    node {bOverridePlayRate=true, updateFunction {functionName=OnUpdateLocomotionPlayrate}}
    updateFunction {bSelfContext=true, memberGuid=#50, memberName=OnUpdateLocomotionPlayrate}
   }
   Output Pose(Result=@n1) {is=title AnimGraphNode_BlendStackResult}
  ]
 }
 OnUpdateLocomotionPlayrate {nodes [
   OnUpdateLocomotionPlayrate!() -> (then, Context:"Anim Update Context Structure (by ref)", Node:"Anim Node Reference Structure (by ref)") {is=title K2Node_FunctionEntry}
   n1 = GetCurveValue(CurveName=Speed)
   n2 = Get CompanionSpeed2D() -> (CompanionSpeed2D:float)
   > Set LocoPlayrate(LocoPlayrate:double="Clamp (Float)"(Value=SafeDivide(A=@n2, B=@n1), Min=0.8, Max=1.2)) -> (Output_Get:double)
   > Return Node()
  ]}
 SpecialIdle_StateEntry {nodes [
   SpecialIdle_StateEntry!() -> (then, Context:"Anim Update Context Structure (by ref)", Node:"Anim Node Reference Structure (by ref)") {is=title K2Node_FunctionEntry}
   > Set bIsSpecialIdlePlaying(bIsSpecialIdlePlaying:bool=true) -> (Output_Get:bool)
   > Return Node()
  ]}
 SpecialIdle_StateExit {nodes [
   SpecialIdle_StateExit!() -> (then, Context:"Anim Update Context Structure (by ref)", Node:"Anim Node Reference Structure (by ref)") {is=title K2Node_FunctionEntry}
   > Set bIsSpecialIdlePlaying(bIsSpecialIdlePlaying=false)
   > Return Node()
  ]}
 StartTurn_StateExit {nodes [
   StartTurn_StateExit!() -> (then, Context:"Anim Update Context Structure (by ref)", Node:"Anim Node Reference Structure (by ref)") {is=title K2Node_FunctionEntry}
   > Set UpdatedGoalRotation(UpdatedGoalRotation:float=0) -> (Output_Get:float)
   > Set bIsLocoStart(bIsLocoStart=false)
   > Return Node()
  ]}
]
variables [
 ChooserTables {type=ProxyTable, default=null}
 LocoPlayrate {type=number}
 Quadruped_Base_Struct {
  type=Quadruped_Base_Struct
  default {
   aimOffset=null
   angryIdle_Exit=null
   angryIdle_Intro=null
   bodyPose_Combat=null
   bodyPose_Worried=null
   bodyTurn_BS=null
   breathing_Loop=null
   chonky_Sit_ADD=null
   idleAngry_Loop=null
   idleAntic=null
   idlePose=null
   locomotionStart=null
   locomotionStop_Fast=null
   locomotionStop_Slow=null
   locomotion_Run=null
   locomotion_Sprint=null
   locomotion_Trot=null
   locomotion_Walk=null
   sitting=null
   sittingToStand=null
   standToSitting=null
   tail_Wag_Angry=null
   tallWag_Sit=null
   tall_Wag_BS=null
  }
 }
 Quadruped_DA {type=FortUserDefinedDataAsset, default=null}
 RandomizationContext {type=ChooserRandomizationContext, default {}}
 SitTail {type=boolean}
]
